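# Niagara skydiving contrail for the StunningMountain cosmetic, drawing flickering lightning between a shared origin point and four skeletal mesh bones.
# Decal sprites sit on the bones, arcs cover both origin and bones, and a pulsing light sits below the origin.
# Main_lightning partitions particles into four bolts using system-shared random values, with point count set by quality level.
# Camera-based values adjust arc size and visibility; Debug_Dot and Lightning_strings are disabled.
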NiagaraSystem /BRCosmetics/Athena/Effects/Particles/Skydiving/NS_StunningMountain/NS_StunningMountainContrail
dependencies {
 usedDataInterfaces [
  &NiagaraDataInterfaceArrayFloat3
  &NiagaraDataInterfaceCurve
  &NiagaraDataInterfaceCamera
  &NiagaraDataInterfaceVectorCurve
  &NiagaraDataInterfacePlatformSet
  &NiagaraDataInterfaceArrayColor
  &NiagaraDataInterfaceArrayInt32
  &NiagaraDataInterfaceSkeletalMesh
 ]
 usedDynamicInputs [&FN_FrontEndFloatModifier, &TransformPosition]
 usedModules [
  &self:SystemSpawnScript.SystemScriptSource.SystemScriptGraph.NiagaraNodeAssignment_7.SetVariables_#1
  &SystemState
  /Game/Effects/Niagara/Modules/Skydiving/FN_DivingAttributes/v5/FN_DivingAttributes
  &Timeline
  &self:SystemSpawnScript.SystemScriptSource.SystemScriptGraph.NiagaraNodeAssignment_2.SetVariables_#2
  &self:SystemSpawnScript.SystemScriptSource.SystemScriptGraph.NiagaraNodeAssignment_3.SetVariables_#3
  &self:SystemSpawnScript.SystemScriptSource.SystemScriptGraph.NiagaraNodeAssignment_4.SetVariables_#4
  &self:SystemSpawnScript.SystemScriptSource.SystemScriptGraph.NiagaraNodeAssignment_0.SetVariables_#5
  &self:SystemSpawnScript.SystemScriptSource.SystemScriptGraph.NiagaraNodeAssignment_1.SetVariables_#6
  &self:SystemSpawnScript.SystemScriptSource.SystemScriptGraph.NiagaraNodeAssignment_5.SetVariables_#7
  &self:SystemSpawnScript.SystemScriptSource.SystemScriptGraph.NiagaraNodeAssignment_6.SetVariables_#8
  &self:SystemSpawnScript.SystemScriptSource.SystemScriptGraph.NiagaraNodeAssignment_8.SetVariables_#9
  &self:SystemSpawnScript.SystemScriptSource.SystemScriptGraph.NiagaraNodeAssignment_9.SetVariables_#10
  &self:SystemSpawnScript.SystemScriptSource.SystemScriptGraph.NiagaraNodeAssignment_14.SetVariables_#11
  &self:SystemSpawnScript.SystemScriptSource.SystemScriptGraph.NiagaraNodeAssignment_13.SetVariables_#12
  &self:SystemSpawnScript.SystemScriptSource.SystemScriptGraph.NiagaraNodeAssignment_12.SetVariables_#13
 ]
}
emitters {
 Debug_Dot {
  bEnabled=false
  emitterUpdateScript [
   &EmitterState()
   &SpawnBurst_Instantaneous(Use Spawn Probability=false, UseLoopCountLimit=false)
  ]
  particleSpawnScript [
   &InitializeParticle(Color Mode=::NewEnumerator0, Lifetime=5, Position Offset=<0 0 100>, UsePositionOffset=true, Write Material Random=true)
   &CameraOffset(Camera Offset Amount=100)
  ]
  particleUpdateScript [
   &ParticleState()
   &DynamicMaterialParameters(
    Index 0 Param 1=&DotProduct(
     Vector A=$Engine.Owner.SystemZAxis
     Vector B=&ReturnCameraVector(Camera Vectors=::NewEnumerator1, Coordinate Space=::Simulation)
    )
    Module.Index0Param0Write=true
    Module.Index0Param1Write=true
    Module.Index0Param2Write=true
    Module.Index0Param3Write=true
   )
   &DynamicMaterialParameters(
    Index 0 Param 1=&DotProduct(Vector A=&ReturnCameraVector(Camera Vectors=::NewEnumerator1, Coordinate Space=::Simulation), Vector B=<1 0 1>)
    Module.Index0Param0Write=true
    Module.Index0Param1Write=true
    Module.Index0Param2Write=true
    Module.Index0Param3Write=true
   ) {module=DynamicMaterialParameters001}
  ]
  rendererClasses [&NiagaraSpriteRendererProperties]
  renderers [{rendererClass=&NiagaraSpriteRendererProperties}]
 }
 Decal {
  emitterUpdateScript [
   &EmitterState(
    Delay First Loop Only=true
    Life Cycle Mode=::NewEnumerator1
    Loop Delay=0.7
    Loop Duration=1
    Recalculate Duration Each Loop=true
    UseLoopDelay=true
   )
   /Game/Effects/Niagara/Modules/Spawn/SpawnRateFromMultipleSpawnGroups_2(Global Spawnrate=8)
   &SpawnRate(SpawnRate=14, Use Spawn Probability=false) {enabled=false}
  ]
  particleSpawnScript [
   /Niagara/Modules/Spawn/Initialization/InitializeParticle(
    Color=<rgba 0.9062 0.4123 0.1067 0.6>
    Lifetime=0.6
    Module.Write Color=true
    Module.Write Lifetime=true
    Module.Write Mass=true
    Module.Write Material Random=true
    Module.Write Position=false
    Module.Write Scale=false
    Module.Write SpriteRotation=false
    Module.Write SpriteSize=true
    Position=&SimulationPosition()
    Sprite Size=<40 40>
   ) {enabled=false}
   /Niagara/Modules/Spawn/Initialization/V2/InitializeParticle(
    Color=<rgba 0.5 1.4869 2 1>
    Lifetime=0.6
    Mass Mode=::NewEnumerator1
    Sprite Size=<50 50>
    Sprite Size Mode=::NewEnumerator4
    UsePositionOffset=false
    Write Material Random=true
   ) {module=InitializeParticle001}
   &CameraOffset()
   &PartitionParticles(Number Of Partitions=4)
  ]
  particleUpdateScript [
   &ParticleState()
   &SkeletalMeshLocation(
    Bone / Socket Index=$Particles.PartitionParticles.Partition
    Bone Sampling Mode=::NewEnumerator4
    Inherit Velocity=false
    Interpolate Mesh Sampling=false
    MaskSpawnGroup=false
    Sampled Position Offset=&SelectVectorFromArray(
     Direct Array Index=$Particles.PartitionParticles.Partition
     Vector Selection Array=DataInterface() {propertyValues {FloatData [<10 0 0>, <-10 0 0>, <-7 5 0>, <7 -5 0>]}}
    )
    Sampled Position Offset Coordinate Space:ENiagaraMeshTransforms=::NewEnumerator3
    Skeletal Mesh=$System.Skeletal Mesh
   )
   &ScaleColor(
    Scale Alpha=&FloatFromCurve(FloatCurve=DataInterface() {propertyValues {Curve {keys [
         <at 0.0001 0 RCIM_Cubic>
         {arriveTangent=1.6347, interpMode=RCIM_Cubic, leaveTangent=1.6347, time=0.1393, value=0.6251}
         {arriveTangent=0.8334, interpMode=RCIM_Cubic, leaveTangent=0.8334, time=0.3202, value=0.5233}
         {arriveTangent=-0.4566, interpMode=RCIM_Cubic, leaveTangent=-0.4566, time=0.4613, value=0.8934}
         {arriveTangent=-0.922, interpMode=RCIM_Cubic, leaveTangent=-0.922, time=0.6241, value=0.3845}
         {arriveTangent=-1.0215, interpMode=RCIM_Cubic, leaveTangent=-1.0215, time=0.7724, value=0.6066}
         <at 1.0006 0 RCIM_Cubic>
        ]}}})
    ScaleA=true
    ScaleRGB=false
   )
   /SetVariables(
    Particles.Linear Color=&MultiplyLinearColor(
     Linear Color=$Particles.Color
     Scale Factor=&MakeFloat4FromVectorAndFloat("Vector (RGB)"=&VectorFromFloat(Value=0.1), "Float (Alpha)"=1)
    )
    Particles.Vector 2D:Vector2f=&Multiply_Vector2D_ByFloat(Float=0.5, Vector2D=$Particles.SpriteSize)
    Particles.float=15
   ) {isSetParameters=true}
   /SetVariables(Particles.VisibilityTag=&MakeCustomIntFromBool(Bool=&SetBoolByFloatComparison(A=$System.CloudDot, B=0.5), True INT=0, False INT=1)) {isSetParameters=true}
  ]
  rendererClasses [&NiagaraSpriteRendererProperties]
  renderers [
   {rendererClass=&NiagaraSpriteRendererProperties}
   {rendererClass=&NiagaraSpriteRendererProperties, rendererIndex=1}
   {rendererClass=&NiagaraSpriteRendererProperties, rendererIndex=2}
   {rendererClass=&NiagaraSpriteRendererProperties, rendererIndex=3}
  ]
 }
 Light {
  emitterUpdateScript [
   &EmitterState(Delay First Loop Only=true, Life Cycle Mode=::NewEnumerator1, Loop Delay=0.5, Loop Duration=1, UseLoopDelay=true)
   &SpawnBurst_Instantaneous(Use Spawn Probability=false, UseLoopCountLimit=true)
  ]
  particleSpawnScript [&InitializeParticle(Color=<rgba 0 18.75 25 1>, Lifetime=5, UsePositionOffset=false, Write Material Random=true)]
  particleUpdateScript [
   &ParticleState(Kill Particles When Lifetime Has Elapsed=false, Let Infinitely Lived Particles Die When Emitter Deactivates=true)
   /SetVariables(Particles.Position=&AddVectorToPosition(Position=$System.Lighting Start, Vector=<0 0 -65>)) {isSetParameters=true}
   &ScaleColor(
    Scale Alpha=&FloatFromCurve(
     FloatCurve=DataInterface() {propertyValues {Curve {keys [{tangentMode=RCTM_User, time=0, value=0}, {tangentMode=RCTM_User, time=0.35, value=1}]}}}
     Scale Curve=&Waveform("Waveform [1]"=::NewEnumerator12, "[1] Frequency"=5, "[1] Amplitude Min/Max"=<0.5 1>)
    )
    ScaleA=true
    ScaleRGB=true
   )
  ]
  rendererClasses [&NiagaraLightRendererProperties]
  renderers [{rendererClass=&NiagaraLightRendererProperties, rendererIndex=0}]
 }
 Lightning_Arcs {
  emitterUpdateScript [
   &EmitterState(Life Cycle Mode=::NewEnumerator1, Loop Delay=0.5, Loop Duration=35790.8945, UseLoopDelay=true)
   &SpawnRate(SpawnGroup=1, SpawnRate=&Multiply_FloatByInt(Float=5, Integer=4), Use Spawn Probability=false) {module=SpawnRate001}
   &SpawnBurst_Instantaneous(Use Spawn Probability=false, UseLoopCountLimit=false) {enabled=false}
   &SpawnRate(SpawnRate=30, Use Spawn Probability=false) {enabled=false}
  ]
  particleSpawnScript [&InitializeParticle(
    Color=<rgba 0.75 2.2303 3 1>
    Lifetime Max=0.2
    Lifetime Min=0.25
    Lifetime Mode=::NewEnumerator1
    Mass Mode=::NewEnumerator3
    Position Mode=::NewEnumerator0
    Sprite Rotation Mode=::NewEnumerator2
    Sprite Size=&MaskVector2dBySpawnGroup(
     Masked Vector2D=&RandomRangeVector2D(Minimum=<30 50>, Maximum=<60 40>)
     Passthrough Vector2D=&Multiply_Vector2D_ByFloat(Float=0.75, Vector2D=&RandomRangeVector2D(Minimum=<30 50>, Maximum=<40 60>))
     Spawn Group Mask=0
    )
    Sprite Size Mode=::NewEnumerator4
    Sprite UV Mode=::NewEnumerator1
    Write Material Random=true
   )]
  particleUpdateScript [
   &ParticleState()
   /SetVariables(Particles.VisibilityTag=&MakeCustomIntFromBool(
     Bool=&SetBoolByFloatComparison(A=$System.CloudDot, B=0.5)
     True INT=&RandomRangeInt(Minimum=0, Maximum=2)
     False INT=&RandomRangeInt(Minimum=3, Maximum=5)
    )) {isSetParameters=true}
   &ScaleSpriteSize(
    Uniform Curve Scale=0.7
    Uniform Curve Sprite Scale=DataInterface() {propertyValues {Curve {keys [
        {
         arriveTangent=12.1149
         arriveTangentWeight=1.0818
         interpMode=RCIM_Cubic
         leaveTangent=12.1149
         leaveTangentWeight=1.0818
         tangentMode=RCTM_User
         tangentWeightMode=RCTWM_WeightedBoth
         time=0
         value=0
        }
        {
         arriveTangentWeight=0.8185
         interpMode=RCIM_Cubic
         leaveTangentWeight=0.8185
         tangentMode=RCTM_User
         tangentWeightMode=RCTWM_WeightedBoth
         time=1
         value=1
        }
       ]}}}
   )
   &ScaleSpriteSize(
    Scale Sprite Size Mode=::NewEnumerator0
    Uniform Scale Factor=&MaskFloatBySpawnGroup(
     Masked Float=1
     Passthrough Float=&FloatFromCurve(
      FloatCurve=DataInterface() {propertyValues {Curve {keys [{tangentMode=RCTM_User, time=-1, value=0.5}, {tangentMode=RCTM_User, time=-0.7, value=1}]}}}
      CurveIndex=&DotProduct(Vector A=<0 0 1>, Vector B=$System.CamForVec)
     )
     Spawn Group Mask=1
    )
   ) {module=ScaleSpriteSize001}
   &ScaleColor(
    Scale Alpha=&FloatFromCurve(FloatCurve=DataInterface() {propertyValues {Curve {keys [
         {arriveTangent=4.1298, interpMode=RCIM_Cubic, leaveTangent=4.1298, tangentMode=RCTM_User, time=0, value=0}
         {interpMode=RCIM_Cubic, tangentMode=RCTM_User, time=0.5, value=1}
         {arriveTangent=-5.4005, interpMode=RCIM_Cubic, leaveTangent=-5.4005, tangentMode=RCTM_User, time=1, value=0}
        ]}}})
    ScaleA=true
    ScaleRGB=true
   )
   &ShapeLocation(MaskSpawnGroup=false, Override Seed=false, Shape Origin=$System.Lighting Start, Sphere Radius=3, UseOwnerScale=true)
   &ScaleColor(
    Scale RGB=&VectorFromFloat(Value=&Lerp_Float(A=0.35, B=1, Alpha=&Sine(Normalized Angle=$Emitter.Age, Period=0.1, Scale=0.5, Bias=0.5)))
    ScaleA=true
    ScaleRGB=true
   ) {module=ScaleColor001}
   &SkeletalMeshLocation(
    Bone / Socket Index=&ModuloInt(A=&ReturnParticleIDAttributes(), B=4)
    Bone Sampling Mode=::NewEnumerator4
    MaskSpawnGroup=false
    Orientation Sampling=::NewEnumerator1
    Position Sampling=::NewEnumerator1
    Sampled Position Offset=&SelectVectorFromArray(
     Direct Array Index=&ModuloInt(A=&ReturnParticleIDAttributes(), B=4)
     Vector Selection Array=DataInterface() {propertyValues {FloatData [<10 0 0>, <-10 0 0>, <-7 5 0>, <7 -5 0>]}}
    )
    Skeletal Mesh=$System.Skeletal Mesh
    Use Sampled Mesh Attribute as Ribbon ID=false
   )
   /SetVariables(Particles.Position=&self:Select_Pos(
     bool:NiagaraBool=&SetBoolByIntComparison(A=&SpawnGroup(), B=1, Comparison Type=::NewEnumerator4)
     False Position:NiagaraPosition=$Particles.ShapeLocation.ShapePosition
     True Position:NiagaraPosition=$Particles.SkeletalMeshLocation.SampledPosition
    )) {isSetParameters=true}
  ]
  rendererClasses [&NiagaraSpriteRendererProperties]
  renderers [
   {rendererClass=&NiagaraSpriteRendererProperties}
   {rendererClass=&NiagaraSpriteRendererProperties, rendererIndex=1}
  ]
 }
 Lightning_strings {
  bEnabled=false
  emitterSpawnScript [
   /SetVariables(Emitter.particles per line:NiagaraInt32=&MakeCustomIntFromBool(
     Bool=&IsPlatformSetActive(New Platform Set=DataInterface() {propertyValues {Platforms {qualityLevelMask=-31}}})
     True INT=4
     False INT=&MakeCustomIntFromBool(
      Bool=&IsPlatformSetActive(New Platform Set=DataInterface() {propertyValues {Platforms {qualityLevelMask=-26}}})
      True INT=6
      False INT=8
     )
    )) {isSetParameters=true}
   /SetVariables(Emitter.Num Groups=4) {isSetParameters=true}
  ]
  emitterUpdateScript [
   &EmitterState(Life Cycle Mode=::NewEnumerator1, Loop Delay=0.5, Loop Duration=2, Recalculate Duration Each Loop=true, UseLoopDelay=true)
   /SetVariables(Emitter.Random Int:NiagaraInt32=&RandomRangeInt(Minimum=0, Maximum=2147483640, Evaluation Type=::NewEnumerator1)) {enabled=false, isSetParameters=true}
   &SpawnBurst_Instantaneous(
    Spawn Count=&Multiply_Int(A=$Emitter.particles per line, B=$Emitter.Num Groups)
    Use Spawn Probability=false
    UseLoopCountLimit=true
   )
   /SetVariables(
    Emitter.PerPartRand:NiagaraFloat=&RandomRangeFloat(Minimum=0, Maximum=1, Evaluation Type=::NewEnumerator1)
    Emitter.PerPartRand 2:NiagaraFloat=&RandomRangeFloat(Minimum=0, Maximum=1, Evaluation Type=::NewEnumerator1)
    Emitter.PerPartRand 3:NiagaraFloat=&RandomRangeFloat(Minimum=0, Maximum=1, Evaluation Type=::NewEnumerator1)
    Emitter.PerPartRand 4:NiagaraFloat=&RandomRangeFloat(Minimum=0, Maximum=1, Evaluation Type=::NewEnumerator1)
   ) {isSetParameters=true}
   /SetVariables(
    Emitter.ini Pos 1:NiagaraFloat=&RandomRangeFloat(Minimum=0.1, Maximum=-0.1, Evaluation Type=::NewEnumerator1)
    Emitter.ini Pos 2:NiagaraFloat=&RandomRangeFloat(Minimum=0.1, Maximum=-0.1, Evaluation Type=::NewEnumerator1)
    Emitter.ini Pos 3:NiagaraFloat=&RandomRangeFloat(Minimum=0.1, Maximum=-0.1, Evaluation Type=::NewEnumerator1)
    Emitter.ini Pos 4:NiagaraFloat=&RandomRangeFloat(Minimum=-0.1, Maximum=0.1, Evaluation Type=::NewEnumerator1)
   ) {isSetParameters=true}
   /SetVariables(
    Emitter.ini Pos Y 1:NiagaraFloat=&RandomRangeFloat(Minimum=-0.1, Maximum=0, Evaluation Type=::NewEnumerator1)
    Emitter.ini Pos Y 2:NiagaraFloat=&RandomRangeFloat(Minimum=-0.1, Maximum=0, Evaluation Type=::NewEnumerator1)
    Emitter.ini Pos Y 3:NiagaraFloat=&RandomRangeFloat(Minimum=-0.1, Maximum=0, Evaluation Type=::NewEnumerator1)
    Emitter.ini Pos Y 4:NiagaraFloat=&RandomRangeFloat(Minimum=-0.1, Maximum=0, Evaluation Type=::NewEnumerator1)
   ) {isSetParameters=true}
   /SetVariables(
    Emitter.Shape 3:NiagaraInt32=&RandomRangeInt(Minimum=0, Maximum=5, Evaluation Type=::NewEnumerator1)
    Emitter.shape 1:NiagaraInt32=&RandomRangeInt(Minimum=0, Maximum=4, Evaluation Type=::NewEnumerator1)
    Emitter.shape 2:NiagaraInt32=&RandomRangeInt(Minimum=0, Maximum=5, Evaluation Type=::NewEnumerator1)
    Emitter.shape 4:NiagaraInt32=&RandomRangeInt(Minimum=0, Maximum=5, Evaluation Type=::NewEnumerator1)
   ) {isSetParameters=true}
   /SetVariables(
    Emitter.Timer 2:NiagaraFloat=&RandomRangeFloat(Minimum=0.01, Maximum=0.25, Evaluation Type=::NewEnumerator1)
    Emitter.Timer 3:NiagaraFloat=&RandomRangeFloat(Minimum=0.01, Maximum=0.25, Evaluation Type=::NewEnumerator1)
    Emitter.timer 1:NiagaraFloat=&RandomRangeFloat(Minimum=0.01, Maximum=0.25, Evaluation Type=::NewEnumerator1)
    Emitter.timer 4:NiagaraFloat=&RandomRangeFloat(Minimum=0.01, Maximum=0.25, Evaluation Type=::NewEnumerator1)
   ) {isSetParameters=true}
   /SetVariables(Emitter.Timer 2=0.5, Emitter.Timer 3=0.5, Emitter.timer 1=0.5, Emitter.timer 4=0.5) {enabled=false, isSetParameters=true}
   /SetVariables(Emitter.rand Line Pos:NiagaraFloat=&RandomRangeFloat(Minimum=0, Maximum=1)) {isSetParameters=true}
   /SetVariables(Emitter.rand Line Pos2:NiagaraFloat=&RandomRangeFloat(Minimum=0, Maximum=1)) {isSetParameters=true}
  ]
  particleSpawnScript [
   &PartitionParticles(
    Normalized Index In Partition Includes One=true
    Number Of Partitions=$Emitter.Num Groups
    Particles Per Partition:NiagaraInt32=$Emitter.particles per line
    Partition Mode:ENiagara_PartitionMode=::NewEnumerator1
    Write Partition Index to Ribbon Link Order:ENiagara_AttributeSamplingApplyOutput=::NewEnumerator0
    Write Partition to Ribbon ID=::NewEnumerator0
   )
   &InitializeParticle(
    Color=<rgba 0.15 0.9974 1.5 1>
    Lifetime Max=1.75
    Lifetime Min=1.4
    Lifetime Mode=::NewEnumerator1
    Mass Mode=::NewEnumerator3
    Ribbon Width=&Multiply_Float(
     A=&SelectFloatFromArray(
      Direct Array Index=$Particles.PartitionParticles.Partition
      Direct Array Mode=::NewEnumerator1
      Float Selection Array=DataInterface() {propertyValues {FloatData [30,-30]}}
     )
     B=&RandomRangeFloat(Minimum=0.5, Maximum=0.75)
    )
    Ribbon Width Mode=::NewEnumerator1
    Sprite Rotation Mode=::NewEnumerator2
    Sprite Size Mode=::NewEnumerator1
    Uniform Sprite Size Min=10
    UsePositionOffset=false
    Write Material Random=true
   )
   /SetVariables(Particles.Get Rand Line Pos 2:NiagaraFloat=$Emitter.rand Line Pos2) {isSetParameters=true}
   /SetVariables(Particles.Get Rand Line Pos:NiagaraFloat=$Emitter.rand Line Pos) {isSetParameters=true}
   /SetVariables(Particles.Get Rand Line Pos3:NiagaraFloat=$Emitter.rand Line Pos) {isSetParameters=true}
   /SetVariables(Particles.Get Rand Line Pos4:NiagaraFloat=$Emitter.rand Line Pos) {isSetParameters=true}
  ]
  particleUpdateScript [
   &ParticleState(Kill Particles When Lifetime Has Elapsed=false, Let Infinitely Lived Particles Die When Emitter Deactivates=true)
   &SkeletalMeshLocation(
    Bone / Socket Index=$Particles.PartitionParticles.Partition
    Bone Sampling Mode=::NewEnumerator4
    MaskSpawnGroup=false
    Orientation Sampling=::NewEnumerator1
    Position Sampling=::NewEnumerator1
    Sampled Position Offset=&SelectVectorFromArray(
     Direct Array Index=$Particles.PartitionParticles.Partition
     Vector Selection Array=DataInterface() {propertyValues {FloatData [<5 0 0>, <-5 0 1>, <-5 7 0>, <5 -7 0>]}}
    )
    Skeletal Mesh=$System.Skeletal Mesh
    Write To Particle Position=false
   )
   &self:SLect_Shape_Mod(
    RandomSelctor:NiagaraInt32=$Particles.PartitionParticles.Partition
    Select Pos:NiagaraFloat=$Emitter.ini Pos 1
    Select Pos 2:NiagaraFloat=$Emitter.ini Pos 2
    Select Pos 3:NiagaraFloat=$Emitter.ini Pos 3
    Select Pos 4:NiagaraFloat=$Emitter.ini Pos 4
    Select Pos Y:NiagaraFloat=$Emitter.ini Pos Y 1
    Select Pos Y 2:NiagaraFloat=$Emitter.ini Pos Y 2
    Select Pos Y 3:NiagaraFloat=$Emitter.ini Pos Y 3
    Select Pos Y 4:NiagaraFloat=$Emitter.ini Pos Y 4
    Shape Select:NiagaraInt32=$Emitter.shape 1
    Shape Select 2:NiagaraInt32=$Emitter.shape 2
    Shape Select 3:NiagaraInt32=$Emitter.Shape 3
    Shape Select 4:NiagaraInt32=$Emitter.shape 4
    Timer Amount:NiagaraFloat=$Emitter.timer 1
    Timer Amount 2:NiagaraFloat=$Emitter.Timer 2
    Timer Amount 3:NiagaraFloat=$Emitter.Timer 3
    Timer Amount 4:NiagaraFloat=$Emitter.timer 4
   ) {module=SLect_Shape_Mod}
   &ShapeLocation(
    Handle Radius=18
    Large Radius=&MakeFloatFromVector(
     Vector=&VectorFromFloat(Value=&Lerp_Float(
       A=50
       B=75
       Alpha=&FloatFromCurve(
        FloatCurve=DataInterface() {propertyValues {Curve {keys [<at -1 1>,<at -0.8 0>]}}}
        CurveIndex=&DotProduct(Vector A=<0 0 1>, Vector B=$System.CamForVec)
       )
      ))
     Channel=::NewEnumerator0
    )
    MaskSpawnGroup=false
    Offset=<0 0 42>
    Offset Mode=::NewEnumerator0
    Shape Origin:Vector3f=$System.Lighting Start
    Shape Primitive=::NewEnumerator3
    Torus Distribution Mode:ENiagaraTorusDistributionMode=::NewEnumerator1
    U Position=&Add_Float(
     A=&self:torus_pos_float(
      float 1:NiagaraFloat=&Add_Float(A=0.25, B=$Particles.SLect_Shape_Mod.Old Select Pos)
      float 2:NiagaraFloat=&Add_Float(A=0.5, B=$Particles.SLect_Shape_Mod.Old Select Pos 2)
      float 3:NiagaraFloat=&Add_Float(A=0, B=$Particles.SLect_Shape_Mod.Old Select Pos 3)
      float 4:NiagaraFloat=&Add_Float(A=0.75, B=$Particles.SLect_Shape_Mod.Old Select Pos 4)
      Master Offset:NiagaraFloat=-0.125
      Partition:NiagaraInt32=$Particles.PartitionParticles.Partition
     )
     B=&Multiply_Float(A=$Particles.PartitionParticles.Normalized Index In Partition, B=0.35)
    )
    UseOwnerScale=true
    V Position=&self:torus_pos_float(
     float 1:NiagaraFloat=$Particles.SLect_Shape_Mod.Old Select Pos Y
     float 2:NiagaraFloat=$Particles.SLect_Shape_Mod.Old Select Pos Y 2
     float 3:NiagaraFloat=$Particles.SLect_Shape_Mod.Old Select Pos Y 3
     float 4:NiagaraFloat=$Particles.SLect_Shape_Mod.Old Select Pos Y 4
     Master Offset:NiagaraFloat=0
     Partition:NiagaraInt32=$Particles.PartitionParticles.Partition
    )
   ) {module=ShapeLocation001}
   &self:CustomTimeline(
    Instantaneous Rewind=true
    Play:NiagaraBool=$Particles.IsLimbTimerActive
    Play Rate:NiagaraFloat=&Divide_Float(A=1, B=0.1)
   ) {module=CustomTimeline}
   /Game/Effects/Niagara/Modules/Curves/FN_HermiteSpline(
    Direction End:Vector3f=&TransformVector(
     Vector=&Multiply_Vector(
      A=&SelectVectorFromArray(
       Direct Array Index=$Particles.PartitionParticles.Partition
       Vector Selection Array=DataInterface() {propertyValues {FloatData [<150 150 100>, <-150 150 100>, <150 -150 100>, <-150 -150 100>]}}
      )
      B=&SelectVectorFromArray(
       Direct Array Index=&MakeCustomIntFromBool(Bool=$System.Use Temp, True INT=$System.Temp Array Selection, False INT=$Particles.ShapeMod)
       Vector Selection Array=$System.Vector Array bottom
      )
     )
     Destination Space=::Simulation
    )
    Direction Start:Vector3f=&TransformVector(
     Vector=&Multiply_Vector(
      A=&SelectVectorFromArray(
       Direct Array Index=$Particles.PartitionParticles.Partition
       Vector Selection Array=DataInterface() {propertyValues {FloatData [<300 50 -75>, <-300 300 -75>, <50 -300 -75>, <-300 -50 -75>]}}
      )
      B=&SelectVectorFromArray(
       Direct Array Index=&MakeCustomIntFromBool(Bool=$System.Use Temp, True INT=$System.Temp Array Selection, False INT=$Particles.ShapeMod)
       Vector Selection Array=$System.Vector Array Top
      )
     )
     Destination Space=::Simulation
    )
    Position Blend:NiagaraFloat=$Particles.PartitionParticles.Normalized Index In Partition
    Position End:Vector3f=$StackContext.SkeletalMeshLocation.SampledPosition
    Position Start:Vector3f=&LerpPosition(A=$Particles.ShapeLocation001.ShapePosition, B=Linked() {linkedVariable {name=System.Lighting Start, type=Vector3f}}, Alpha=0)
    Write Particles Position=true
   )
   &FN_HermiteSpline(
    Direction End:Vector3f=&TransformVector(
     Vector=&Multiply_Vector(
      A=&SelectVectorFromArray(
       Direct Array Index=$Particles.PartitionParticles.Partition
       Vector Selection Array=DataInterface() {propertyValues {FloatData [<150 150 100>, <-150 150 100>, <150 -150 100>, <-150 -150 100>]}}
      )
      B=&SelectVectorFromArray(
       Direct Array Index=&MakeCustomIntFromBool(Bool=$System.Use Temp, True INT=$System.Temp Array Selection, False INT=$Particles.ShapeMod)
       Vector Selection Array=$System.Vector Array bottom
      )
     )
     Destination Space=::Simulation
    )
    Direction Start:Vector3f=&TransformVector(
     Vector=&Multiply_Vector(
      A=&SelectVectorFromArray(
       Direct Array Index=$Particles.PartitionParticles.Partition
       Vector Selection Array=DataInterface() {propertyValues {FloatData [<150 150 -75>, <-150 150 -75>, <150 -150 -75>, <-150 -150 -75>]}}
      )
      B=&SelectVectorFromArray(
       Direct Array Index=&MakeCustomIntFromBool(Bool=$System.Use Temp, True INT=$System.Temp Array Selection, False INT=$Particles.ShapeMod)
       Vector Selection Array=$System.Vector Array Top
      )
     )
     Destination Space=::Simulation
    )
    Position Blend:NiagaraFloat=$Particles.PartitionParticles.Normalized Index In Partition
    Position End:Vector3f=$StackContext.SkeletalMeshLocation.SampledPosition
    Position Start:Vector3f=&LerpPosition(A=$Particles.ShapeLocation001.ShapePosition, B=Linked() {linkedVariable {name=System.Lighting Start, type=Vector3f}}, Alpha=0)
    Write Particles Position=true
   ) {enabled=false, module=FN_HermiteSpline001}
   &JitterPosition(Jitter Amount=0.65)
   &self:OffsetLightning(
    Particle Offset:Vector3f=&Multiply_Vector(
     A=&RandomVector(
      Vector Scale=&FloatFromCurve(
       FloatCurve=DataInterface() {propertyValues {Curve {keys [
           {interpMode=RCIM_Cubic, tangentMode=RCTM_User, time=0, value=0}
           {interpMode=RCIM_Cubic, tangentMode=RCTM_User, time=0.5, value=1}
           {tangentMode=RCTM_User, time=1, value=0}
          ]}}}
       CurveIndex=$Particles.PartitionParticles.Normalized Index In Partition
       Scale Curve=&RandomRangeFloat(Minimum=5, Maximum=25)
      )
      Evaluation Type=::NewEnumerator1
     )
     B=&VectorFromFloat(Value=0.35)
    )
    Timer Amount:NiagaraFloat=&FloatFromCurve(
     FloatCurve=DataInterface() {propertyValues {Curve {keys [
         {
          arriveTangent=0.466
          arriveTangentWeight=0.7477
          interpMode=RCIM_Cubic
          leaveTangent=0.466
          leaveTangentWeight=0.7477
          tangentMode=RCTM_User
          tangentWeightMode=RCTWM_WeightedBoth
          time=0
          value=0
         }
         {
          arriveTangent=0.5735
          arriveTangentWeight=0.6348
          interpMode=RCIM_Cubic
          leaveTangent=0.5735
          leaveTangentWeight=0.6348
          tangentMode=RCTM_User
          tangentWeightMode=RCTWM_WeightedBoth
          time=1
          value=1
         }
        ]}}}
     CurveIndex=&RandomRangeFloat(Minimum=0, Maximum=1, Evaluation Type=::NewEnumerator1)
    )
   ) {module=OffsetLightning}
   &ScaleColor(
    Scale Alpha=&FloatFromCurve(
     FloatCurve=DataInterface() {propertyValues {Curve {keys [<at 0.9 0 RCIM_Cubic>, {arriveTangent=1.8907, interpMode=RCIM_Cubic, leaveTangent=1.8907, tangentMode=RCTM_User, time=1.3, value=1}]}}}
     CurveIndex=&Add_Float(
      A=&Multiply_Float(
       A=&Add_Float(A=$Engine.System.Age, B=-0.5)
       B=&SelectFloatFromArray(Direct Array Index=$Particles.PartitionParticles.Partition, Float Selection Array=DataInterface() {propertyValues {FloatData [5,4,3,6]}})
      )
      B=&Add_Float(A=&OneMinusFloat(Float=$Particles.PartitionParticles.Normalized Index In Partition), B=-1)
     )
    )
    ScaleA=true
    ScaleRGB=true
   )
   &ScaleRibbonWidth(Ribbon Width Scale=&FloatFromCurve(
     FloatCurve=DataInterface() {propertyValues {Curve {keys [<at 0 0.5>,<at 0.1 1>,<at 0.8 0.9>,<at 1 0.5>]}}}
     CurveIndex=$Particles.PartitionParticles.Normalized Index In Partition
    ))
   &ScaleRibbonWidth(Ribbon Width Scale=&FloatFromCurve(
     FloatCurve=DataInterface() {propertyValues {Curve {keys [{tangentMode=RCTM_User, time=-1, value=0.75}, {tangentMode=RCTM_User, time=-0.7, value=1}]}}}
     CurveIndex=&DotProduct(Vector A=<0 0 1>, Vector B=$System.CamForVec)
    )) {module=ScaleRibbonWidth002}
   &self:Jitter_UV(
    InvertIndex=true
    NormalizedIndex:NiagaraFloat=$Particles.PartitionParticles.Normalized Index In Partition
    PerParticleRandomFloat:NiagaraFloat=&Multiply_Float(A=0, B=&RandomRangeFloat(Minimum=-1, Maximum=1, Evaluation Type=::NewEnumerator1))
    PerPartitionRandomFloat:NiagaraFloat=$Emitter.PerPartRand
    PerPartitionRandomFloat2:NiagaraFloat=$Emitter.PerPartRand 2
    PerPartitionRandomFloat3:NiagaraFloat=$Emitter.PerPartRand 3
    PerPartitionRandomFloat4:NiagaraFloat=$Emitter.PerPartRand 4
    RandomSelctor:NiagaraInt32=$Particles.PartitionParticles.Partition
    Time:NiagaraFloat=&self:ScratchDynamicInput_01(
     Steps Per Second:NiagaraFloat=&RandomRangeFloat(Minimum=5, Maximum=8)
     Time:NiagaraFloat=&Multiply_Float(A=$Engine.System.Age, B=1)
    )
    Timer Amount=15
    Use Timer=true
   ) {module=Jitter_UV}
   &ScaleColor(
    Scale RGB=&VectorFromFloat(Value=&Lerp_Float(
      A=0.2
      B=1
      Alpha=&Sine(
       Normalized Angle=&Add_Float(A=$Emitter.Age, B=&RandomRangeFloat(Minimum=0, Maximum=10))
       Period=0.25
       Scale=0.5
       Bias=0.5
      )
     ))
    ScaleA=true
    ScaleRGB=true
   ) {module=ScaleColor001}
   &ScaleColor(
    Scale Alpha=&FloatFromCurve(
     FloatCurve=DataInterface() {propertyValues {Curve {keys [<at 0 0>,<at 0.01 1>,<at 0.99 1>,<at 1 0>]}}}
     CurveIndex=$Particles.PartitionParticles.Normalized Index In Partition
    )
    ScaleA=true
    ScaleRGB=true
   ) {module=ScaleColor003}
   &ScaleColor(
    Scale RGB=&VectorFromFloat(Value=&Lerp_Float(
      A=1
      B=5
      Alpha=&FloatFromCurve(
       FloatCurve=DataInterface() {propertyValues {Curve {keys [
           {arriveTangent=-2.259, arriveTangentWeight=0.829, interpMode=RCIM_Cubic, leaveTangentWeight=0.829, time=0, value=1}
           {arriveTangentWeight=1.0015, interpMode=RCIM_Cubic, leaveTangentWeight=1.0015, time=1, value=0}
          ]}}}
       CurveIndex=$Particles.CustomTimeline.NormalizedTime
      )
     ))
    ScaleA=true
    ScaleRGB=true
   ) {module=ScaleColor004}
   /SetVariables(Particles.Top Color:LinearColor=&MultiplyLinearColor(
     Linear Color=$Particles.Color
     Scale Factor=&MakeFloat4FromVectorAndFloat(
      "Vector (RGB)"=<1 1 1>
      "Float (Alpha)"=&MakeCustomFloatFromBool(Bool=&SetBoolByFloatComparison(A=$System.CloudDot, B=0.5), True Float=0, False Float=1)
     )
    )) {isSetParameters=true}
   &ScaleColor(
    Scale Alpha=&MakeCustomFloatFromBool(Bool=&SetBoolByFloatComparison(A=$System.CloudDot, B=0.5), True Float=1, False Float=0)
    ScaleA=true
    ScaleRGB=true
   ) {module=ScaleColor002}
   &DynamicMaterialParameters(
    Index 0 Param 1=1
    Index 0 Param 2=1
    Index 0 Param 3=1
    Index 0 Param 4=1
    Index 1 Param 2=2
    Index 1 Param 3=-1
    Index 1 Param 4=10
    Index 2 Param 1=5
    Index 2 Param 3=500
    Index 2 Param 4=30
    Index 3 Param 3=1
    Index 3 Param 4=-0.7
    Module.Index0Param0Write=true
    Module.Index0Param1Write=true
    Module.Index0Param2Write=true
    Module.Index0Param3Write=true
    Module.Index1Param0Write=true
    Module.Index1Param1Write=true
    Module.Index1Param2Write=true
    Module.Index1Param3Write=true
    Module.Index2Param0Write=true
    Module.Index2Param1Write=true
    Module.Index2Param2Write=true
    Module.Index2Param3Write=true
    Module.Index3Param0Write=true
    Module.Index3Param1Write=true
    Module.Index3Param2Write=true
    Module.Index3Param3Write=true
    Write Parameter Index 1=true
    Write Parameter Index 2=true
    Write Parameter Index 3=true
   )
   /SetVariables(Particles.RibbonWidth2:NiagaraFloat=&MakeCustomFloatFromBool(Bool=&SetBoolByFloatComparison(A=$System.CloudDot, B=0.5), True Float=$Particles.RibbonWidth, False Float=0)) {isSetParameters=true}
   &ScaleRibbonWidth(Ribbon Width Scale=&MakeCustomFloatFromBool(Bool=&SetBoolByFloatComparison(A=$System.CloudDot, B=0.5), True Float=0, False Float=1)) {module=ScaleRibbonWidth001}
   /SetVariables(
    Particles.Glow Cam Offset=100
    Particles.Sprite Color:LinearColor=&MultiplyLinearColor(
     Linear Color=<rgba 0.15 0.6 1.5 0.2>
     Scale Factor=&MakeFloat4FromVectorAndFloat(
      "Vector (RGB)"=&VectorFromFloat(Value=&FNTimeOfDayScalar(Night Scale=1, Morning Scale=1, Evening Scale=1))
      "Float (Alpha)"=&FloatFromCurve(
       FloatCurve=DataInterface() {propertyValues {Curve {keys [<at -0.4 0.15>,<at -0.2 1.25>]}}}
       CurveIndex=&DotProduct(Vector A=$System.CamForVec, Vector B=<0 0 1>)
       Scale Curve=&FloatFromCurve(
        FloatCurve=DataInterface() {propertyValues {Curve {keys [{tangentMode=RCTM_User, time=0.655, value=0}, {tangentMode=RCTM_User, time=0.855, value=1}]}}}
        CurveIndex=$Engine.System.Age
       )
      )
     )
    )
    "Particles.SpriteSizeGlow ":Vector2f=&Vector2DFromFloat(Value=&FloatFromCurve(
      FloatCurve=DataInterface() {propertyValues {Curve {keys [<at 0 1 RCIM_Constant>,<at 0.01 0>]}}}
      CurveIndex=$Particles.PartitionParticles.Normalized Index In Partition
      Scale Curve=&Lerp_Float(A=40, B=40, Alpha=0.5)
     ))
    Particles.dynam:Vector4f=<2 1 0 0>
   ) {isSetParameters=true}
   /SetVariables(
    Particles.DEcalSprite Color:LinearColor=&MultiplyLinearColor(
     Linear Color=&MultiplyLinearColor(
      Linear Color=<rgba 1 4 10 1>
      Scale Factor=&MakeFloat4FromVectorAndFloat(
       "Vector (RGB)"=&VectorFromFloat(Value=&FNTimeOfDayScalar(Night Scale=0.7, Morning Scale=1, Evening Scale=1))
       "Float (Alpha)"=1
      )
     )
     Scale Factor=&MakeFloat4FromVectorAndFloat(
      "Vector (RGB)"=&VectorFromFloat(Value=&FloatFromCurve(
        FloatCurve=DataInterface() {propertyValues {Curve {keys [<at -0.4 0.15>,<at -0.2 1.25>]}}}
        CurveIndex=&DotProduct(Vector A=$System.CamForVec, Vector B=<0 0 1>)
        Scale Curve=&FloatFromCurve(
         FloatCurve=DataInterface() {propertyValues {Curve {keys [{tangentMode=RCTM_User, time=0.655, value=0}, {tangentMode=RCTM_User, time=0.855, value=1}]}}}
         CurveIndex=$Engine.System.Age
        )
       ))
      "Float (Alpha)"=1
     )
    )
    "Particles.Decal SpriteSizeGlow ":Vector2f=&Vector2DFromFloat(Value=&FloatFromCurve(
      FloatCurve=DataInterface() {propertyValues {Curve {keys [<at 0 1 RCIM_Constant>,<at 0.01 0>]}}}
      CurveIndex=$Particles.PartitionParticles.Normalized Index In Partition
      Scale Curve=&Lerp_Float(
       A=35
       B=50
       Alpha=&FloatFromCurve(
        FloatCurve=DataInterface() {propertyValues {Curve {keys [<at -0.5 0.15>,<at -0.1 1.25>]}}}
        CurveIndex=&DotProduct(Vector A=$System.CamForVec, Vector B=<0 0 1>)
        Scale Curve=&FloatFromCurve(
         FloatCurve=DataInterface() {propertyValues {Curve {keys [{tangentMode=RCTM_User, time=0.655, value=0}, {tangentMode=RCTM_User, time=0.855, value=1}]}}}
         CurveIndex=$Engine.System.Age
        )
       )
      )
     ))
   ) {isSetParameters=true}
   /SetVariables(Particles.Front decal col:LinearColor=&MultiplyLinearColor(
     Linear Color=$Particles.DEcalSprite Color
     Scale Factor=&MakeFloat4FromVectorAndFloat("Vector (RGB)"=&VectorFromFloat(Value=1), "Float (Alpha)"=0.05)
    )) {isSetParameters=true}
  ]
  rendererClasses [&NiagaraRibbonRendererProperties, &NiagaraSpriteRendererProperties]
  renderers [
   {rendererClass=&NiagaraRibbonRendererProperties}
   {rendererClass=&NiagaraRibbonRendererProperties, rendererIndex=1}
   {rendererClass=&NiagaraSpriteRendererProperties, rendererIndex=2}
   {rendererClass=&NiagaraSpriteRendererProperties, rendererIndex=3}
   {rendererClass=&NiagaraSpriteRendererProperties, rendererIndex=4}
  ]
 }
 Main_lightning {
  emitterSpawnScript [
   /SetVariables(Emitter.particles per line:NiagaraInt32=&MakeCustomIntFromBool(
     Bool=&IsPlatformSetActive(New Platform Set=DataInterface() {propertyValues {Platforms {qualityLevelMask=-31}}})
     True INT=4
     False INT=&MakeCustomIntFromBool(
      Bool=&IsPlatformSetActive(New Platform Set=DataInterface() {propertyValues {Platforms {qualityLevelMask=-26}}})
      True INT=6
      False INT=8
     )
    )) {isSetParameters=true, module=SetVariables_#14}
   /SetVariables(Emitter.Num Groups=4) {isSetParameters=true, module=SetVariables_#15}
  ]
  emitterUpdateScript [
   &EmitterState(Delay First Loop Only=true, Life Cycle Mode=::NewEnumerator1, Loop Delay=0.5, Loop Duration=1, UseLoopDelay=true)
   &SpawnBurst_Instantaneous(
    Spawn Count=&Multiply_Int(A=$Emitter.particles per line, B=$Emitter.Num Groups)
    Use Spawn Probability=false
    UseLoopCountLimit=true
   )
   /SetVariables(
    Emitter.ini Pos 1:NiagaraFloat=$System.sys ini Pos 1
    Emitter.ini Pos 2:NiagaraFloat=$System.sys ini Pos 2
    Emitter.ini Pos 3:NiagaraFloat=$System.sys ini Pos 3
    Emitter.ini Pos 4:NiagaraFloat=$System.sys ini Pos 4
   ) {isSetParameters=true}
   /SetVariables(
    Emitter.ini Pos Y 1:NiagaraFloat=$System.sys ini Pos Y 1
    Emitter.ini Pos Y 2:NiagaraFloat=$System.sys ini Pos Y 2
    Emitter.ini Pos Y 3:NiagaraFloat=$System.sys ini Pos Y 3
    Emitter.ini Pos Y 4:NiagaraFloat=$System.sys ini Pos Y 4
   ) {isSetParameters=true}
   /SetVariables(
    Emitter.Shape 3:NiagaraInt32=$System.SysShape 3
    Emitter.shape 1:NiagaraInt32=$System.SysShape 1
    Emitter.shape 2:NiagaraInt32=$System.SysShape 2
    Emitter.shape 4:NiagaraInt32=$System.SysShape 4
   ) {isSetParameters=true}
   /SetVariables(
    Emitter.Timer 2:NiagaraFloat=$System.Sys Timer 2
    Emitter.Timer 3:NiagaraFloat=$System.Sys Timer 3
    Emitter.timer 1:NiagaraFloat=$System.Sys Timer 1
    Emitter.timer 4:NiagaraFloat=$System.Sys Timer 4
   ) {isSetParameters=true}
  ]
  particleSpawnScript [
   &InitializeParticle(
    Color Mode=::NewEnumerator0
    Lifetime=0.75
    Ribbon Width=15
    Ribbon Width Mode=::NewEnumerator1
    Sprite Size Mode=::NewEnumerator3
    UsePositionOffset=false
    Write Material Random=true
   )
   &PartitionParticles(
    Normalized Index In Partition Includes One=true
    Number Of Partitions=$Emitter.Num Groups
    Particles Per Partition:NiagaraInt32=$Emitter.particles per line
    Partition Mode:ENiagara_PartitionMode=::NewEnumerator1
    Write Partition Index to Ribbon Link Order:ENiagara_AttributeSamplingApplyOutput=::NewEnumerator0
    Write Partition to Ribbon ID=::NewEnumerator0
   )
  ]
  particleUpdateScript [
   &ParticleState(Kill Particles When Lifetime Has Elapsed=false, Let Infinitely Lived Particles Die When Emitter Deactivates=true)
   &self:ScratchModule_08(
    Timer Amount:NiagaraFloat=&RandomRangeFloat(Minimum=0, Maximum=0.35)
    float:NiagaraFloat=&RandomRangeFloat(Minimum=0, Maximum=1, Evaluation Type=::NewEnumerator1)
   ) {module=ScratchModule_08}
   &self:ScratchModule_08(
    Timer Amount:NiagaraFloat=&RandomRangeFloat(Minimum=0.1, Maximum=0.35)
    float:NiagaraFloat=&RandomRangeFloat(Minimum=0, Maximum=1, Evaluation Type=::NewEnumerator1)
   ) {module=ScratchModule_08001}
   &SkeletalMeshLocation(
    Bone / Socket Index=$Particles.PartitionParticles.Partition
    Bone Sampling Mode=::NewEnumerator4
    MaskSpawnGroup=false
    Orientation Sampling=::NewEnumerator1
    Position Sampling=::NewEnumerator1
    Sampled Position Offset=&SelectVectorFromArray(
     Direct Array Index=$Particles.PartitionParticles.Partition
     Vector Selection Array=DataInterface() {propertyValues {FloatData [<5 0 0>, <-5 0 1>, <-5 7 0>, <5 -7 0>]}}
    )
    Skeletal Mesh=$System.Skeletal Mesh
    Write To Particle Position=false
   )
   &self:SLect_Shape_Mod(
    RandomSelctor:NiagaraInt32=$Particles.PartitionParticles.Partition
    Select Pos:NiagaraFloat=$Emitter.ini Pos 1
    Select Pos 2:NiagaraFloat=$Emitter.ini Pos 2
    Select Pos 3:NiagaraFloat=$Emitter.ini Pos 3
    Select Pos 4:NiagaraFloat=$Emitter.ini Pos 4
    Select Pos Y:NiagaraFloat=$Emitter.ini Pos Y 1
    Select Pos Y 2:NiagaraFloat=$Emitter.ini Pos Y 2
    Select Pos Y 3:NiagaraFloat=$Emitter.ini Pos Y 3
    Select Pos Y 4:NiagaraFloat=$Emitter.ini Pos Y 4
    Shape Select:NiagaraInt32=$Emitter.shape 1
    Shape Select 2:NiagaraInt32=$Emitter.shape 2
    Shape Select 3:NiagaraInt32=$Emitter.Shape 3
    Shape Select 4:NiagaraInt32=$Emitter.shape 4
    Timer Amount:NiagaraFloat=$Emitter.timer 1
    Timer Amount 2:NiagaraFloat=$Emitter.Timer 2
    Timer Amount 3:NiagaraFloat=$Emitter.Timer 3
    Timer Amount 4:NiagaraFloat=$Emitter.timer 4
   ) {module=SLect_Shape_Mod}
   &ShapeLocation(
    Handle Radius=18
    Large Radius=&MakeFloatFromVector(
     Vector=&VectorFromFloat(Value=&Lerp_Float(
       A=50
       B=75
       Alpha=&FloatFromCurve(
        FloatCurve=DataInterface() {propertyValues {Curve {keys [<at -1 1>,<at -0.8 0>]}}}
        CurveIndex=&DotProduct(Vector A=<0 0 1>, Vector B=$System.CamForVec)
       )
      ))
     Channel=::NewEnumerator0
    )
    MaskSpawnGroup=false
    Offset=<0 0 42>
    Offset Mode=::NewEnumerator0
    Shape Origin=$System.Lighting Start
    Shape Primitive=::NewEnumerator3
    Torus Distribution Mode:ENiagaraTorusDistributionMode=::NewEnumerator1
    U Position=&Add_Float(
     A=&self:torus_pos_float(
      float 1:NiagaraFloat=&Add_Float(A=0.25, B=0)
      float 2:NiagaraFloat=&Add_Float(A=0.5, B=0)
      float 3:NiagaraFloat=&Add_Float(A=0, B=0)
      float 4:NiagaraFloat=&Add_Float(A=0.75, B=0)
      Master Offset:NiagaraFloat=-0.125
      Partition:NiagaraInt32=$Particles.PartitionParticles.Partition
     )
     B=&Multiply_Float(A=$Particles.PartitionParticles.Normalized Index In Partition, B=0)
    )
    UseOwnerScale=true
    V Position=-0.05
   ) {module=ShapeLocation001}
   &self:ScratchModule_03(
    EndPos:NiagaraPosition=$Particles.SkeletalMeshLocation.SampledPosition
    StartPos:NiagaraPosition=&LerpPosition(A=$Particles.ShapeLocation001.ShapePosition, B=$System.Lighting Start, Alpha=0)
   ) {module=ScratchModule_03}
   &InterpolateOverTime(
    Initial Vector Value:Vector3f=<1 1 1>
    Rate Of Change:NiagaraFloat=&Divide_Float(A=1, B=0.3)
    Target Vector Value:Vector3f=&TransformVector(
     Vector=&Multiply_Vector(
      A=&SelectVectorFromArray(
       Direct Array Index=$Particles.PartitionParticles.Partition
       Vector Selection Array=DataInterface() {propertyValues {FloatData [<300 50 -75>, <-300 300 -75>, <50 -300 -75>, <-300 -50 -75>]}}
      )
      B=&SelectVectorFromArray(
       Direct Array Index=&MakeCustomIntFromBool(Bool=$System.Use Temp, True INT=$System.Temp Array Selection, False INT=$Particles.ShapeMod)
       Vector Selection Array=$System.Vector Array Top
      )
     )
     Destination Space=::Simulation
    )
    Variable Type=::NewEnumerator2
   ) {module=InterpolateOverTime001}
   &InterpolateOverTime(
    Initial Vector Value:Vector3f=<1 1 1>
    Rate Of Change:NiagaraFloat=&Divide_Float(A=1, B=0.3)
    Target Vector Value:Vector3f=&TransformVector(
     Vector=&Multiply_Vector(
      A=&SelectVectorFromArray(
       Direct Array Index=$Particles.PartitionParticles.Partition
       Vector Selection Array=DataInterface() {propertyValues {FloatData [<150 150 100>, <-150 150 100>, <150 -150 100>, <-150 -150 100>]}}
      )
      B=&SelectVectorFromArray(
       Direct Array Index=&MakeCustomIntFromBool(Bool=$System.Use Temp, True INT=$System.Temp Array Selection, False INT=$Particles.ShapeMod)
       Vector Selection Array=$System.Vector Array bottom
      )
     )
     Destination Space=::Simulation
    )
    Variable Type=::NewEnumerator2
   ) {module=InterpolateOverTime002}
   &FN_HermiteSpline(
    Direction End:Vector3f=$Particles.InterpolateOverTime002.InterpolateOverTime_V3_.Moving Average
    Direction Start:Vector3f=$Particles.InterpolateOverTime001.InterpolateOverTime_V3_.Moving Average
    Position Blend:NiagaraFloat=$Particles.PartitionParticles.Normalized Index In Partition
    Position End:Vector3f=$Particles.SkeletalMeshLocation.SampledPosition
    Position Start:Vector3f=&LerpPosition(A=$Particles.ShapeLocation001.ShapePosition, B=$System.Lighting Start, Alpha=0)
    Write Particles Position=true
   )
   /SetVariables(Particles.RibbonLength:NiagaraFloat=&DistanceBetweenPositions(
     A=&LerpPosition(A=$Particles.ShapeLocation001.ShapePosition, B=$System.Lighting Start, Alpha=0)
     B=$Particles.SkeletalMeshLocation.SampledPosition
    )) {isSetParameters=true, module=SetVariables_#16}
   /SetVariables(Particles.W:NiagaraFloat=&Frac_Float(Float=&Add_Float(
      A=$Particles.NormalizedAge
      B=&RandomRangeFloat(
       Minimum=0
       Maximum=1.2
       Randomness Mode=::NewEnumerator1
       Random Seed=&Add_Integer(A=$Particles.PartitionParticles.Partition, B=6)
       Fixed Random Seed=true
      )
     ))) {isSetParameters=true, module=SetVariables_#17}
   &InterpolateOverTime(Rate Of Change:NiagaraFloat=&Divide_Float(A=1, B=0.25), Target Float Value:NiagaraFloat=$Particles.RDM)
   &self:BendRibbon_01(
    BigNoiseMask:NiagaraFloat=&FloatFromCurve(
     FloatCurve=DataInterface() {propertyValues {Curve {keys [
         {arriveTangent=2.6981, interpMode=RCIM_Cubic, leaveTangent=2.6981, tangentMode=RCTM_User, time=0, value=0}
         {interpMode=RCIM_Cubic, tangentMode=RCTM_User, time=0.5, value=1}
         {arriveTangent=-0.3624, interpMode=RCIM_Cubic, tangentMode=RCTM_Break, time=1, value=0}
        ]}}}
     CurveIndex=$Particles.RibbonLinkOrder
    )
    Offset:NiagaraFloat=&FloatFromCurve(
     FloatCurve=DataInterface() {propertyValues {Curve {keys [
         {arriveTangent=-0.0524, interpMode=RCIM_Cubic, leaveTangent=-0.0524, tangentMode=RCTM_User, time=0, value=0}
         {interpMode=RCIM_Cubic, tangentMode=RCTM_User, time=0.5, value=1}
         {arriveTangent=-4.0426, tangentMode=RCTM_User, time=1, value=0}
        ]}}}
     CurveIndex=$Particles.RibbonLinkOrder
     Scale Curve=&FloatFromCurve(
      FloatCurve=DataInterface() {propertyValues {Curve {keys [{interpMode=RCIM_Cubic, leaveTangent=2.3336, tangentMode=RCTM_Break, time=0, value=0.2}, <at 1 1>]}}}
      CurveIndex=$Particles.W
      Scale Curve=&Multiply_Float(
       A=&RemapRange(
        Input Value To Remap=$Particles.InterpolateOverTime.InterpolateOverTime_Float.Moving Average
        Input Min=0
        Input Max=1
        Output Min=-1
        Output Max=1
       )
       B=10
      )
     )
    )
    RibbonLength:NiagaraFloat=$Particles.RibbonLength
    YBigNoise:NiagaraFloat=&Waveform(
     Global Input Position=&Multiply_Float(A=$Particles.InterpolateOverTime.InterpolateOverTime_Float.Moving Average, B=10)
     Global Phase=&Add_Float(
      A=&Multiply_Float(A=$Particles.RibbonLinkOrder, B=2)
      B=&Multiply_Float(A=$Emitter.Age, B=-2)
     )
     "Waveform [1]"=::NewEnumerator12
     "[1] Frequency"=1
     "[1] Amplitude Min/Max"=<-1 1>
     "[1] Scale"=0
    )
    ZBigNoise:NiagaraFloat=&Waveform(
     Global Input Position=&Multiply_Float(A=$Particles.InterpolateOverTime.InterpolateOverTime_Float.Moving Average, B=10)
     Global Phase=&Add_Float(
      A=&Multiply_Float(A=$Particles.RibbonLinkOrder, B=3)
      B=&Multiply_Float(A=$Emitter.Age, B=2)
     )
     "Waveform [1]"=::NewEnumerator12
     "[1] Frequency"=1
     "[1] Amplitude Min/Max"=<-1 0>
     "[1] Scale"=15
    )
   ) {module=BendRibbon_01}
   &ScaleRibbonWidth(Ribbon Width Scale=&FloatFromCurve(
     FloatCurve=DataInterface() {propertyValues {Curve {keys [
         {interpMode=RCIM_Cubic, tangentMode=RCTM_Break, time=0, value=0.6}
         {interpMode=RCIM_Cubic, tangentMode=RCTM_Break, time=0.1, value=0.4}
         {interpMode=RCIM_Cubic, tangentMode=RCTM_Break, time=0.9, value=0.4}
         <at 1 1>
        ]}}}
     CurveIndex=$Particles.PartitionParticles.Normalized Index In Partition
     Scale Curve=&FloatFromCurve(
      FloatCurve=DataInterface() {propertyValues {Curve {keys [
          {tangentMode=RCTM_User, time=0, value=1}
          {arriveTangent=-0.0405, interpMode=RCIM_Cubic, leaveTangent=-0.0405, tangentMode=RCTM_User, time=0.5, value=1}
          {arriveTangent=-4.0987, interpMode=RCIM_Cubic, leaveTangent=-4.0987, tangentMode=RCTM_User, time=1, value=0}
         ]}}}
      CurveIndex=$Particles.W
      Scale Curve=&FloatFromCurve(
       FloatCurve=DataInterface() {propertyValues {Curve {keys [
           {arriveTangent=20.043, interpMode=RCIM_Cubic, leaveTangent=20.043, tangentMode=RCTM_User, time=0, value=0}
           {arriveTangent=-0.2233, interpMode=RCIM_Cubic, leaveTangent=-0.2233, tangentMode=RCTM_User, time=0.1, value=1}
          ]}}}
       CurveIndex=$Particles.PartitionParticles.Normalized Index In Partition
      )
     )
    ))
   &ScaleRibbonWidth(Ribbon Width Scale=&RemapRange(Input Value To Remap=$Particles.RDM, Input Min=0, Input Max=1, Output Min=0.7, Output Max=2)) {module=ScaleRibbonWidth002}
   &DynamicMaterialParameters(
    Index 0 Param 1=$Particles.ScratchModule_03.Z
    Index 0 Param 2=$Particles.ScratchModule_03.Y
    Index 0 Param 3=$Particles.ScratchModule_03.Z
    Index 0 Param 4=$Particles.ScratchModule_03.W
    Index 1 Param 1=&FloatFromCurve(
     FloatCurve=DataInterface() {propertyValues {Curve {keys [{tangentMode=RCTM_User, time=0, value=0}, {tangentMode=RCTM_User, time=1, value=1}]}}}
     CurveIndex=$Particles.W
    )
    Index 1 Param 2=&FloatFromCurve(FloatCurve=DataInterface() {propertyValues {Curve {keys [<at 0 0>,<at 0.3 1>]}}}, CurveIndex=$Particles.W)
    Index 1 Param 3=1
    Index 1 Param 4=&Add_Float(
     A=&FloatFromCurve(
      FloatCurve=DataInterface() {propertyValues {Curve {keys [{tangentMode=RCTM_User, time=0, value=-1}, {tangentMode=RCTM_User, time=1, value=0.7}]}}}
      CurveIndex=$Particles.W
     )
     B=&FloatFromCurve(
      FloatCurve=DataInterface() {propertyValues {Curve {keys [
          {interpMode=RCIM_Cubic, tangentMode=RCTM_Break, time=0, value=-1}
          {interpMode=RCIM_Cubic, tangentMode=RCTM_Break, time=0.1, value=0}
          {interpMode=RCIM_Cubic, tangentMode=RCTM_Break, time=0.9, value=0}
          <at 1 -1>
         ]}}}
      CurveIndex=$Particles.RibbonLinkOrder
      Scale Curve=0.3
     )
    )
    Module.Index0Param0Write=true
    Module.Index0Param1Write=true
    Module.Index0Param2Write=true
    Module.Index0Param3Write=true
    Module.Index1Param0Write=true
    Module.Index1Param1Write=true
    Module.Index1Param2Write=true
    Module.Index1Param3Write=true
    Write Parameter Index 1=true
   ) {enabled=false}
   &DynamicMaterialParameters(
    Index 0 Param 1=$Particles.ScratchModule_03.Z
    Index 0 Param 2=$Particles.ScratchModule_03.Y
    Index 0 Param 3=$Particles.ScratchModule_03.Z
    Index 0 Param 4=$Particles.ScratchModule_03.W
    Index 1 Param 1=&FloatFromCurve(
     FloatCurve=DataInterface() {propertyValues {Curve {keys [{tangentMode=RCTM_User, time=0, value=0}, {tangentMode=RCTM_User, time=1, value=1}]}}}
     CurveIndex=$Particles.W
    )
    Index 1 Param 2=&FloatFromCurve(
     FloatCurve=DataInterface() {propertyValues {Curve {keys [
         {arriveTangent=9.0871, interpMode=RCIM_Cubic, time=0, value=0}
         {arriveTangent=11.4679, interpMode=RCIM_Cubic, leaveTangent=11.4679, tangentMode=RCTM_User, time=0.3, value=1}
        ]}}}
     CurveIndex=$Particles.W
    )
    Index 1 Param 3=&FloatFromCurve(
     FloatCurve=DataInterface() {propertyValues {Curve {keys [<at 0 0>,<at 0.5 1>,<at 0.6 1>,<at 1 0>]}}}
     CurveIndex=$Particles.W
     Scale Curve=1.25
    )
    Index 1 Param 4=&Add_Float(
     A=&FloatFromCurve(
      FloatCurve=DataInterface() {propertyValues {Curve {keys [{tangentMode=RCTM_User, time=0, value=-1}, {tangentMode=RCTM_User, time=1, value=0.7}]}}}
      CurveIndex=$Particles.W
     )
     B=&FloatFromCurve(
      FloatCurve=DataInterface() {propertyValues {Curve {keys [
          {interpMode=RCIM_Cubic, tangentMode=RCTM_Break, time=0, value=-1}
          {interpMode=RCIM_Cubic, tangentMode=RCTM_Break, time=0.1, value=0}
          {interpMode=RCIM_Cubic, tangentMode=RCTM_Break, time=0.9, value=0}
          <at 1 -1>
         ]}}}
      CurveIndex=$Particles.RibbonLinkOrder
      Scale Curve=0.3
     )
    )
    Module.Index0Param0Write=true
    Module.Index0Param1Write=true
    Module.Index0Param2Write=true
    Module.Index0Param3Write=true
    Module.Index1Param0Write=true
    Module.Index1Param1Write=true
    Module.Index1Param2Write=true
    Module.Index1Param3Write=true
    Write Parameter Index 1=true
   ) {module=DynamicMaterialParameters001}
   /SetVariables(Particles.RibbonW2:NiagaraFloat=&Multiply_Float(A=$Particles.RibbonWidth, B=2.5)) {isSetParameters=true, module=SetVariables_#18}
   /SetVariables(
    Particles.High Width:NiagaraFloat=&MakeCustomFloatFromBool(Bool=&SetBoolByFloatComparison(A=$System.CloudDot, B=0.5), True Float=$Particles.RibbonWidth, False Float=0)
    Particles.low Width:NiagaraFloat=&MakeCustomFloatFromBool(Bool=&SetBoolByFloatComparison(A=$System.CloudDot, B=0.5), True Float=1, False Float=$Particles.RibbonWidth)
   ) {isSetParameters=true}
   /SetVariables(
    Particles.High Width 2:NiagaraFloat=&MakeCustomFloatFromBool(Bool=&SetBoolByFloatComparison(A=$System.CloudDot, B=0.5), True Float=$Particles.RibbonW2, False Float=0)
    Particles.low Width 2:NiagaraFloat=&MakeCustomFloatFromBool(Bool=&SetBoolByFloatComparison(A=$System.CloudDot, B=0.5), True Float=1, False Float=$Particles.RibbonW2)
   ) {isSetParameters=true}
   /SetVariables(
    Particles.Glow Cam Offset=100
    Particles.Sprite Color:LinearColor=&MultiplyLinearColor(
     Linear Color=<rgba 0.15 0.6 1.5 0.1>
     Scale Factor=&MakeFloat4FromVectorAndFloat(
      "Vector (RGB)"=&VectorFromFloat(Value=&FNTimeOfDayScalar(Night Scale=1, Morning Scale=1, Evening Scale=1))
      "Float (Alpha)"=&FloatFromCurve(
       FloatCurve=DataInterface() {propertyValues {Curve {keys [<at -0.4 0.15>,<at -0.2 1.25>]}}}
       CurveIndex=&DotProduct(Vector A=$System.CamForVec, Vector B=<0 0 1>)
       Scale Curve=&FloatFromCurve(
        FloatCurve=DataInterface() {propertyValues {Curve {keys [{tangentMode=RCTM_User, time=0.655, value=0}, {tangentMode=RCTM_User, time=0.855, value=1}]}}}
        CurveIndex=$Engine.System.Age
        Scale Curve=&FloatFromCurve(
         FloatCurve=DataInterface() {propertyValues {Curve {keys [
             {arriveTangent=9.0725, interpMode=RCIM_Cubic, leaveTangent=9.0725, tangentMode=RCTM_User, time=0, value=0}
             {arriveTangent=-2.5249, interpMode=RCIM_Cubic, leaveTangent=-2.5249, tangentMode=RCTM_User, time=0.05, value=1}
             {arriveTangent=-0.0405, interpMode=RCIM_Cubic, leaveTangent=-0.0405, tangentMode=RCTM_User, time=1, value=0}
            ]}}}
         CurveIndex=$Particles.W
         Scale Curve=2
        )
       )
      )
     )
    )
    "Particles.SpriteSizeGlow ":Vector2f=&Vector2DFromFloat(Value=&FloatFromCurve(
      FloatCurve=DataInterface() {propertyValues {Curve {keys [<at 0 1 RCIM_Constant>,<at 0.01 0 RCIM_Constant>,<at 0.99 1 RCIM_Constant>]}}}
      CurveIndex=$Particles.PartitionParticles.Normalized Index In Partition
      Scale Curve=&Lerp_Float(A=40, B=40, Alpha=0.5)
     ))
    Particles.dynam:Vector4f=<2 1 0 0>
   ) {isSetParameters=true}
   /SetVariables(
    Particles.DEcalSprite Color:LinearColor=&MultiplyLinearColor(
     Linear Color=&MultiplyLinearColor(
      Linear Color=<rgba 0.3 1.2 3 1>
      Scale Factor=&MakeFloat4FromVectorAndFloat(
       "Vector (RGB)"=&VectorFromFloat(Value=&FNTimeOfDayScalar(Night Scale=0.4, Morning Scale=1, Evening Scale=1))
       "Float (Alpha)"=&FloatFromCurve(
        FloatCurve=DataInterface() {propertyValues {Curve {keys [
            {arriveTangent=9.0725, interpMode=RCIM_Cubic, leaveTangent=9.0725, tangentMode=RCTM_User, time=0, value=0}
            {arriveTangent=-2.5249, interpMode=RCIM_Cubic, leaveTangent=-2.5249, tangentMode=RCTM_User, time=0.05, value=1}
            {arriveTangent=-0.0405, interpMode=RCIM_Cubic, leaveTangent=-0.0405, tangentMode=RCTM_User, time=1, value=0}
           ]}}}
        CurveIndex=$Particles.W
        Scale Curve=2
       )
      )
     )
     Scale Factor=&MakeFloat4FromVectorAndFloat(
      "Vector (RGB)"=&VectorFromFloat(Value=&FloatFromCurve(
        FloatCurve=DataInterface() {propertyValues {Curve {keys [<at -0.4 0.15>,<at -0.2 1.25>]}}}
        CurveIndex=&DotProduct(Vector A=$System.CamForVec, Vector B=<0 0 1>)
        Scale Curve=&FloatFromCurve(
         FloatCurve=DataInterface() {propertyValues {Curve {keys [{tangentMode=RCTM_User, time=0.655, value=0}, {tangentMode=RCTM_User, time=0.855, value=1}]}}}
         CurveIndex=$Engine.System.Age
        )
       ))
      "Float (Alpha)"=1
     )
    )
    "Particles.Decal SpriteSizeGlow ":Vector2f=&Vector2DFromFloat(Value=&FloatFromCurve(
      FloatCurve=DataInterface() {propertyValues {Curve {keys [<at 0 1 RCIM_Constant>,<at 0.01 0>]}}}
      CurveIndex=$Particles.PartitionParticles.Normalized Index In Partition
      Scale Curve=&Lerp_Float(
       A=35
       B=50
       Alpha=&FloatFromCurve(
        FloatCurve=DataInterface() {propertyValues {Curve {keys [<at -0.5 0.15>,<at -0.1 1.25>]}}}
        CurveIndex=&DotProduct(Vector A=$System.CamForVec, Vector B=<0 0 1>)
        Scale Curve=&FloatFromCurve(
         FloatCurve=DataInterface() {propertyValues {Curve {keys [{tangentMode=RCTM_User, time=0.655, value=0}, {tangentMode=RCTM_User, time=0.855, value=1}]}}}
         CurveIndex=$Engine.System.Age
        )
       )
      )
     ))
   ) {isSetParameters=true}
   /SetVariables(Particles.Front decal col:LinearColor=&MultiplyLinearColor(
     Linear Color=$Particles.DEcalSprite Color
     Scale Factor=&MakeFloat4FromVectorAndFloat("Vector (RGB)"=&VectorFromFloat(Value=1), "Float (Alpha)"=0.05)
    )) {isSetParameters=true}
  ]
  rendererClasses [&NiagaraRibbonRendererProperties, &NiagaraSpriteRendererProperties]
  renderers [
   {rendererClass=&NiagaraRibbonRendererProperties}
   {rendererClass=&NiagaraRibbonRendererProperties, rendererIndex=1}
   {rendererClass=&NiagaraRibbonRendererProperties, rendererIndex=2}
   {rendererClass=&NiagaraRibbonRendererProperties, rendererIndex=3}
   {rendererClass=&NiagaraSpriteRendererProperties, rendererIndex=4}
   {rendererClass=&NiagaraSpriteRendererProperties, rendererIndex=5}
   {rendererClass=&NiagaraSpriteRendererProperties, rendererIndex=6}
  ]
 }
 Plates {
  bEnabled=false
  emitterUpdateScript [
   &EmitterState(Life Cycle Mode=::NewEnumerator1, Loop Behavior=::NewEnumerator1, Loop Delay=0.5, Loop Duration Mode=::NewEnumerator1, UseLoopDelay=true)
   &SpawnBurst_Instantaneous(Spawn Count=5, Use Spawn Probability=false, UseLoopCountLimit=true)
  ]
  particleSpawnScript [
   &InitializeParticle(
    Color=&FNTimeOfDayColor(
     Night Color=&SelectLinearColorFromArray(
      Array Sampling Mode=::NewEnumerator1
      Direct Array Index=&ReturnParticleIDAttributes()
      Color Selection Array=DataInterface() {propertyValues {ColorData [<rgba 0.875 1.75 1.4583 0.64>, <rgba 0 1.6667 5 0.64>, <rgba 3 3 3 0.64>, <rgba 2 2 2 0.64>, <rgba 2 2 2 0.64>]}}
      Random Seed=?
     )
     Morning Color=&SelectLinearColorFromArray(
      Array Sampling Mode=::NewEnumerator1
      Direct Array Index=&ReturnParticleIDAttributes()
      Color Selection Array=DataInterface() {propertyValues {ColorData [<rgba 1.75 1.75 1.75 0.64>, <rgba 0 1.6667 5 0.64>, <rgba 3 3 3 0.64>, <rgba 2 2 2 0.64>, <rgba 2 2 2 0.64>]}}
      Random Seed=?
     )
     Day Color=&SelectLinearColorFromArray(
      Array Sampling Mode=::NewEnumerator1
      Direct Array Index=&ReturnParticleIDAttributes()
      Color Selection Array=DataInterface() {propertyValues {ColorData [<rgba 1 1 1 0.64>, <rgba 0 1.6667 5 0.64>, <rgba 1 1 1 0.64>, <rgba 2 2 2 0.64>, <rgba 2 2 2 0.64>]}}
      Random Seed=?
     )
     Evening Color=&SelectLinearColorFromArray(
      Array Sampling Mode=::NewEnumerator1
      Direct Array Index=&ReturnParticleIDAttributes()
      Color Selection Array=DataInterface() {propertyValues {ColorData [<rgba 1.75 1.75 1.75 0.64>, <rgba 0 1.6667 5 0.64>, <rgba 3 3 3 0.64>, <rgba 2 2 2 0.64>, <rgba 2 2 2 0.64>]}}
      Random Seed=?
     )
    )
    Lifetime=5
    Mesh Scale=&SelectVectorFromArray(
     Direct Array Index=&ReturnParticleIDAttributes()
     Vector Selection Array=DataInterface() {propertyValues {FloatData [<35 35 -10>, <10 10 -20>, <27 27 5>, <15 15 -5>, <10 10 20>]}}
    )
    Mesh Scale Mode=::NewEnumerator4
    Position=$System.Lighting Start
    Position Mode=::NewEnumerator1
    UsePositionOffset=false
    Write Material Random=true
   )
   &InitialMeshOrientation(
    Mesh Orientation Mode=::NewEnumerator7
    Rotation=&SelectVectorFromArray(
     Direct Array Index=&ReturnParticleIDAttributes()
     Vector Selection Array=DataInterface() {propertyValues {FloatData [<0 0.5 0>, <0 0 0>, <0 0 0>, <0 0 0>, <0 0.5 0>]}}
    )
    Use Rotation Vector=true
   )
   &KillParticles(Kill Particles=&Boolean_AndOperation(
     Bool1=&IsPlatformSetActive(New Platform Set=DataInterface() {propertyValues {Platforms {qualityLevelMask=-29}}})
     Bool2=&SetBoolByIntComparison(A=&ReturnParticleIDAttributes(), B=3, Comparison Type=::NewEnumerator4)
    ))
   &KillParticles(Kill Particles=&Boolean_AndOperation(
     Bool1=&IsPlatformSetActive(New Platform Set=DataInterface() {propertyValues {Platforms {qualityLevelMask=-31}}})
     Bool2=&SetBoolByIntComparison(A=&ReturnParticleIDAttributes(), B=1, Comparison Type=::NewEnumerator4)
    )) {module=KillParticles002}
   &KillParticles(Kill Particles=&Boolean_AndOperation(
     Bool1=&IsPlatformSetActive(New Platform Set=DataInterface() {propertyValues {Platforms {qualityLevelMask=-31}}})
     Bool2=&SetBoolByIntComparison(A=&ReturnParticleIDAttributes(), B=4, Comparison Type=::NewEnumerator4)
    )) {module=KillParticles001}
   &KillParticles(Kill Particles=&Boolean_AndOperation(
     Bool1=&IsPlatformSetActive(New Platform Set=DataInterface() {propertyValues {Platforms {qualityLevelMask=-29}}})
     Bool2=&SetBoolByIntComparison(A=&ReturnParticleIDAttributes(), B=1, Comparison Type=::NewEnumerator4)
    )) {module=KillParticles003}
   &KillParticles(Kill Particles=&Boolean_AndOperation(
     Bool1=&IsPlatformSetActive(New Platform Set=DataInterface() {propertyValues {Platforms {qualityLevelMask=-29}}})
     Bool2=&SetBoolByIntComparison(A=&ReturnParticleIDAttributes(), B=4, Comparison Type=::NewEnumerator4)
    )) {module=KillParticles004}
   &Color(Color=&MultiplyLinearColor(
     Linear Color=$Particles.Color
     Scale Factor=/Niagara/DynamicInputs/TypeConversions/MakeCustomVector4FromBool(
      Bool:NiagaraBool=&IsPlatformSetActive(New Platform Set=DataInterface() {propertyValues {Platforms {qualityLevelMask=-4}}})
      True Vector4:Vector4f=<1 1 1 1>
      False Vector4:Vector4f=&MakeFloat4FromVectorAndFloat("Vector (RGB)"=&VectorFromFloat(Value=1), "Float (Alpha)"=1)
     )
    ))
  ]
  particleUpdateScript [
   &ParticleState(Kill Particles When Lifetime Has Elapsed=false, Let Infinitely Lived Particles Die When Emitter Deactivates=true)
   &ShapeLocation(
    MaskSpawnGroup=false
    Offset=&SelectVectorFromArray(
     Direct Array Index=&ReturnParticleIDAttributes()
     Vector Selection Array=DataInterface() {propertyValues {FloatData [<0 0 50>, <0 0 0>, <0 0 38>, <0 0 15>, <0 0 1>]}}
    )
    Offset Mode=::NewEnumerator0
    Override Seed=false
    Shape Origin=$System.Lighting Start
    Sphere Radius=0
    UseOwnerScale=true
   )
   &ScaleColor(
    Scale Alpha=&MakeCustomFloatFromBool(
     Bool=&SetBoolByIntComparison(A=&ReturnParticleIDAttributes(), B=0, Comparison Type=::NewEnumerator4)
     True Float=1
     False Float=&FloatFromCurve(
      FloatCurve=DataInterface() {propertyValues {Curve {keys [<at -0.4 0>, {tangentMode=RCTM_User, time=-0.36, value=1}]}}}
      CurveIndex=&DotProduct(Vector A=$System.CamForVec, Vector B=<0 0 1>)
     )
    )
    ScaleA=true
    ScaleRGB=true
   ) {module=ScaleColor002}
   &ScaleColor(
    Scale Alpha=&FloatFromCurve(FloatCurve=DataInterface() {propertyValues {Curve {keys [{tangentMode=RCTM_User, time=0, value=0.5}, {tangentMode=RCTM_User, time=0.1, value=1}]}}})
    Scale RGB=&MakeVectorFromLinearColorRGB(Color=<rgba 0.1301 0.3712 0.8148 1>)
    ScaleA=true
    ScaleRGB=true
   )
   /SetVariables(Particles.VisibilityTag=&SelectIntFromArray(
     Direct Array Index=&ReturnParticleIDAttributes()
     Int Selection Array=DataInterface() {propertyValues {IntData [2,0,1,0,0]}}
    )) {isSetParameters=true}
   &ScaleColor(
    Scale Alpha=&FloatFromCurve(
     FloatCurve=DataInterface() {propertyValues {Curve {keys [
         {
          arriveTangent=4.0469
          arriveTangentWeight=0.9089
          interpMode=RCIM_Cubic
          leaveTangent=4.0469
          leaveTangentWeight=0.9089
          tangentMode=RCTM_User
          tangentWeightMode=RCTWM_WeightedBoth
          time=0
          value=0
         }
         {
          arriveTangentWeight=0.6453
          interpMode=RCIM_Cubic
          leaveTangentWeight=0.6453
          tangentMode=RCTM_User
          tangentWeightMode=RCTWM_WeightedBoth
          time=1
          value=1
         }
        ]}}}
     CurveIndex=$System.ViewDot
    )
    ScaleA=true
    ScaleRGB=true
   ) {enabled=false, module=ScaleColor001}
   &ScaleColor(
    Scale Alpha=&FloatFromCurve(
     FloatCurve=DataInterface() {propertyValues {Curve {keys [
         {
          arriveTangentWeight=0.0906
          interpMode=RCIM_Cubic
          leaveTangentWeight=0.0906
          tangentMode=RCTM_User
          tangentWeightMode=RCTWM_WeightedBoth
          time=-0.85
          value=0
         }
         {
          arriveTangent=7.032
          arriveTangentWeight=0.7813
          interpMode=RCIM_Cubic
          leaveTangent=7.032
          tangentMode=RCTM_User
          tangentWeightMode=RCTWM_WeightedBoth
          time=-0.67
          value=1
         }
        ]}}}
     CurveIndex=&DotProduct(Vector A=$System.CamForVec, Vector B=<0 0 1>)
    )
    ScaleA=true
    ScaleRGB=true
   ) {enabled=false, module=ScaleColor003}
   /SetVariables(Particles.tow low color:LinearColor=&MultiplyLinearColor(
     Linear Color=$Particles.Color
     Scale Factor=&MakeFloat4FromVectorAndFloat(
      "Vector (RGB)"=<1 1 1>
      "Float (Alpha)"=&FloatFromCurve(
       FloatCurve=DataInterface() {propertyValues {Curve {keys [<at -0.9 0>,<at -0.7 5>]}}}
       CurveIndex=&DotProduct(Vector A=<0 0 1>, Vector B=$System.CamForVec)
      )
     )
    )) {isSetParameters=true}
   /SetVariables(Particles.Top Low scale:Vector3f=&MakeCustomVectorFromBool(
     Bool=&SetBoolByFloatComparison(
      A=&MakeFloatFromLinearColor(LinearColor=$Particles.tow low color, Channel:ENiagaraLinearColor_Channels=::NewEnumerator3)
      B=0
      Comparison Type=::NewEnumerator7
     )
     True Vector=<0 0 1>
     False Vector=&Multiply_Vector(A=$Particles.Scale, B=&VectorFromFloat(Value=0.75))
    )) {isSetParameters=true}
   /SetVariables(Particles.JiraColor2:LinearColor=&MultiplyLinearColor(
     Linear Color=$Particles.Color
     Scale Factor=&MakeFloat4FromVectorAndFloat(
      "Vector (RGB)"=<1 1 1>
      "Float (Alpha)"=&Lerp_Float(
       A=1
       B=0
       Alpha=&FloatFromCurve(
        FloatCurve=DataInterface() {propertyValues {Curve {keys [
            {arriveTangent=-4.4607, interpMode=RCIM_Cubic, leaveTangent=-4.4607, tangentMode=RCTM_User, time=-1, value=1}
            {arriveTangent=-1.4134, interpMode=RCIM_Cubic, leaveTangent=-1.4134, tangentMode=RCTM_User, time=-0.57, value=0}
           ]}}}
        CurveIndex=&DotProduct(Vector A=$System.CamForVec, Vector B=<0 0 1>)
       )
      )
     )
    )) {isSetParameters=true}
   /SetVariables(Particles.JiraColor3:LinearColor=&MultiplyLinearColor(
     Linear Color=$Particles.tow low color
     Scale Factor=&MakeFloat4FromVectorAndFloat(
      "Vector (RGB)"=<1 1 1>
      "Float (Alpha)"=&Lerp_Float(
       A=0.25
       B=0
       Alpha=&FloatFromCurve(
        FloatCurve=DataInterface() {propertyValues {Curve {keys [
            {arriveTangent=-4.4607, interpMode=RCIM_Cubic, leaveTangent=-4.4607, tangentMode=RCTM_User, time=-0.7596, value=1}
            {arriveTangent=-1.4134, interpMode=RCIM_Cubic, leaveTangent=-1.4134, tangentMode=RCTM_User, time=-0.3296, value=0}
           ]}}}
        CurveIndex=&DotProduct(Vector A=$System.CamForVec, Vector B=<0 0 1>)
       )
      )
     )
    )) {isSetParameters=true}
  ]
  rendererClasses [&NiagaraMeshRendererProperties]
  renderers [
   {rendererClass=&NiagaraMeshRendererProperties, rendererIndex=0}
   {rendererClass=&NiagaraMeshRendererProperties, rendererIndex=1}
   {rendererClass=&NiagaraMeshRendererProperties, rendererIndex=2}
   {rendererClass=&NiagaraMeshRendererProperties, rendererIndex=3}
   {rendererClass=&NiagaraMeshRendererProperties, rendererIndex=4}
   {rendererClass=&NiagaraMeshRendererProperties, rendererIndex=5}
  ]
 }
 Rotating_Air_Bottom {
  emitterUpdateScript [
   &EmitterState()
   &SpawnRate(SpawnRate=30, Use Spawn Probability=false)
  ]
  particleSpawnScript [
   /Game/Effects/Niagara/Modules/Orbit/FN_Orbit_Init_Variables_v2()
   /Niagara/Modules/Spawn/Initialization/InitializeParticle(
    Color=<rgba 0.0546 0.0662 0.0677 0.35>
    Lifetime=&UniformRangedFloat(Minimum=1, Maximum=1.5)
    Mesh Scale=&UniformRangedVector(Minimum=<5 5 8>, Maximum=<8 8 11>)
    Module.Write Color=true
    Module.Write Lifetime=true
    Module.Write Mass=false
    Module.Write Material Random=true
    Module.Write Position=false
    Module.Write Scale=true
    Module.Write SpriteRotation=false
    Module.Write SpriteSize=false
    Position=&SimulationPosition()
   ) {enabled=false}
   /SetVariables(Particles.Scale both:Vector3f=<0.75 0.75 1.2>) {isSetParameters=true}
   /Niagara/Modules/Spawn/Initialization/V2/InitializeParticle(
    Color=<rgba 0.0799 0.2278 0.5 0.65>
    Lifetime=&UniformRangedFloat(Minimum=1, Maximum=1.5)
    Mesh Scale=&Multiply_Vector(A=<1 1 0.75>, B=$Particles.Scale both)
    Mesh Scale Mode=::NewEnumerator4
    Sprite Size Mode=::NewEnumerator4
    UsePositionOffset=false
    Write Material Random=true
   ) {module=InitializeParticle002}
   &DynamicMaterialParameters(
    Index 0 Param 1=&UniformRangedFloat(Minimum=0.1, Maximum=0.25)
    Index 0 Param 2=1
    Module.Param0WriteEnabled=true
    Module.Param1WriteEnabled=true
    Module.Param2WriteEnabled=true
    Module.Param3WriteEnabled=true
   )
   /Game/Effects/Niagara/Modules/Mesh/FN_MeshRenderer_v3/FN_MeshRenderer_v2(
    Color:LinearColor=&LerpLinearColors(
     StartColor=<rgba 0.2396 0.6834 1.5 1>
     EndColor=<rgba 0.1301 0.3712 0.8148 1>
     LerpFactor=&FloatFromCurve(
      FloatCurve=DataInterface() {propertyValues {Curve {keys [
          {
           arriveTangentWeight=0.7106
           interpMode=RCIM_Cubic
           leaveTangentWeight=0.7106
           tangentMode=RCTM_User
           tangentWeightMode=RCTWM_WeightedBoth
           time=0
           value=0
          }
          {
           arriveTangent=5.2731
           arriveTangentWeight=0.5583
           interpMode=RCIM_Cubic
           leaveTangent=5.2731
           leaveTangentWeight=0.5583
           tangentMode=RCTM_User
           tangentWeightMode=RCTWM_WeightedBoth
           time=1
           value=1
          }
         ]}}}
      CurveIndex=&RandomRangeFloat(Minimum=0, Maximum=1)
     )
    )
    Create COlor Attribute=true
    Create Dynamic Param Attribute=true
    Create Scale Attribute=true
    DynamicMaterial:Vector4f=&MakeVector4(
     X=&UniformRangedFloat(Minimum=0, Maximum=1)
     Y=&UniformRangedFloat(Minimum=5, Maximum=6)
     Z=0
     W=2
    )
    Scale:Vector3f=&Multiply_Vector(A=&UniformRangedVector(Minimum=<1 1 0.25>, Maximum=<1 1 0.25>), B=$Particles.Scale both)
   )
   &InitialMeshOrientation(
    Mesh Orientation Mode=::NewEnumerator7
    Rotation=&RandomRangeVector(Minimum=<0 0 0>, Maximum=<0 0 1>)
    Use Rotation Vector=true
   )
  ]
  particleUpdateScript [
   &ParticleState()
   &ShapeLocation(
    Cylinder Height=25
    Cylinder Radius=0
    MaskSpawnGroup=false
    Offset=&Add_Vector(
     A=<0 0 25>
     B=&Multiply_Vector(
      A=&ReturnMeshOrientationAxis(Mesh Orientation Axis=::XAxis, Mesh Orientation Quat=$Particles.MeshOrientation)
      B=&VectorFromFloat(Value=&Lerp_Float(
        A=1
        B=20
        Alpha=&FloatFromCurve(
         FloatCurve=DataInterface() {propertyValues {Curve {keys [<at -1 1>,<at -0.8 0>]}}}
         CurveIndex=&DotProduct(Vector A=<0 0 1>, Vector B=$System.CamForVec)
        )
       ))
     )
    )
    Offset Mode=::NewEnumerator0
    Override Seed=false
    Shape Origin=$System.Lighting Start
    Shape Primitive=::NewEnumerator1
    Surface Only=false
    UseOwnerScale=true
   )
   &ScaleMeshSize(Scale Factor=&VectorFromCurve(
     VectorCurve=DataInterface() {propertyValues {
       XCurve {keys [{tangentMode=RCTM_User, time=0, value=0.5}, {tangentMode=RCTM_User, time=1, value=1}]}
       YCurve {keys [{tangentMode=RCTM_User, time=0, value=0.5}, {tangentMode=RCTM_User, time=1, value=1}]}
       ZCurve {keys [<at 0 1 RCIM_Cubic>,<at 1 1>]}
      }}
     Scale Curve=&VectorFromCurve(
      VectorCurve=DataInterface() {propertyValues {XCurve {keys [<at 0 0.5>,<at 1 1>]}, YCurve {keys [<at 0 0.5>,<at 1 1>]}, ZCurve {keys [<at 0 1>,<at 1 1>]}}}
      CurveIndex=&Divide_Float(A=&MakeFloatFromVector(Vector=$Particles.ShapeLocation.ShapeVector, Channel=::NewEnumerator2), B=25)
     )
    ))
   &ScaleColor(
    Scale Alpha=&FloatFromCurve(FloatCurve=DataInterface() {propertyValues {Curve {keys [
         <at 0 0 RCIM_Cubic>
         <at 0.3 1 RCIM_Cubic>
         {arriveTangent=-4.5166, interpMode=RCIM_Cubic, leaveTangent=-4.5166, tangentMode=RCTM_User, time=1, value=0}
        ]}}})
    ScaleA=true
    ScaleRGB=true
   )
   &FN_MeshRenderer_v2(
    Color:LinearColor=&MultiplyLinearColor(
     Linear Color=$Particles.Initial.FN_MeshRenderer_v2.Color
     Scale Factor=&MakeFloat4FromVectorAndFloat(
      "Vector (RGB)"=&VectorFromFloat(Value=1)
      "Float (Alpha)"=&FloatFromCurve(
       FloatCurve=DataInterface() {propertyValues {Curve {keys [
           <at 0 0 RCIM_Cubic>
           {arriveTangent=1.3333, leaveTangent=1.3333, time=0.15, value=1}
           <at 0.85 1 RCIM_Cubic>
           {arriveTangent=-4.5166, interpMode=RCIM_Cubic, leaveTangent=-4.5166, tangentMode=RCTM_User, time=1, value=0}
          ]}}}
       Scale Curve=&Multiply_Float(
        A=&FloatFromCurve(
         FloatCurve=DataInterface() {propertyValues {Curve {keys [
             {
              arriveTangentWeight=0.0906
              interpMode=RCIM_Cubic
              leaveTangentWeight=0.0906
              tangentMode=RCTM_User
              tangentWeightMode=RCTWM_WeightedBoth
              time=-0.85
              value=-0.01
             }
             {
              arriveTangent=7.032
              arriveTangentWeight=0.7813
              interpMode=RCIM_Cubic
              leaveTangent=7.032
              tangentMode=RCTM_User
              tangentWeightMode=RCTWM_WeightedBoth
              time=-0.67
              value=1
             }
            ]}}}
         CurveIndex=&DotProduct(Vector A=$System.CamForVec, Vector B=<0 0 1>)
        )
        B=&MakeCustomFloatFromBool(
         Bool=&IsPlatformSetActive(New Platform Set=DataInterface() {propertyValues {Platforms {qualityLevelMask=-31}}})
         True Float=5
         False Float=&MakeCustomFloatFromBool(
          Bool=&IsPlatformSetActive(New Platform Set=DataInterface() {propertyValues {Platforms {qualityLevelMask=-30}}})
          True Float=3
          False Float=&MakeCustomFloatFromBool(
           Bool=&IsPlatformSetActive(New Platform Set=DataInterface() {propertyValues {Platforms {qualityLevelMask=-28}}})
           True Float=1
           False Float=1
          )
         )
        )
       )
      )
     )
    )
    Create COlor Attribute=true
    Create Position Attribute=true
    Create Scale Attribute=true
    Scale:Vector3f=&Multiply_Vector(
     A=$Particles.Initial.FN_MeshRenderer_v2.Scale
     B=&VectorFromCurve(
      VectorCurve=DataInterface() {propertyValues {
        XCurve {keys [{tangentMode=RCTM_User, time=0, value=0.5}, {tangentMode=RCTM_User, time=1, value=1}]}
        YCurve {keys [{tangentMode=RCTM_User, time=0, value=0.5}, {tangentMode=RCTM_User, time=1, value=1}]}
        ZCurve {keys [<at 0 1 RCIM_Cubic>,<at 1 1>]}
       }}
      Scale Curve=&VectorFromCurve(
       VectorCurve=DataInterface() {propertyValues {XCurve {keys [<at 0 0.5>,<at 1 1>]}, YCurve {keys [<at 0 0.5>,<at 1 1>]}, ZCurve {keys [<at 0 1>,<at 1 1>]}}}
       CurveIndex=&Divide_Float(A=&MakeFloatFromVector(Vector=$Particles.ShapeLocation.ShapeVector, Channel=::NewEnumerator2), B=25)
      )
     )
    )
   ) {module=FN_MeshRenderer_v2001}
   &UpdateMeshOrientation(Rotation Vector=<0 0 1>)
   /SetVariables(Particles.TransCloudQuat:Quat4f=$Particles.Initial.MeshOrientation) {isSetParameters=true, module=SetVariables_#19}
   /SetVariables(Particles.VisibilityTag=&MakeCustomIntFromBool(Bool=&SetBoolByFloatComparison(A=$System.CloudDot, B=0.5), True INT=1, False INT=0)) {isSetParameters=true}
   &ScaleColor(Scale Alpha=$System.ViewDot, ScaleA=true, ScaleRGB=true) {module=ScaleColor001}
   &ScaleMeshSize(Scale Factor=&VectorFromFloat(Value=&MakeCustomFloatFromBool(
      Bool=&SetBoolByFloatComparison(
       A=&MakeFloatFromLinearColor(LinearColor=$Particles.FN_MeshRenderer_v2001.Color, Channel:ENiagaraLinearColor_Channels=::NewEnumerator3)
       B=0
      )
      True Float=1
      False Float=0
     ))) {module=ScaleMeshSize001}
   &ScaleColor(
    Scale Alpha=&MakeCustomFloatFromBool(
     Bool=&IsPlatformSetActive(New Platform Set=DataInterface() {propertyValues {Platforms {qualityLevelMask=-31}}})
     True Float=1.1
     False Float=&MakeCustomFloatFromBool(
      Bool=&IsPlatformSetActive(New Platform Set=DataInterface() {propertyValues {Platforms {qualityLevelMask=-30}}})
      True Float=1.1
      False Float=&MakeCustomFloatFromBool(
       Bool=&IsPlatformSetActive(New Platform Set=DataInterface() {propertyValues {Platforms {qualityLevelMask=-28}}})
       True Float=1
       False Float=1
      )
     )
    )
    ScaleA=true
    ScaleRGB=true
   ) {module=ScaleColor002}
   /SetVariables(Particles.new cloud scale:Vector3f=&Multiply_Vector(
     A=$Particles.FN_MeshRenderer_v2001.Scale
     B=&MakeCustomVectorFromBool(
      Bool=&SetBoolByFloatComparison(
       A=&MakeFloatFromLinearColor(LinearColor=$Particles.FN_MeshRenderer_v2001.Color, Channel:ENiagaraLinearColor_Channels=::NewEnumerator3)
       B=0
      )
      True Vector=&VectorFromFloat(Value=1)
      False Vector=&VectorFromFloat(Value=0)
     )
    )) {isSetParameters=true}
   /SetVariables(Particles.JiraColor:LinearColor=&MultiplyLinearColor(
     Linear Color=$Particles.FN_MeshRenderer_v2001.Color
     Scale Factor=&MakeFloat4FromVectorAndFloat(
      "Vector (RGB)"=<1 1 1>
      "Float (Alpha)"=/Niagara/DynamicInputs/Math/LerpMultipleFloats(
       Float Count:ENiagara_MultipleLerpCount=::NewEnumerator0
       1:NiagaraFloat=1
       2:NiagaraFloat=&FloatFromCurve(
        FloatCurve=DataInterface() {propertyValues {Curve {keys [<at -0.7 0 RCIM_Cubic>, {arriveTangent=2.9293, interpMode=RCIM_Cubic, leaveTangent=2.9293, tangentMode=RCTM_User, time=-0.0237, value=1}]}}}
        CurveIndex=&DotProduct(
         Vector A=&Multiply_Vector(
          A=&NormalizeVector(
           Vector To Normalize=&ReturnMeshOrientationAxis(Mesh Orientation Axis=::XAxis, Mesh Orientation Quat=$Particles.MeshOrientation)
           Safe Normalize=true
           Safe Fallback Threshold=0
           Safe Fallback Vector=<0 0 0.001>
          )
          B=<1 1 0>
         )
         Vector B=&NormalizeVector(
          Vector To Normalize=&Multiply_Vector(A=$System.CamForVec, B=<1 1 0>)
          Safe Normalize=true
          Safe Fallback Threshold=0
          Safe Fallback Vector=<0 0 0.001>
         )
        )
       )
       3:NiagaraFloat=0
       4:NiagaraFloat=?
       5:NiagaraFloat=?
       Alpha:NiagaraFloat=&FloatFromCurve(
        FloatCurve=DataInterface() {propertyValues {Curve {keys [{arriveTangent=-12.2352, interpMode=RCIM_Cubic, time=-0.7055, value=1}, <at -0.5002 0 RCIM_Cubic>]}}}
        CurveIndex=&DotProduct(Vector A=$System.CamForVec, Vector B=<0 0 1>)
       )
      )
     )
    )) {isSetParameters=true}
   /SetVariables(Particles.JiraColor2:LinearColor=&MultiplyLinearColor(
     Linear Color=$Particles.Color
     Scale Factor=&MakeFloat4FromVectorAndFloat(
      "Vector (RGB)"=<1 1 1>
      "Float (Alpha)"=&LerpMultipleFloats(
       Float Count:ENiagara_MultipleLerpCount=::NewEnumerator0
       1:NiagaraFloat=1
       2:NiagaraFloat=&FloatFromCurve(
        FloatCurve=DataInterface() {propertyValues {Curve {keys [<at -0.7 0 RCIM_Cubic>, {arriveTangent=2.9293, interpMode=RCIM_Cubic, leaveTangent=2.9293, tangentMode=RCTM_User, time=-0.0237, value=1}]}}}
        CurveIndex=&DotProduct(
         Vector A=&Multiply_Vector(
          A=&NormalizeVector(
           Vector To Normalize=&ReturnMeshOrientationAxis(Mesh Orientation Axis=::XAxis, Mesh Orientation Quat=$Particles.MeshOrientation)
           Safe Normalize=true
           Safe Fallback Threshold=0
           Safe Fallback Vector=<0 0 0.001>
          )
          B=<1 1 0>
         )
         Vector B=&NormalizeVector(
          Vector To Normalize=&Multiply_Vector(A=$System.CamForVec, B=<1 1 0>)
          Safe Normalize=true
          Safe Fallback Threshold=0
          Safe Fallback Vector=<0 0 0.001>
         )
        )
       )
       3:NiagaraFloat=0
       4:NiagaraFloat=?
       5:NiagaraFloat=?
       Alpha:NiagaraFloat=&FloatFromCurve(
        FloatCurve=DataInterface() {propertyValues {Curve {keys [{arriveTangent=-12.2352, interpMode=RCIM_Cubic, time=-0.5735, value=1}, <at -0.3682 0 RCIM_Cubic>]}}}
        CurveIndex=&DotProduct(Vector A=$System.CamForVec, Vector B=<0 0 1>)
       )
      )
     )
    )) {isSetParameters=true}
  ]
  rendererClasses [&NiagaraMeshRendererProperties]
  renderers [
   {rendererClass=&NiagaraMeshRendererProperties, rendererIndex=0}
   {rendererClass=&NiagaraMeshRendererProperties, rendererIndex=1}
   {rendererClass=&NiagaraMeshRendererProperties, rendererIndex=2}
   {rendererClass=&NiagaraMeshRendererProperties, rendererIndex=3}
  ]
 }
 Rotating_Air_Meshes {
  emitterUpdateScript [
   &EmitterState()
   &SpawnRate(Use Spawn Probability=false) {enabled=false, module=SpawnRate001}
   &SpawnRate(SpawnRate=25, Use Spawn Probability=false)
  ]
  particleSpawnScript [
   &FN_Orbit_Init_Variables_v2()
   /Niagara/Modules/Spawn/Initialization/InitializeParticle(
    Color=<rgba 0.0546 0.0662 0.0677 0.35>
    Lifetime=&UniformRangedFloat(Minimum=1, Maximum=1.5)
    Mesh Scale=&UniformRangedVector(Minimum=<5 5 8>, Maximum=<8 8 11>)
    Module.Write Color=true
    Module.Write Lifetime=true
    Module.Write Mass=false
    Module.Write Material Random=true
    Module.Write Position=false
    Module.Write Scale=true
    Module.Write SpriteRotation=false
    Module.Write SpriteSize=false
    Position=&SimulationPosition()
   ) {enabled=false}
   /Niagara/Modules/Spawn/Initialization/V2/InitializeParticle(
    Color=<rgba 0.0799 0.2278 0.5 1>
    Lifetime=&UniformRangedFloat(Minimum=1, Maximum=1)
    Mesh Scale=&Multiply_VectorByFloat(Vector=&RandomRangeVector(Minimum=<1 1 0.5>, Maximum=<1 1 0.75>), Float=1.5)
    Mesh Scale Mode=::NewEnumerator4
    Sprite Size Mode=::NewEnumerator4
    UsePositionOffset=false
    Write Material Random=true
   ) {module=InitializeParticle002}
   &DynamicMaterialParameters(
    Index 0 Param 1=&UniformRangedFloat(Minimum=0.1, Maximum=0.25)
    Index 0 Param 2=1
    Module.Param0WriteEnabled=true
    Module.Param1WriteEnabled=true
    Module.Param2WriteEnabled=true
    Module.Param3WriteEnabled=true
   )
   &FN_MeshRenderer_v2(
    Color:LinearColor=&LerpLinearColors(
     StartColor=<rgba 0.2396 0.6834 1.5 0.75>
     EndColor=<rgba 0.1301 0.3712 0.8148 0.75>
     LerpFactor=&FloatFromCurve(
      FloatCurve=DataInterface() {propertyValues {Curve {keys [
          {
           arriveTangentWeight=0.7106
           interpMode=RCIM_Cubic
           leaveTangentWeight=0.7106
           tangentMode=RCTM_User
           tangentWeightMode=RCTWM_WeightedBoth
           time=0
           value=0
          }
          {
           arriveTangent=5.2731
           arriveTangentWeight=0.5583
           interpMode=RCIM_Cubic
           leaveTangent=5.2731
           leaveTangentWeight=0.5583
           tangentMode=RCTM_User
           tangentWeightMode=RCTWM_WeightedBoth
           time=1
           value=1
          }
         ]}}}
      CurveIndex=&RandomRangeFloat(Minimum=0, Maximum=1)
     )
    )
    Create COlor Attribute=true
    Create Dynamic Param Attribute=true
    Create Scale Attribute=true
    DynamicMaterial:Vector4f=&MakeVector4(
     X=&UniformRangedFloat(Minimum=0, Maximum=1)
     Y=&UniformRangedFloat(Minimum=5, Maximum=6)
     Z=0
     W=2
    )
    Scale:Vector3f=&Multiply_Vector(
     A=&UniformRangedVector(Minimum=<1 1 0.65>, Maximum=<1 1 0.65>)
     B=&VectorFromFloat(Value=1.5)
    )
   )
   &InitialMeshOrientation(
    Mesh Orientation Mode=::NewEnumerator7
    Rotation=&RandomRangeVector(Minimum=<0 0 0>, Maximum=<0 0 1>)
    Use Rotation Vector=true
   )
   /SetVariables(Particles.Cloud 2 Scale:Vector3f=&Multiply_VectorByFloat(Vector=<1.2 1.5 1.2>, Float=1.1)) {isSetParameters=true, module=SetVariables_#20}
  ]
  particleUpdateScript [
   &ParticleState()
   &ShapeLocation(
    MaskSpawnGroup=false
    Offset=&Add_Vector(
     A=<0 0 42>
     B=&Multiply_Vector(
      A=&ReturnMeshOrientationAxis(Mesh Orientation Axis=::XAxis, Mesh Orientation Quat=$Particles.MeshOrientation)
      B=&VectorFromFloat(Value=&Lerp_Float(
        A=1
        B=25
        Alpha=&FloatFromCurve(
         FloatCurve=DataInterface() {propertyValues {Curve {keys [<at -1 1>,<at -0.8 0>]}}}
         CurveIndex=&DotProduct(Vector A=<0 0 1>, Vector B=$System.CamForVec)
        )
       ))
     )
    )
    Offset Mode=::NewEnumerator0
    Override Seed=false
    Shape Origin=$System.Lighting Start
    Sphere Radius=0
    UseOwnerScale=true
   )
   &ScaleMeshSize(Scale Factor=&VectorFromCurve(VectorCurve=DataInterface() {propertyValues {
       XCurve {keys [{tangentMode=RCTM_User, time=0, value=0.5}, {tangentMode=RCTM_User, time=1, value=1}]}
       YCurve {keys [{tangentMode=RCTM_User, time=0, value=0.5}, {tangentMode=RCTM_User, time=1, value=1}]}
       ZCurve {keys [<at 0 1 RCIM_Cubic>,<at 1 1>]}
      }}))
   &ScaleColor(
    Scale Alpha=&FloatFromCurve(
     FloatCurve=DataInterface() {propertyValues {Curve {keys [
         <at 0 0 RCIM_Cubic>
         <at 0.3 1 RCIM_Cubic>
         {arriveTangent=-4.5166, interpMode=RCIM_Cubic, leaveTangent=-4.5166, tangentMode=RCTM_User, time=1, value=0}
        ]}}}
     Scale Curve=&MakeCustomFloatFromBool(
      Bool=&IsPlatformSetActive(New Platform Set=DataInterface() {propertyValues {Platforms {qualityLevelMask=-4}}})
      True Float=1
      False Float=50
     )
    )
    ScaleA=true
    ScaleRGB=true
   )
   &FN_MeshRenderer_v2(
    Color:LinearColor=&MultiplyLinearColor(
     Linear Color=$Particles.Initial.FN_MeshRenderer_v2.Color
     Scale Factor=&MakeFloat4FromVectorAndFloat(
      "Vector (RGB)"=&VectorFromFloat(Value=1)
      "Float (Alpha)"=&FloatFromCurve(
       FloatCurve=DataInterface() {propertyValues {Curve {keys [
           {interpMode=RCIM_Cubic, tangentMode=RCTM_User, time=0, value=0}
           {interpMode=RCIM_Cubic, tangentMode=RCTM_User, time=0.5, value=0.2}
           {tangentMode=RCTM_User, time=1, value=0}
          ]}}}
       Scale Curve=&Lerp_Float(
        A=&MakeCustomFloatFromBool(
         Bool=&IsPlatformSetActive(New Platform Set=DataInterface() {propertyValues {Platforms {qualityLevelMask=-4}}})
         True Float=2
         False Float=2
        )
        B=&Multiply_Float(
         A=$System.ViewDot
         B=&MakeCustomFloatFromBool(
          Bool=&IsPlatformSetActive(New Platform Set=DataInterface() {propertyValues {Platforms {qualityLevelMask=-31}}})
          True Float=5
          False Float=&MakeCustomFloatFromBool(
           Bool=&IsPlatformSetActive(New Platform Set=DataInterface() {propertyValues {Platforms {qualityLevelMask=-30}}})
           True Float=3
           False Float=&MakeCustomFloatFromBool(
            Bool=&IsPlatformSetActive(New Platform Set=DataInterface() {propertyValues {Platforms {qualityLevelMask=-28}}})
            True Float=1
            False Float=1
           )
          )
         )
        )
        Alpha=0
       )
      )
     )
    )
    Create COlor Attribute=true
    Create Position Attribute=true
    Create Scale Attribute=true
    Scale:Vector3f=&Multiply_Vector(
     A=$Particles.Initial.FN_MeshRenderer_v2.Scale
     B=&VectorFromCurve(VectorCurve=DataInterface() {propertyValues {
        XCurve {keys [{tangentMode=RCTM_User, time=0, value=0.5}, {tangentMode=RCTM_User, time=1, value=1}]}
        YCurve {keys [{tangentMode=RCTM_User, time=0, value=0.5}, {tangentMode=RCTM_User, time=1, value=1}]}
        ZCurve {keys [<at 0 1 RCIM_Cubic>,<at 1 1>]}
       }})
    )
   ) {module=FN_MeshRenderer_v2001}
   &UpdateMeshOrientation(
    Rotation Rate=&Lerp_Float(
     A=&RandomRangeFloat(Minimum=0.25, Maximum=0.4)
     B=&RandomRangeFloat(Minimum=0.25, Maximum=0.4)
     Alpha=&FloatFromCurve(
      FloatCurve=DataInterface() {propertyValues {Curve {keys [<at -1 1>,<at -0.8 0>]}}}
      CurveIndex=&DotProduct(Vector A=<0 0 1>, Vector B=$System.CamForVec)
     )
    )
    Rotation Vector=<0 0 1>
   )
   &UpdateMeshOrientation(Rotation Rate=&Multiply_Float(A=0.03, B=&RandomRangeFloat(Minimum=-1, Maximum=1))) {module=UpdateMeshOrientation001}
   /SetVariables(Particles.TransCloudQuat:Quat4f=$Particles.Initial.MeshOrientation) {enabled=false, isSetParameters=true, module=SetVariables_#19}
   /SetVariables(Particles.VisibilityTag=&MakeCustomIntFromBool(Bool=&SetBoolByFloatComparison(A=$System.CloudDot, B=0.5), True INT=1, False INT=0)) {enabled=false, isSetParameters=true, module=SetVariables_#21}
   &ScaleColor(
    Scale Alpha=&FloatFromCurve(
     FloatCurve=DataInterface() {propertyValues {Curve {keys [
         {
          arriveTangent=1.1973
          arriveTangentWeight=0.5161
          interpMode=RCIM_Cubic
          leaveTangent=1.1973
          leaveTangentWeight=0.5161
          tangentMode=RCTM_User
          tangentWeightMode=RCTWM_WeightedBoth
          time=0
          value=0.2
         }
         {
          arriveTangent=0.593
          arriveTangentWeight=0.6151
          interpMode=RCIM_Cubic
          leaveTangent=0.593
          leaveTangentWeight=0.6151
          tangentMode=RCTM_User
          tangentWeightMode=RCTWM_WeightedBoth
          time=1
          value=1
         }
        ]}}}
     CurveIndex=$System.ViewDot
    )
    ScaleA=true
    ScaleRGB=true
   ) {module=ScaleColor001}
   &ScaleColor(
    Scale Alpha=&MakeCustomFloatFromBool(
     Bool=&IsPlatformSetActive(New Platform Set=DataInterface() {propertyValues {Platforms {qualityLevelMask=-31}}})
     True Float=1.1
     False Float=&MakeCustomFloatFromBool(
      Bool=&IsPlatformSetActive(New Platform Set=DataInterface() {propertyValues {Platforms {qualityLevelMask=-30}}})
      True Float=1.1
      False Float=&MakeCustomFloatFromBool(
       Bool=&IsPlatformSetActive(New Platform Set=DataInterface() {propertyValues {Platforms {qualityLevelMask=-28}}})
       True Float=1
       False Float=1
      )
     )
    )
    ScaleA=true
    ScaleRGB=true
   ) {module=ScaleColor002}
   &ScaleColor(
    Scale Alpha=&FloatFromCurve(
     FloatCurve=DataInterface() {propertyValues {Curve {keys [<at -1 -1>,<at 1 0>]}}}
     CurveIndex=&DotProduct(
      Vector A=&Multiply_Vector(
       A=&NormalizeVector(
        Vector To Normalize=&ReturnMeshOrientationAxis(Mesh Orientation Axis=::XAxis, Mesh Orientation Quat=$Particles.MeshOrientation)
        Safe Normalize=true
        Safe Fallback Threshold=0
        Safe Fallback Vector=<0 0 0.001>
       )
       B=<1 1 0>
      )
      Vector B=&NormalizeVector(
       Vector To Normalize=&Multiply_Vector(A=$System.CamForVec, B=<1 1 0>)
       Safe Normalize=true
       Safe Fallback Threshold=0
       Safe Fallback Vector=<0 0 0.001>
      )
     )
    )
    ScaleA=true
    ScaleRGB=true
   ) {enabled=false, module=ScaleColor003}
   /SetVariables(Particles.JiraColor:LinearColor=&MultiplyLinearColor(
     Linear Color=$Particles.FN_MeshRenderer_v2001.Color
     Scale Factor=&MakeFloat4FromVectorAndFloat(
      "Vector (RGB)"=<1 1 1>
      "Float (Alpha)"=&Lerp_Float(
       A=1
       B=&FloatFromCurve(
        FloatCurve=DataInterface() {propertyValues {Curve {keys [<at -0.7 0 RCIM_Cubic>, {arriveTangent=2.9293, interpMode=RCIM_Cubic, leaveTangent=2.9293, tangentMode=RCTM_User, time=-0.0237, value=1}]}}}
        CurveIndex=&DotProduct(
         Vector A=&Multiply_Vector(
          A=&NormalizeVector(
           Vector To Normalize=&ReturnMeshOrientationAxis(Mesh Orientation Axis=::XAxis, Mesh Orientation Quat=$Particles.MeshOrientation)
           Safe Normalize=true
           Safe Fallback Threshold=0
           Safe Fallback Vector=<0 0 0.001>
          )
          B=<1 1 0>
         )
         Vector B=&NormalizeVector(
          Vector To Normalize=&Multiply_Vector(A=$System.CamForVec, B=<1 1 0>)
          Safe Normalize=true
          Safe Fallback Threshold=0
          Safe Fallback Vector=<0 0 0.001>
         )
        )
       )
       Alpha=&FloatFromCurve(
        FloatCurve=DataInterface() {propertyValues {Curve {keys [
            <at -1 0 RCIM_Cubic>
            {interpMode=RCIM_Cubic, tangentMode=RCTM_User, time=-0.85, value=1}
            {interpMode=RCIM_Cubic, tangentMode=RCTM_User, time=-0.5566, value=1}
            <at -0.4066 0 RCIM_Cubic>
           ]}}}
        CurveIndex=&DotProduct(Vector A=$System.CamForVec, Vector B=<0 0 1>)
        Scale Curve=0.95
       )
      )
     )
    )) {isSetParameters=true}
  ]
  rendererClasses [&NiagaraMeshRendererProperties]
  renderers [
   {rendererClass=&NiagaraMeshRendererProperties, rendererIndex=0}
   {rendererClass=&NiagaraMeshRendererProperties, rendererIndex=1}
  ]
 }
 Rotating_Air_Meshes001 {
  emitterUpdateScript [
   &EmitterState()
   &SpawnRate(Use Spawn Probability=false) {enabled=false, module=SpawnRate001}
   &SpawnRate(SpawnRate=10, Use Spawn Probability=false)
  ]
  particleSpawnScript [
   &FN_Orbit_Init_Variables_v2()
   /Niagara/Modules/Spawn/Initialization/InitializeParticle(
    Color=<rgba 0.0546 0.0662 0.0677 0.35>
    Lifetime=&UniformRangedFloat(Minimum=1, Maximum=1.5)
    Mesh Scale=&UniformRangedVector(Minimum=<5 5 8>, Maximum=<8 8 11>)
    Module.Write Color=true
    Module.Write Lifetime=true
    Module.Write Mass=false
    Module.Write Material Random=true
    Module.Write Position=false
    Module.Write Scale=true
    Module.Write SpriteRotation=false
    Module.Write SpriteSize=false
    Position=&SimulationPosition()
   ) {enabled=false}
   /Niagara/Modules/Spawn/Initialization/V2/InitializeParticle(
    Color=<rgba 0.0799 0.2278 0.5 1>
    Lifetime=&UniformRangedFloat(Minimum=1.5, Maximum=2)
    Mesh Scale=&Multiply_VectorByFloat(Vector=&RandomRangeVector(Minimum=<1 1 0.5>, Maximum=<1 1 0.75>), Float=1.5)
    Mesh Scale Mode=::NewEnumerator4
    Sprite Size Mode=::NewEnumerator4
    UsePositionOffset=false
    Write Material Random=true
   ) {module=InitializeParticle002}
   &DynamicMaterialParameters(
    Index 0 Param 1=&UniformRangedFloat(Minimum=0.1, Maximum=0.25)
    Index 0 Param 2=1
    Module.Param0WriteEnabled=true
    Module.Param1WriteEnabled=true
    Module.Param2WriteEnabled=true
    Module.Param3WriteEnabled=true
   )
   &FN_MeshRenderer_v2(
    Color:LinearColor=&LerpLinearColors(
     StartColor=<rgba 0.2396 0.6834 1.5 0.9>
     EndColor=<rgba 0.1301 0.3712 0.8148 0.9>
     LerpFactor=&FloatFromCurve(
      FloatCurve=DataInterface() {propertyValues {Curve {keys [
          {
           arriveTangentWeight=0.7106
           interpMode=RCIM_Cubic
           leaveTangentWeight=0.7106
           tangentMode=RCTM_User
           tangentWeightMode=RCTWM_WeightedBoth
           time=0
           value=0
          }
          {
           arriveTangent=5.2731
           arriveTangentWeight=0.5583
           interpMode=RCIM_Cubic
           leaveTangent=5.2731
           leaveTangentWeight=0.5583
           tangentMode=RCTM_User
           tangentWeightMode=RCTWM_WeightedBoth
           time=1
           value=1
          }
         ]}}}
      CurveIndex=&RandomRangeFloat(Minimum=0, Maximum=1)
     )
    )
    Create COlor Attribute=true
    Create Dynamic Param Attribute=true
    Create Scale Attribute=true
    DynamicMaterial:Vector4f=&MakeVector4(
     X=&UniformRangedFloat(Minimum=0, Maximum=1)
     Y=&UniformRangedFloat(Minimum=5, Maximum=6)
     Z=0
     W=2
    )
    Scale:Vector3f=&Multiply_Vector(
     A=&UniformRangedVector(Minimum=<1 1 0.65>, Maximum=<1 1 0.65>)
     B=&VectorFromFloat(Value=1.5)
    )
   )
   &InitialMeshOrientation(
    Mesh Orientation Mode=::NewEnumerator7
    Rotation=&RandomRangeVector(Minimum=<0 0 0>, Maximum=<0 0 1>)
    Use Rotation Vector=true
   )
   /SetVariables(Particles.Cloud 2 Scale:Vector3f=&Multiply_VectorByFloat(Vector=<1.2 1.5 1.2>, Float=1.1)) {isSetParameters=true, module=SetVariables_#20}
   /SetVariables(Particles.Rand Dist:NiagaraFloat=&RandomRangeFloat(Minimum=0.2, Maximum=1)) {isSetParameters=true}
  ]
  particleUpdateScript [
   &ParticleState()
   &ScaleMeshSize(Scale Factor=&VectorFromCurve(VectorCurve=DataInterface() {propertyValues {
       XCurve {keys [{tangentMode=RCTM_User, time=0, value=0.5}, {tangentMode=RCTM_User, time=1, value=1}]}
       YCurve {keys [{tangentMode=RCTM_User, time=0, value=0.5}, {tangentMode=RCTM_User, time=1, value=1}]}
       ZCurve {keys [<at 0 1 RCIM_Cubic>,<at 1 1>]}
      }}))
   &ScaleColor(
    Scale Alpha=&FloatFromCurve(FloatCurve=DataInterface() {propertyValues {Curve {keys [
         <at 0 0 RCIM_Cubic>
         <at 0.3 1 RCIM_Cubic>
         {arriveTangent=-4.5166, interpMode=RCIM_Cubic, leaveTangent=-4.5166, tangentMode=RCTM_User, time=1, value=0}
        ]}}})
    ScaleA=true
    ScaleRGB=true
   )
   &FN_MeshRenderer_v2(
    Color:LinearColor=&MultiplyLinearColor(
     Linear Color=$Particles.Initial.FN_MeshRenderer_v2.Color
     Scale Factor=&MakeFloat4FromVectorAndFloat(
      "Vector (RGB)"=&VectorFromFloat(Value=1)
      "Float (Alpha)"=&Multiply_Float(
       A=&FloatFromCurve(
        FloatCurve=DataInterface() {propertyValues {Curve {keys [
            {interpMode=RCIM_Cubic, tangentMode=RCTM_User, time=0, value=0}
            {interpMode=RCIM_Cubic, tangentMode=RCTM_User, time=0.5, value=0.2}
            {tangentMode=RCTM_User, time=1, value=0}
           ]}}}
        Scale Curve=&Lerp_Float(
         A=&MakeCustomFloatFromBool(
          Bool=&IsPlatformSetActive(New Platform Set=DataInterface() {propertyValues {Platforms {qualityLevelMask=-4}}})
          True Float=2
          False Float=2
         )
         B=&Multiply_Float(
          A=$System.ViewDot
          B=&MakeCustomFloatFromBool(
           Bool=&IsPlatformSetActive(New Platform Set=DataInterface() {propertyValues {Platforms {qualityLevelMask=-31}}})
           True Float=5
           False Float=&MakeCustomFloatFromBool(
            Bool=&IsPlatformSetActive(New Platform Set=DataInterface() {propertyValues {Platforms {qualityLevelMask=-30}}})
            True Float=3
            False Float=&MakeCustomFloatFromBool(
             Bool=&IsPlatformSetActive(New Platform Set=DataInterface() {propertyValues {Platforms {qualityLevelMask=-28}}})
             True Float=1
             False Float=1
            )
           )
          )
         )
         Alpha=0
        )
       )
       B=&FloatFromCurve(
        FloatCurve=DataInterface() {propertyValues {Curve {keys [
            {arriveTangent=4.0469, arriveTangentWeight=0.9089, interpMode=RCIM_Cubic, leaveTangentWeight=0.9089, time=0, value=1}
            {
             arriveTangentWeight=0.6453
             interpMode=RCIM_Cubic
             leaveTangentWeight=0.6453
             tangentMode=RCTM_User
             tangentWeightMode=RCTWM_WeightedBoth
             time=1
             value=1
            }
           ]}}}
        CurveIndex=$System.ViewDot
       )
      )
     )
    )
    Create COlor Attribute=true
    Create Position Attribute=true
    Create Scale Attribute=true
    Scale:Vector3f=&Multiply_Vector(
     A=$Particles.Initial.FN_MeshRenderer_v2.Scale
     B=&VectorFromCurve(VectorCurve=DataInterface() {propertyValues {
        XCurve {keys [{tangentMode=RCTM_User, time=0, value=0.5}, {tangentMode=RCTM_User, time=1, value=1}]}
        YCurve {keys [{tangentMode=RCTM_User, time=0, value=0.5}, {tangentMode=RCTM_User, time=1, value=1}]}
        ZCurve {keys [<at 0 1 RCIM_Cubic>,<at 1 1>]}
       }})
    )
   ) {module=FN_MeshRenderer_v2001}
   &FN_MeshRenderer_v2(
    Color:LinearColor=&MultiplyLinearColor(
     Linear Color=$Particles.Initial.FN_MeshRenderer_v2.Color
     Scale Factor=&MakeFloat4FromVectorAndFloat(
      "Vector (RGB)"=&VectorFromFloat(Value=1)
      "Float (Alpha)"=&Multiply_Float(
       A=&FloatFromCurve(
        FloatCurve=DataInterface() {propertyValues {Curve {keys [
            {interpMode=RCIM_Cubic, tangentMode=RCTM_User, time=0, value=0}
            {interpMode=RCIM_Cubic, tangentMode=RCTM_User, time=0.5, value=0.2}
            {tangentMode=RCTM_User, time=1, value=0}
           ]}}}
        Scale Curve=&Lerp_Float(
         A=&MakeCustomFloatFromBool(
          Bool=&IsPlatformSetActive(New Platform Set=DataInterface() {propertyValues {Platforms {qualityLevelMask=-4}}})
          True Float=10
          False Float=50
         )
         B=&Multiply_Float(
          A=$System.ViewDot
          B=&MakeCustomFloatFromBool(
           Bool=&IsPlatformSetActive(New Platform Set=DataInterface() {propertyValues {Platforms {qualityLevelMask=-31}}})
           True Float=5
           False Float=&MakeCustomFloatFromBool(
            Bool=&IsPlatformSetActive(New Platform Set=DataInterface() {propertyValues {Platforms {qualityLevelMask=-30}}})
            True Float=3
            False Float=&MakeCustomFloatFromBool(
             Bool=&IsPlatformSetActive(New Platform Set=DataInterface() {propertyValues {Platforms {qualityLevelMask=-28}}})
             True Float=1
             False Float=1
            )
           )
          )
         )
         Alpha=0
        )
       )
       B=&FloatFromCurve(
        FloatCurve=DataInterface() {propertyValues {Curve {keys [
            {arriveTangent=4.0469, arriveTangentWeight=0.9089, interpMode=RCIM_Cubic, leaveTangentWeight=0.9089, time=0, value=-0.2}
            {
             arriveTangentWeight=0.6453
             interpMode=RCIM_Cubic
             leaveTangentWeight=0.6453
             tangentMode=RCTM_User
             tangentWeightMode=RCTWM_WeightedBoth
             time=1
             value=1
            }
           ]}}}
        CurveIndex=$System.ViewDot
       )
      )
     )
    )
    Create COlor Attribute=true
    Create Position Attribute=true
    Create Scale Attribute=true
    Scale:Vector3f=&Multiply_Vector(
     A=$Particles.Initial.FN_MeshRenderer_v2.Scale
     B=&VectorFromCurve(VectorCurve=DataInterface() {propertyValues {
        XCurve {keys [{tangentMode=RCTM_User, time=0, value=0.5}, {tangentMode=RCTM_User, time=1, value=1}]}
        YCurve {keys [{tangentMode=RCTM_User, time=0, value=0.5}, {tangentMode=RCTM_User, time=1, value=1}]}
        ZCurve {keys [<at 0 1 RCIM_Cubic>,<at 1 1>]}
       }})
    )
   ) {enabled=false, module=FN_MeshRenderer_v2002}
   &UpdateMeshOrientation(
    Rotation Rate=&Lerp_Float(
     A=&RandomRangeFloat(Minimum=0.25, Maximum=0.5)
     B=&RandomRangeFloat(Minimum=0.25, Maximum=0.5)
     Alpha=&FloatFromCurve(
      FloatCurve=DataInterface() {propertyValues {Curve {keys [<at -1 1>,<at -0.8 0>]}}}
      CurveIndex=&DotProduct(Vector A=<0 0 1>, Vector B=$System.CamForVec)
     )
    )
    Rotation Vector=<0 0 1>
   )
   &UpdateMeshOrientation(Rotation Rate=&Multiply_Float(A=0.03, B=&RandomRangeFloat(Minimum=-1, Maximum=1))) {module=UpdateMeshOrientation001}
   &ShapeLocation(
    Cylinder Height=0
    Cylinder Radius=1
    MaskSpawnGroup=false
    Offset=&Add_Vector(
     A=&MakeVector(
      X=0
      Y=0
      Z=&FloatFromCurve(
       FloatCurve=DataInterface() {propertyValues {Curve {keys [<at -0.2 22>,<at 0.04 40>,<at 0.64 45>]}}}
       CurveIndex=&DotProduct(Vector A=$System.CamForVec, Vector B=<0 0 1>)
      )
     )
     B=&Multiply_VectorByFloat(
      Vector=&ReturnMeshOrientationAxis(Mesh Orientation Axis=::XAxis, Mesh Orientation Quat=$Particles.MeshOrientation)
      Float=&Multiply_Float(
       A=$Particles.Rand Dist
       B=&Multiply_Float(
        A=50
        B=&MakeFloatFromVector(
         Vector=&VectorFromFloat(Value=&Lerp_Float(
           A=1
           B=2
           Alpha=&FloatFromCurve(
            FloatCurve=DataInterface() {propertyValues {Curve {keys [<at -1 1>,<at -0.8 0>]}}}
            CurveIndex=&DotProduct(Vector A=<0 0 1>, Vector B=$System.CamForVec)
           )
          ))
         Channel=::NewEnumerator0
        )
       )
      )
     )
    )
    Offset Mode=::NewEnumerator0
    Override Seed=false
    Shape Origin=$System.Lighting Start
    Shape Primitive=::NewEnumerator1
    Surface Only=false
    UseOwnerScale=true
   )
   /SetVariables(Particles.Cloud 2 Scale:Vector3f=&Multiply_VectorByFloat(Vector=&Lerp_Vector(A=<0.5 2 1>, B=<1 1.65 1.32>, Alpha=$Particles.Rand Dist), Float=1.1)) {isSetParameters=true}
   /SetVariables(Particles.TransCloudQuat:Quat4f=$Particles.Initial.MeshOrientation) {enabled=false, isSetParameters=true, module=SetVariables_#19}
   /SetVariables(Particles.VisibilityTag=&MakeCustomIntFromBool(Bool=&SetBoolByFloatComparison(A=$System.CloudDot, B=0.5), True INT=1, False INT=0)) {enabled=false, isSetParameters=true, module=SetVariables_#21}
   &ScaleColor(
    Scale Alpha=&FloatFromCurve(
     FloatCurve=DataInterface() {propertyValues {Curve {keys [
         {
          arriveTangent=4.0469
          arriveTangentWeight=0.9089
          interpMode=RCIM_Cubic
          leaveTangent=4.0469
          leaveTangentWeight=0.9089
          tangentMode=RCTM_User
          tangentWeightMode=RCTWM_WeightedBoth
          time=0
          value=0
         }
         {
          arriveTangentWeight=0.6453
          interpMode=RCIM_Cubic
          leaveTangentWeight=0.6453
          tangentMode=RCTM_User
          tangentWeightMode=RCTWM_WeightedBoth
          time=1
          value=1
         }
        ]}}}
     CurveIndex=$System.ViewDot
    )
    ScaleA=true
    ScaleRGB=true
   ) {enabled=false, module=ScaleColor001}
   &ScaleColor(
    Scale Alpha=&MakeCustomFloatFromBool(
     Bool=&IsPlatformSetActive(New Platform Set=DataInterface() {propertyValues {Platforms {qualityLevelMask=-31}}})
     True Float=1.1
     False Float=&MakeCustomFloatFromBool(
      Bool=&IsPlatformSetActive(New Platform Set=DataInterface() {propertyValues {Platforms {qualityLevelMask=-30}}})
      True Float=1.1
      False Float=&MakeCustomFloatFromBool(
       Bool=&IsPlatformSetActive(New Platform Set=DataInterface() {propertyValues {Platforms {qualityLevelMask=-28}}})
       True Float=1
       False Float=1
      )
     )
    )
    ScaleA=true
    ScaleRGB=true
   ) {enabled=false, module=ScaleColor002}
   /SetVariables(Particles.JiraColor:LinearColor=&MultiplyLinearColor(
     Linear Color=$Particles.FN_MeshRenderer_v2001.Color
     Scale Factor=&MakeFloat4FromVectorAndFloat(
      "Vector (RGB)"=<1 1 1>
      "Float (Alpha)"=&Lerp_Float(
       A=1
       B=&FloatFromCurve(
        FloatCurve=DataInterface() {propertyValues {Curve {keys [<at -0.7 0 RCIM_Cubic>, {arriveTangent=2.9293, interpMode=RCIM_Cubic, leaveTangent=2.9293, tangentMode=RCTM_User, time=-0.0237, value=1}]}}}
        CurveIndex=&DotProduct(
         Vector A=&Multiply_Vector(
          A=&NormalizeVector(
           Vector To Normalize=&ReturnMeshOrientationAxis(Mesh Orientation Axis=::XAxis, Mesh Orientation Quat=$Particles.MeshOrientation)
           Safe Normalize=true
           Safe Fallback Threshold=0
           Safe Fallback Vector=<0 0 0.001>
          )
          B=<1 1 0>
         )
         Vector B=&NormalizeVector(
          Vector To Normalize=&Multiply_Vector(A=$System.CamForVec, B=<1 1 0>)
          Safe Normalize=true
          Safe Fallback Threshold=0
          Safe Fallback Vector=<0 0 0.001>
         )
        )
       )
       Alpha=&FloatFromCurve(
        FloatCurve=DataInterface() {propertyValues {Curve {keys [<at -0.72 1 RCIM_Cubic>,<at -0.55 0 RCIM_Cubic>]}}}
        CurveIndex=&DotProduct(Vector A=$System.CamForVec, Vector B=<0 0 1>)
       )
      )
     )
    )) {enabled=false, isSetParameters=true}
   /SetVariables(Particles.JiraColor:LinearColor=&MultiplyLinearColor(
     Linear Color=$Particles.FN_MeshRenderer_v2001.Color
     Scale Factor=&MakeFloat4FromVectorAndFloat(
      "Vector (RGB)"=<1 1 1>
      "Float (Alpha)"=&LerpMultipleFloats(
       Float Count:ENiagara_MultipleLerpCount=::NewEnumerator0
       1:NiagaraFloat=1
       2:NiagaraFloat=&FloatFromCurve(
        FloatCurve=DataInterface() {propertyValues {Curve {keys [<at -0.7 0 RCIM_Cubic>, {arriveTangent=2.9293, interpMode=RCIM_Cubic, leaveTangent=2.9293, tangentMode=RCTM_User, time=-0.0237, value=1}]}}}
        CurveIndex=&DotProduct(
         Vector A=&Multiply_Vector(
          A=&NormalizeVector(
           Vector To Normalize=&ReturnMeshOrientationAxis(Mesh Orientation Axis=::XAxis, Mesh Orientation Quat=$Particles.MeshOrientation)
           Safe Normalize=true
           Safe Fallback Threshold=0
           Safe Fallback Vector=<0 0 0.001>
          )
          B=<1 1 0>
         )
         Vector B=&NormalizeVector(
          Vector To Normalize=&Multiply_Vector(A=$System.CamForVec, B=<1 1 0>)
          Safe Normalize=true
          Safe Fallback Threshold=0
          Safe Fallback Vector=<0 0 0.001>
         )
        )
       )
       3:NiagaraFloat=0
       4:NiagaraFloat=?
       5:NiagaraFloat=?
       Alpha:NiagaraFloat=&FloatFromCurve(
        FloatCurve=DataInterface() {propertyValues {Curve {keys [{arriveTangent=-12.2352, interpMode=RCIM_Cubic, time=-0.7055, value=1}, <at -0.5002 0 RCIM_Cubic>]}}}
        CurveIndex=&DotProduct(Vector A=$System.CamForVec, Vector B=<0 0 1>)
       )
      )
     )
    )) {isSetParameters=true}
  ]
  rendererClasses [&NiagaraMeshRendererProperties]
  renderers [{rendererClass=&NiagaraMeshRendererProperties, rendererIndex=0}]
 }
 Rotating_Air_Meshes003 {
  bEnabled=false
  emitterUpdateScript [
   &EmitterState()
   &SpawnRate(Use Spawn Probability=false) {module=SpawnRate001}
   &SpawnRate(SpawnRate=25, Use Spawn Probability=false)
  ]
  particleSpawnScript [
   &FN_Orbit_Init_Variables_v2()
   /Niagara/Modules/Spawn/Initialization/InitializeParticle(
    Color=<rgba 0.0546 0.0662 0.0677 0.35>
    Lifetime=&UniformRangedFloat(Minimum=1, Maximum=1.5)
    Mesh Scale=&UniformRangedVector(Minimum=<5 5 8>, Maximum=<8 8 11>)
    Module.Write Color=true
    Module.Write Lifetime=true
    Module.Write Mass=false
    Module.Write Material Random=true
    Module.Write Position=false
    Module.Write Scale=true
    Module.Write SpriteRotation=false
    Module.Write SpriteSize=false
    Position=&SimulationPosition()
   ) {enabled=false}
   /Niagara/Modules/Spawn/Initialization/V2/InitializeParticle(
    Color=&FNTimeOfDayColor(
     Night Color=<rgba 0.1597 0.4556 1 0.75>
     Morning Color=<rgba 0.0799 0.2278 0.5 0.75>
     Day Color=<rgba 0.0799 0.2278 0.5 0.75>
     Evening Color=<rgba 0.0799 0.2278 0.5 0.75>
    )
    Lifetime=&UniformRangedFloat(Minimum=1, Maximum=1.5)
    Mesh Scale=&Multiply_VectorByFloat(Vector=&RandomRangeVector(Minimum=<1 1 0.5>, Maximum=<1 1 0.75>), Float=1.5)
    Mesh Scale Mode=::NewEnumerator4
    Sprite Size Mode=::NewEnumerator4
    UsePositionOffset=false
    Write Material Random=true
   ) {module=InitializeParticle002}
   &DynamicMaterialParameters(
    Index 0 Param 1=&UniformRangedFloat(Minimum=0.1, Maximum=0.25)
    Index 0 Param 2=1
    Module.Param0WriteEnabled=true
    Module.Param1WriteEnabled=true
    Module.Param2WriteEnabled=true
    Module.Param3WriteEnabled=true
   )
   &FN_MeshRenderer_v2(
    Color:LinearColor=&LerpLinearColors(
     StartColor=<rgba 0.2396 0.6834 1.5 1>
     EndColor=<rgba 0.1301 0.3712 0.8148 1>
     LerpFactor=&FloatFromCurve(
      FloatCurve=DataInterface() {propertyValues {Curve {keys [
          {
           arriveTangentWeight=0.7106
           interpMode=RCIM_Cubic
           leaveTangentWeight=0.7106
           tangentMode=RCTM_User
           tangentWeightMode=RCTWM_WeightedBoth
           time=0
           value=0
          }
          {
           arriveTangent=5.2731
           arriveTangentWeight=0.5583
           interpMode=RCIM_Cubic
           leaveTangent=5.2731
           leaveTangentWeight=0.5583
           tangentMode=RCTM_User
           tangentWeightMode=RCTWM_WeightedBoth
           time=1
           value=1
          }
         ]}}}
      CurveIndex=&RandomRangeFloat(Minimum=0, Maximum=1)
     )
    )
    Create COlor Attribute=true
    Create Dynamic Param Attribute=true
    Create Scale Attribute=true
    DynamicMaterial:Vector4f=&MakeVector4(
     X=&UniformRangedFloat(Minimum=0, Maximum=1)
     Y=&UniformRangedFloat(Minimum=5, Maximum=6)
     Z=0
     W=2
    )
    Scale:Vector3f=&Multiply_Vector(
     A=&UniformRangedVector(Minimum=<1 1 0.65>, Maximum=<1 1 0.65>)
     B=&VectorFromFloat(Value=1.5)
    )
   )
   &InitialMeshOrientation(
    Mesh Orientation Mode=::NewEnumerator7
    Rotation=&RandomRangeVector(Minimum=<0 0 0>, Maximum=<0 0 1>)
    Use Rotation Vector=true
   )
   /SetVariables(Particles.Cloud 2 Scale:Vector3f=<1.2 1.2 1.2>) {isSetParameters=true, module=SetVariables_#20}
  ]
  particleUpdateScript [
   &ParticleState()
   &ShapeLocation(
    MaskSpawnGroup=false
    Offset=&Add_Vector(
     A=<0 0 42>
     B=&Multiply_Vector(
      A=&ReturnMeshOrientationAxis(Mesh Orientation Axis=::XAxis, Mesh Orientation Quat=$Particles.MeshOrientation)
      B=&VectorFromFloat(Value=&Lerp_Float(
        A=1
        B=25
        Alpha=&FloatFromCurve(
         FloatCurve=DataInterface() {propertyValues {Curve {keys [<at -1 1>,<at -0.8 0>]}}}
         CurveIndex=&DotProduct(Vector A=<0 0 1>, Vector B=$System.CamForVec)
        )
       ))
     )
    )
    Offset Mode=::NewEnumerator0
    Override Seed=false
    Shape Origin=$System.Lighting Start
    Sphere Radius=0
    UseOwnerScale=true
   )
   &ScaleMeshSize(Scale Factor=&VectorFromCurve(
     VectorCurve=DataInterface() {propertyValues {
       XCurve {keys [{tangentMode=RCTM_User, time=0, value=0.5}, {tangentMode=RCTM_User, time=1, value=1}]}
       YCurve {keys [{tangentMode=RCTM_User, time=0, value=0.5}, {tangentMode=RCTM_User, time=1, value=1}]}
       ZCurve {keys [<at 0 1 RCIM_Cubic>,<at 1 1>]}
      }}
     Scale Curve=<500 500 500>
    ))
   &ScaleColor(
    Scale Alpha=&FloatFromCurve(FloatCurve=DataInterface() {propertyValues {Curve {keys [
         <at 0 0 RCIM_Cubic>
         <at 0.3 1 RCIM_Cubic>
         {arriveTangent=-1.0288, interpMode=RCIM_Cubic, leaveTangent=-1.0288, tangentMode=RCTM_User, time=1, value=0}
        ]}}})
    ScaleA=true
    ScaleRGB=true
   )
   &FN_MeshRenderer_v2(
    Color:LinearColor=&MultiplyLinearColor(
     Linear Color=$Particles.Initial.FN_MeshRenderer_v2.Color
     Scale Factor=&MakeFloat4FromVectorAndFloat(
      "Vector (RGB)"=&VectorFromFloat(Value=1)
      "Float (Alpha)"=&FloatFromCurve(
       FloatCurve=DataInterface() {propertyValues {Curve {keys [
           <at 0 0 RCIM_Cubic>
           {arriveTangent=1.3333, leaveTangent=1.3333, time=0.15, value=1}
           <at 0.85 1 RCIM_Cubic>
           {arriveTangent=-4.5166, interpMode=RCIM_Cubic, leaveTangent=-4.5166, tangentMode=RCTM_User, time=1, value=0}
          ]}}}
       Scale Curve=&Lerp_Float(
        A=1
        B=&Multiply_Float(
         A=$System.ViewDot
         B=&MakeCustomFloatFromBool(
          Bool=&IsPlatformSetActive(New Platform Set=DataInterface() {propertyValues {Platforms {qualityLevelMask=-31}}})
          True Float=5
          False Float=&MakeCustomFloatFromBool(
           Bool=&IsPlatformSetActive(New Platform Set=DataInterface() {propertyValues {Platforms {qualityLevelMask=-30}}})
           True Float=3
           False Float=&MakeCustomFloatFromBool(
            Bool=&IsPlatformSetActive(New Platform Set=DataInterface() {propertyValues {Platforms {qualityLevelMask=-28}}})
            True Float=1
            False Float=1
           )
          )
         )
        )
        Alpha=0
       )
      )
     )
    )
    Create COlor Attribute=true
    Create Position Attribute=true
    Create Scale Attribute=true
    Scale:Vector3f=&Multiply_Vector(
     A=$Particles.Initial.FN_MeshRenderer_v2.Scale
     B=&VectorFromCurve(
      VectorCurve=DataInterface() {propertyValues {
        XCurve {keys [{tangentMode=RCTM_User, time=0, value=0.5}, {tangentMode=RCTM_User, time=1, value=1}]}
        YCurve {keys [{tangentMode=RCTM_User, time=0, value=0.5}, {tangentMode=RCTM_User, time=1, value=1}]}
        ZCurve {keys [<at 0 1 RCIM_Cubic>,<at 1 1>]}
       }}
      Scale Curve=<500 500 500>
     )
    )
   ) {module=FN_MeshRenderer_v2001}
   &UpdateMeshOrientation(
    Rotation Rate=&Lerp_Float(
     A=&RandomRangeFloat(Minimum=0.25, Maximum=0.5)
     B=&RandomRangeFloat(Minimum=0.25, Maximum=0.5)
     Alpha=&FloatFromCurve(
      FloatCurve=DataInterface() {propertyValues {Curve {keys [<at -1 1>,<at -0.8 0>]}}}
      CurveIndex=&DotProduct(Vector A=<0 0 1>, Vector B=$System.CamForVec)
     )
    )
    Rotation Vector=<0 0 1>
   )
   &UpdateMeshOrientation(Rotation Rate=&Multiply_Float(A=0.03, B=&RandomRangeFloat(Minimum=-1, Maximum=1))) {module=UpdateMeshOrientation001}
   /SetVariables(Particles.TransCloudQuat:Quat4f=$Particles.Initial.MeshOrientation) {isSetParameters=true, module=SetVariables_#19}
   /SetVariables(Particles.VisibilityTag=&MakeCustomIntFromBool(Bool=&SetBoolByFloatComparison(A=$System.CloudDot, B=0.5), True INT=1, False INT=0)) {enabled=false, isSetParameters=true, module=SetVariables_#21}
   &ScaleColor(Scale Alpha=$System.ViewDot, ScaleA=true, ScaleRGB=true) {enabled=false, module=ScaleColor001}
   &ScaleColor(
    Scale Alpha=&MakeCustomFloatFromBool(
     Bool=&IsPlatformSetActive(New Platform Set=DataInterface() {propertyValues {Platforms {qualityLevelMask=-31}}})
     True Float=1.1
     False Float=&MakeCustomFloatFromBool(
      Bool=&IsPlatformSetActive(New Platform Set=DataInterface() {propertyValues {Platforms {qualityLevelMask=-30}}})
      True Float=1.1
      False Float=&MakeCustomFloatFromBool(
       Bool=&IsPlatformSetActive(New Platform Set=DataInterface() {propertyValues {Platforms {qualityLevelMask=-28}}})
       True Float=1
       False Float=1
      )
     )
    )
    ScaleA=true
    ScaleRGB=true
   ) {module=ScaleColor002}
   &ScaleColor(
    Scale RGB=&VectorFromFloat(Value=&self:ScratchDynamicInput_03(
      F1:NiagaraFloat=1.05
      F2:NiagaraFloat=1.07
      scale High:NiagaraFloat=&FNTimeOfDayScalar(Night Scale=3.5, Morning Scale=4, Day Scale=4, Evening Scale=4)
      scale low:NiagaraFloat=1
      time:NiagaraFloat=&Multiply_Float(A=$Engine.System.Age, B=1)
     ))
    ScaleA=true
    ScaleRGB=true
   ) {module=ScaleColor003}
  ]
  rendererClasses [&NiagaraMeshRendererProperties]
  renderers [{rendererClass=&NiagaraMeshRendererProperties, rendererIndex=0}]
 }
 Rotating_Air_Meshes_Middle001 {
  bEnabled=false
  emitterUpdateScript [
   &EmitterState()
   &SpawnRate(SpawnRate=15, Use Spawn Probability=false)
  ]
  particleSpawnScript [
   &FN_Orbit_Init_Variables_v2()
   /Niagara/Modules/Spawn/Initialization/InitializeParticle(
    Color=<rgba 0.0546 0.0662 0.0677 0.35>
    Lifetime=&UniformRangedFloat(Minimum=1, Maximum=1.5)
    Mesh Scale=&UniformRangedVector(Minimum=<5 5 8>, Maximum=<8 8 11>)
    Module.Write Color=true
    Module.Write Lifetime=true
    Module.Write Mass=false
    Module.Write Material Random=true
    Module.Write Position=false
    Module.Write Scale=true
    Module.Write SpriteRotation=false
    Module.Write SpriteSize=false
    Position=&SimulationPosition()
   ) {enabled=false}
   /Niagara/Modules/Spawn/Initialization/V2/InitializeParticle(
    Color=<rgba 0.0799 0.2278 0.5 0.65>
    Lifetime=&UniformRangedFloat(Minimum=1, Maximum=1.5)
    Mesh Scale=&Multiply_VectorByFloat(Vector=<0.5 0.5 0.75>, Float=1)
    Mesh Scale Mode=::NewEnumerator4
    Sprite Size Mode=::NewEnumerator4
    UsePositionOffset=false
    Write Material Random=true
   ) {module=InitializeParticle002}
   &DynamicMaterialParameters(
    Index 0 Param 1=&UniformRangedFloat(Minimum=0.1, Maximum=0.25)
    Index 0 Param 2=1
    Module.Param0WriteEnabled=true
    Module.Param1WriteEnabled=true
    Module.Param2WriteEnabled=true
    Module.Param3WriteEnabled=true
   )
   &FN_MeshRenderer_v2(
    Color:LinearColor=&LerpLinearColors(
     StartColor=<rgba 0.2396 0.6834 1.5 1>
     EndColor=<rgba 0.1301 0.3712 0.8148 1>
     LerpFactor=&FloatFromCurve(
      FloatCurve=DataInterface() {propertyValues {Curve {keys [
          {
           arriveTangentWeight=0.7106
           interpMode=RCIM_Cubic
           leaveTangentWeight=0.7106
           tangentMode=RCTM_User
           tangentWeightMode=RCTWM_WeightedBoth
           time=0
           value=0
          }
          {
           arriveTangent=5.2731
           arriveTangentWeight=0.5583
           interpMode=RCIM_Cubic
           leaveTangent=5.2731
           leaveTangentWeight=0.5583
           tangentMode=RCTM_User
           tangentWeightMode=RCTWM_WeightedBoth
           time=1
           value=1
          }
         ]}}}
      CurveIndex=&RandomRangeFloat(Minimum=0, Maximum=1)
     )
    )
    Create COlor Attribute=true
    Create Dynamic Param Attribute=true
    Create Scale Attribute=true
    DynamicMaterial:Vector4f=&MakeVector4(
     X=&UniformRangedFloat(Minimum=0, Maximum=1)
     Y=&UniformRangedFloat(Minimum=5, Maximum=6)
     Z=0
     W=2
    )
    Scale:Vector3f=&Multiply_Vector(
     A=&UniformRangedVector(Minimum=<0.5 0.5 0.5>, Maximum=<1 1 0.5>)
     B=&VectorFromFloat(Value=1)
    )
   )
   &InitialMeshOrientation(
    Mesh Orientation Mode=::NewEnumerator7
    Rotation=&RandomRangeVector(Minimum=<0 0 1>, Maximum=<0 0 0>)
    Use Rotation Vector=true
   )
  ]
  particleUpdateScript [
   &ParticleState()
   &ShapeLocation(
    MaskSpawnGroup=false
    Offset=&Add_Vector(
     A=<0 0 40>
     B=&Multiply_Vector(
      A=&ReturnMeshOrientationAxis(Mesh Orientation Axis=::XAxis, Mesh Orientation Quat=$Particles.MeshOrientation)
      B=&VectorFromFloat(Value=-10)
     )
    )
    Offset Mode=::NewEnumerator0
    Override Seed=false
    Shape Origin=$System.Lighting Start
    Sphere Radius=0
    UseOwnerScale=true
   )
   &ScaleMeshSize(Scale Factor=&VectorFromCurve(VectorCurve=DataInterface() {propertyValues {
       XCurve {keys [{tangentMode=RCTM_User, time=0, value=0.5}, {tangentMode=RCTM_User, time=1, value=1}]}
       YCurve {keys [{tangentMode=RCTM_User, time=0, value=0.5}, {tangentMode=RCTM_User, time=1, value=1}]}
       ZCurve {keys [<at 0 1 RCIM_Cubic>,<at 1 1>]}
      }}))
   &ScaleColor(
    Scale Alpha=&FloatFromCurve(FloatCurve=DataInterface() {propertyValues {Curve {keys [
         <at 0 0 RCIM_Cubic>
         <at 0.3 1 RCIM_Cubic>
         {arriveTangent=-4.5166, interpMode=RCIM_Cubic, leaveTangent=-4.5166, tangentMode=RCTM_User, time=1, value=0}
        ]}}})
    ScaleA=true
    ScaleRGB=true
   )
   &FN_MeshRenderer_v2(
    Color:LinearColor=&MultiplyLinearColor(
     Linear Color=$Particles.Initial.FN_MeshRenderer_v2.Color
     Scale Factor=&MakeFloat4FromVectorAndFloat(
      "Vector (RGB)"=&VectorFromFloat(Value=1)
      "Float (Alpha)"=&FloatFromCurve(
       FloatCurve=DataInterface() {propertyValues {Curve {keys [
           <at 0 0 RCIM_Cubic>
           {arriveTangent=1.3333, leaveTangent=1.3333, time=0.15, value=1}
           <at 0.85 1 RCIM_Cubic>
           {arriveTangent=-4.5166, interpMode=RCIM_Cubic, leaveTangent=-4.5166, tangentMode=RCTM_User, time=1, value=0}
          ]}}}
       Scale Curve=&FloatFromCurve(
        FloatCurve=DataInterface() {propertyValues {Curve {keys [
            {
             arriveTangentWeight=0.0906
             interpMode=RCIM_Cubic
             leaveTangentWeight=0.0906
             tangentMode=RCTM_User
             tangentWeightMode=RCTWM_WeightedBoth
             time=-0.85
             value=-0.01
            }
            {
             arriveTangent=7.032
             arriveTangentWeight=0.7813
             interpMode=RCIM_Cubic
             leaveTangent=7.032
             tangentMode=RCTM_User
             tangentWeightMode=RCTWM_WeightedBoth
             time=-0.67
             value=1
            }
           ]}}}
        CurveIndex=&DotProduct(Vector A=$System.CamForVec, Vector B=<0 0 1>)
       )
      )
     )
    )
    Create COlor Attribute=true
    Create Position Attribute=true
    Create Scale Attribute=true
    Scale:Vector3f=&Multiply_Vector(
     A=$Particles.Initial.FN_MeshRenderer_v2.Scale
     B=&VectorFromCurve(VectorCurve=DataInterface() {propertyValues {
        XCurve {keys [{tangentMode=RCTM_User, time=0, value=0.5}, {tangentMode=RCTM_User, time=1, value=1}]}
        YCurve {keys [{tangentMode=RCTM_User, time=0, value=0.5}, {tangentMode=RCTM_User, time=1, value=1}]}
        ZCurve {keys [<at 0 1 RCIM_Cubic>,<at 1 1>]}
       }})
    )
   ) {module=FN_MeshRenderer_v2001}
   /SetVariables() {isSetParameters=true}
   &UpdateMeshOrientation(Rotation Vector=<0 0 1>) {enabled=false}
   /SetVariables(Particles.TransCloudQuat:Quat4f=$Particles.Initial.MeshOrientation) {enabled=false, isSetParameters=true, module=SetVariables_#19}
   /SetVariables(Particles.VisibilityTag=&MakeCustomIntFromBool(Bool=&SetBoolByFloatComparison(A=$System.CloudDot, B=0.5), True INT=0, False INT=1)) {isSetParameters=true, module=SetVariables_#21}
   &ScaleMeshSize(Scale Factor=&VectorFromFloat(Value=&MakeCustomFloatFromBool(
      Bool=&SetBoolByFloatComparison(
       A=&MakeFloatFromLinearColor(LinearColor=$Particles.FN_MeshRenderer_v2001.Color, Channel:ENiagaraLinearColor_Channels=::NewEnumerator3)
       B=0
      )
      True Float=1
      False Float=0
     ))) {module=ScaleMeshSize001}
   &ScaleColor(Scale Alpha=$System.ViewDot, ScaleA=true, ScaleRGB=true) {enabled=false, module=ScaleColor001}
   &ScaleColor(
    Scale Alpha=&FloatFromCurve(
     FloatCurve=DataInterface() {propertyValues {Curve {keys [
         {
          arriveTangentWeight=0.0906
          interpMode=RCIM_Cubic
          leaveTangentWeight=0.0906
          tangentMode=RCTM_User
          tangentWeightMode=RCTWM_WeightedBoth
          time=-0.7
          value=0
         }
         {
          arriveTangent=7.032
          arriveTangentWeight=0.7813
          interpMode=RCIM_Cubic
          leaveTangent=7.032
          tangentMode=RCTM_User
          tangentWeightMode=RCTWM_WeightedBoth
          time=-0.67
          value=1
         }
        ]}}}
     CurveIndex=&DotProduct(Vector A=$System.CamForVec, Vector B=<0 0 1>)
    )
    ScaleA=true
    ScaleRGB=true
   ) {enabled=false, module=ScaleColor003}
   /SetVariables(Particles.new cloud scale:Vector3f=&Multiply_Vector(
     A=$Particles.FN_MeshRenderer_v2001.Scale
     B=&MakeCustomVectorFromBool(
      Bool=&SetBoolByFloatComparison(
       A=&MakeFloatFromLinearColor(LinearColor=$Particles.FN_MeshRenderer_v2001.Color, Channel:ENiagaraLinearColor_Channels=::NewEnumerator3)
       B=0
      )
      True Vector=&VectorFromFloat(Value=1.25)
      False Vector=&VectorFromFloat(Value=0)
     )
    )) {isSetParameters=true}
  ]
  rendererClasses [&NiagaraMeshRendererProperties]
  renderers [{rendererClass=&NiagaraMeshRendererProperties, rendererIndex=0}]
 }
 Secondary_Lightning {
  emitterSpawnScript [
   /SetVariables(Emitter.particles per line:NiagaraInt32=&MakeCustomIntFromBool(
     Bool=&IsPlatformSetActive(New Platform Set=DataInterface() {propertyValues {Platforms {qualityLevelMask=-31}}})
     True INT=4
     False INT=&MakeCustomIntFromBool(
      Bool=&IsPlatformSetActive(New Platform Set=DataInterface() {propertyValues {Platforms {qualityLevelMask=-26}}})
      True INT=6
      False INT=8
     )
    )) {isSetParameters=true, module=SetVariables_#14}
   /SetVariables(Emitter.Num Groups=4) {isSetParameters=true, module=SetVariables_#15}
  ]
  emitterUpdateScript [
   &EmitterState(Delay First Loop Only=true, Life Cycle Mode=::NewEnumerator1, Loop Delay=1, Loop Duration=1, UseLoopDelay=true)
   &SpawnBurst_Instantaneous(
    Spawn Count=&Multiply_Int(A=$Emitter.particles per line, B=$Emitter.Num Groups)
    Use Spawn Probability=false
    UseLoopCountLimit=true
   )
   /SetVariables(
    Emitter.Timer 2:NiagaraFloat=$System.Sys Timer 2
    Emitter.Timer 3:NiagaraFloat=$System.Sys Timer 3
    Emitter.timer 1:NiagaraFloat=$System.Sys Timer 1
    Emitter.timer 4:NiagaraFloat=$System.Sys Timer 4
   ) {isSetParameters=true}
   /SetVariables(
    Emitter.Shape 3:NiagaraInt32=$System.SysShape 3
    Emitter.shape 1:NiagaraInt32=$System.SysShape 1
    Emitter.shape 2:NiagaraInt32=$System.SysShape 2
    Emitter.shape 4:NiagaraInt32=$System.SysShape 4
   ) {isSetParameters=true}
   /SetVariables(
    Emitter.ini Pos Y 1:NiagaraFloat=$System.sys ini Pos Y 1
    Emitter.ini Pos Y 2:NiagaraFloat=$System.sys ini Pos Y 2
    Emitter.ini Pos Y 3:NiagaraFloat=$System.sys ini Pos Y 3
    Emitter.ini Pos Y 4:NiagaraFloat=$System.sys ini Pos Y 4
   ) {isSetParameters=true}
   /SetVariables(
    Emitter.ini Pos 1:NiagaraFloat=$System.sys ini Pos 1
    Emitter.ini Pos 2:NiagaraFloat=$System.sys ini Pos 2
    Emitter.ini Pos 3:NiagaraFloat=$System.sys ini Pos 3
    Emitter.ini Pos 4:NiagaraFloat=$System.sys ini Pos 4
   ) {isSetParameters=true}
  ]
  particleSpawnScript [
   &InitializeParticle(
    Color Mode=::NewEnumerator0
    Lifetime=0.5
    Ribbon Width=15
    Ribbon Width Mode=::NewEnumerator1
    Sprite Size Mode=::NewEnumerator3
    UsePositionOffset=false
    Write Material Random=true
   )
   &PartitionParticles(
    Normalized Index In Partition Includes One=true
    Number Of Partitions=$Emitter.Num Groups
    Particles Per Partition:NiagaraInt32=$Emitter.particles per line
    Partition Mode:ENiagara_PartitionMode=::NewEnumerator1
    Write Partition Index to Ribbon Link Order:ENiagara_AttributeSamplingApplyOutput=::NewEnumerator0
    Write Partition to Ribbon ID=::NewEnumerator0
   )
  ]
  particleUpdateScript [
   &ParticleState(Kill Particles When Lifetime Has Elapsed=false, Let Infinitely Lived Particles Die When Emitter Deactivates=true)
   &self:ScratchModule_08(
    Timer Amount:NiagaraFloat=&RandomRangeFloat(Minimum=0, Maximum=0.35)
    float:NiagaraFloat=&RandomRangeFloat(Minimum=0, Maximum=1, Evaluation Type=::NewEnumerator1)
   ) {module=ScratchModule_08}
   &self:ScratchModule_08(
    Timer Amount:NiagaraFloat=&RandomRangeFloat(Minimum=0.1, Maximum=0.35)
    float:NiagaraFloat=&RandomRangeFloat(Minimum=0, Maximum=1, Evaluation Type=::NewEnumerator1)
   ) {module=ScratchModule_08001}
   &SkeletalMeshLocation(
    Bone / Socket Index=$Particles.PartitionParticles.Partition
    Bone Sampling Mode=::NewEnumerator4
    MaskSpawnGroup=false
    Orientation Sampling=::NewEnumerator1
    Position Sampling=::NewEnumerator1
    Sampled Position Offset=&SelectVectorFromArray(
     Direct Array Index=$Particles.PartitionParticles.Partition
     Vector Selection Array=DataInterface() {propertyValues {FloatData [<5 0 0>, <-5 0 1>, <-5 7 0>, <5 -7 0>]}}
    )
    Skeletal Mesh=$System.Skeletal Mesh
    Write To Particle Position=false
   )
   &self:SLect_Shape_Mod(
    RandomSelctor:NiagaraInt32=$Particles.PartitionParticles.Partition
    Select Pos:NiagaraFloat=$Emitter.ini Pos 1
    Select Pos 2:NiagaraFloat=$Emitter.ini Pos 2
    Select Pos 3:NiagaraFloat=$Emitter.ini Pos 3
    Select Pos 4:NiagaraFloat=$Emitter.ini Pos 4
    Select Pos Y:NiagaraFloat=$Emitter.ini Pos Y 1
    Select Pos Y 2:NiagaraFloat=$Emitter.ini Pos Y 2
    Select Pos Y 3:NiagaraFloat=$Emitter.ini Pos Y 3
    Select Pos Y 4:NiagaraFloat=$Emitter.ini Pos Y 4
    Shape Select:NiagaraInt32=$Emitter.shape 1
    Shape Select 2:NiagaraInt32=$Emitter.shape 2
    Shape Select 3:NiagaraInt32=$Emitter.Shape 3
    Shape Select 4:NiagaraInt32=$Emitter.shape 4
    Timer Amount:NiagaraFloat=$Emitter.timer 1
    Timer Amount 2:NiagaraFloat=$Emitter.Timer 2
    Timer Amount 3:NiagaraFloat=$Emitter.Timer 3
    Timer Amount 4:NiagaraFloat=$Emitter.timer 4
   ) {module=SLect_Shape_Mod}
   &ShapeLocation(
    Handle Radius=18
    Large Radius=&MakeFloatFromVector(
     Vector=&VectorFromFloat(Value=&Lerp_Float(
       A=50
       B=75
       Alpha=&FloatFromCurve(
        FloatCurve=DataInterface() {propertyValues {Curve {keys [<at -1 1>,<at -0.8 0>]}}}
        CurveIndex=&DotProduct(Vector A=<0 0 1>, Vector B=$System.CamForVec)
       )
      ))
     Channel=::NewEnumerator0
    )
    MaskSpawnGroup=false
    Offset=<0 0 42>
    Offset Mode=::NewEnumerator0
    Shape Origin=$System.Lighting Start
    Shape Primitive=::NewEnumerator3
    Torus Distribution Mode:ENiagaraTorusDistributionMode=::NewEnumerator1
    U Position=&Add_Float(
     A=&self:torus_pos_float(
      float 1:NiagaraFloat=&Add_Float(A=0.25, B=0)
      float 2:NiagaraFloat=&Add_Float(A=0.5, B=0)
      float 3:NiagaraFloat=&Add_Float(A=0, B=0)
      float 4:NiagaraFloat=&Add_Float(A=0.75, B=0)
      Master Offset:NiagaraFloat=-0.1
      Partition:NiagaraInt32=$Particles.PartitionParticles.Partition
     )
     B=&Multiply_Float(A=$Particles.PartitionParticles.Normalized Index In Partition, B=0)
    )
    UseOwnerScale=true
    V Position=0.1
   ) {module=ShapeLocation001}
   &self:ScratchModule_03(
    EndPos:NiagaraPosition=$Particles.SkeletalMeshLocation.SampledPosition
    StartPos:NiagaraPosition=&LerpPosition(A=$Particles.ShapeLocation001.ShapePosition, B=$System.Lighting Start, Alpha=0)
   ) {module=ScratchModule_03}
   &InterpolateOverTime(
    Initial Vector Value:Vector3f=<1 1 1>
    Rate Of Change:NiagaraFloat=&Divide_Float(A=1, B=0.3)
    Target Vector Value:Vector3f=&TransformVector(
     Vector=&Multiply_Vector(
      A=&SelectVectorFromArray(
       Direct Array Index=$Particles.PartitionParticles.Partition
       Vector Selection Array=DataInterface() {propertyValues {FloatData [<300 50 -75>, <-300 300 -75>, <50 -300 -75>, <-300 -50 -75>]}}
      )
      B=&SelectVectorFromArray(
       Direct Array Index=&MakeCustomIntFromBool(Bool=$System.Use Temp, True INT=$System.Temp Array Selection, False INT=$Particles.ShapeMod)
       Vector Selection Array=$System.Vector Array Top
      )
     )
     Destination Space=::Simulation
    )
    Variable Type=::NewEnumerator2
   ) {module=InterpolateOverTime001}
   &InterpolateOverTime(
    Initial Vector Value:Vector3f=<1 1 1>
    Rate Of Change:NiagaraFloat=&Divide_Float(A=1, B=0.3)
    Target Vector Value:Vector3f=&TransformVector(
     Vector=&Multiply_Vector(
      A=&SelectVectorFromArray(
       Direct Array Index=$Particles.PartitionParticles.Partition
       Vector Selection Array=DataInterface() {propertyValues {FloatData [<150 150 100>, <-150 150 100>, <150 -150 100>, <-150 -150 100>]}}
      )
      B=&SelectVectorFromArray(
       Direct Array Index=&MakeCustomIntFromBool(Bool=$System.Use Temp, True INT=$System.Temp Array Selection, False INT=$Particles.ShapeMod)
       Vector Selection Array=$System.Vector Array bottom
      )
     )
     Destination Space=::Simulation
    )
    Variable Type=::NewEnumerator2
   ) {module=InterpolateOverTime002}
   &FN_HermiteSpline(
    Direction End:Vector3f=$Particles.InterpolateOverTime002.InterpolateOverTime_V3_.Moving Average
    Direction Start:Vector3f=$Particles.InterpolateOverTime001.InterpolateOverTime_V3_.Moving Average
    Position Blend:NiagaraFloat=$Particles.PartitionParticles.Normalized Index In Partition
    Position End:Vector3f=$Particles.SkeletalMeshLocation.SampledPosition
    Position Start:Vector3f=&LerpPosition(A=$Particles.ShapeLocation001.ShapePosition, B=$System.Lighting Start, Alpha=0)
    Write Particles Position=true
   )
   /SetVariables(Particles.RibbonLength:NiagaraFloat=&DistanceBetweenPositions(
     A=&LerpPosition(A=$Particles.ShapeLocation001.ShapePosition, B=$System.Lighting Start, Alpha=0)
     B=$Particles.SkeletalMeshLocation.SampledPosition
    )) {isSetParameters=true, module=SetVariables_#16}
   /SetVariables(Particles.W:NiagaraFloat=&Frac_Float(Float=&Add_Float(
      A=$Particles.NormalizedAge
      B=&RandomRangeFloat(
       Minimum=0
       Maximum=10
       Randomness Mode=::NewEnumerator1
       Random Seed=&Add_Integer(A=$Particles.PartitionParticles.Partition, B=3)
       Fixed Random Seed=true
      )
     ))) {isSetParameters=true, module=SetVariables_#17}
   &InterpolateOverTime(Rate Of Change:NiagaraFloat=&Divide_Float(A=1, B=0.25), Target Float Value:NiagaraFloat=$Particles.RDM)
   &self:BendRibbon_01(
    BigNoiseMask:NiagaraFloat=&FloatFromCurve(
     FloatCurve=DataInterface() {propertyValues {Curve {keys [
         {arriveTangent=2.6981, interpMode=RCIM_Cubic, leaveTangent=2.6981, tangentMode=RCTM_User, time=0, value=0}
         {interpMode=RCIM_Cubic, tangentMode=RCTM_User, time=0.5, value=1}
         {arriveTangent=-0.3624, interpMode=RCIM_Cubic, tangentMode=RCTM_Break, time=1, value=0}
        ]}}}
     CurveIndex=$Particles.RibbonLinkOrder
    )
    Offset:NiagaraFloat=&FloatFromCurve(
     FloatCurve=DataInterface() {propertyValues {Curve {keys [
         {arriveTangent=-0.0524, interpMode=RCIM_Cubic, leaveTangent=-0.0524, tangentMode=RCTM_User, time=0, value=0}
         {interpMode=RCIM_Cubic, tangentMode=RCTM_User, time=0.5, value=1}
         {arriveTangent=-4.0426, tangentMode=RCTM_User, time=1, value=0}
        ]}}}
     CurveIndex=$Particles.RibbonLinkOrder
     Scale Curve=&FloatFromCurve(
      FloatCurve=DataInterface() {propertyValues {Curve {keys [{interpMode=RCIM_Cubic, leaveTangent=2.3336, tangentMode=RCTM_Break, time=0, value=0.2}, <at 1 1>]}}}
      CurveIndex=$Particles.W
      Scale Curve=&Multiply_Float(
       A=&RemapRange(
        Input Value To Remap=$Particles.InterpolateOverTime.InterpolateOverTime_Float.Moving Average
        Input Min=0
        Input Max=1
        Output Min=-1
        Output Max=1
       )
       B=10
      )
     )
    )
    RibbonLength:NiagaraFloat=$Particles.RibbonLength
    YBigNoise:NiagaraFloat=&Waveform(
     Global Input Position=&Multiply_Float(A=$Particles.InterpolateOverTime.InterpolateOverTime_Float.Moving Average, B=10)
     Global Phase=&Add_Float(
      A=&Multiply_Float(A=$Particles.RibbonLinkOrder, B=2)
      B=&Multiply_Float(A=$Emitter.Age, B=-2)
     )
     "Waveform [1]"=::NewEnumerator12
     "[1] Frequency"=1
     "[1] Amplitude Min/Max"=<-1 1>
     "[1] Scale"=0
    )
    ZBigNoise:NiagaraFloat=&Waveform(
     Global Input Position=&Multiply_Float(A=$Particles.InterpolateOverTime.InterpolateOverTime_Float.Moving Average, B=10)
     Global Phase=&Add_Float(
      A=&Multiply_Float(A=$Particles.RibbonLinkOrder, B=3)
      B=&Multiply_Float(A=$Emitter.Age, B=2)
     )
     "Waveform [1]"=::NewEnumerator12
     "[1] Frequency"=1
     "[1] Amplitude Min/Max"=<-1 0>
     "[1] Scale"=15
    )
   ) {module=BendRibbon_01}
   &ScaleRibbonWidth(Ribbon Width Scale=&FloatFromCurve(
     FloatCurve=DataInterface() {propertyValues {Curve {keys [
         {interpMode=RCIM_Cubic, tangentMode=RCTM_Break, time=0, value=0.6}
         {interpMode=RCIM_Cubic, tangentMode=RCTM_Break, time=0.1, value=0.4}
         {interpMode=RCIM_Cubic, tangentMode=RCTM_Break, time=0.9, value=0.4}
         <at 1 1>
        ]}}}
     CurveIndex=$Particles.PartitionParticles.Normalized Index In Partition
     Scale Curve=&FloatFromCurve(
      FloatCurve=DataInterface() {propertyValues {Curve {keys [
          {tangentMode=RCTM_User, time=0, value=1}
          {arriveTangent=-0.0405, interpMode=RCIM_Cubic, leaveTangent=-0.0405, tangentMode=RCTM_User, time=0.5, value=1}
          {arriveTangent=-4.0987, interpMode=RCIM_Cubic, leaveTangent=-4.0987, tangentMode=RCTM_User, time=1, value=0}
         ]}}}
      CurveIndex=$Particles.W
      Scale Curve=&FloatFromCurve(
       FloatCurve=DataInterface() {propertyValues {Curve {keys [
           {arriveTangent=20.043, interpMode=RCIM_Cubic, leaveTangent=20.043, tangentMode=RCTM_User, time=0, value=0}
           {arriveTangent=-0.2233, interpMode=RCIM_Cubic, leaveTangent=-0.2233, tangentMode=RCTM_User, time=0.1, value=1}
          ]}}}
       CurveIndex=$Particles.PartitionParticles.Normalized Index In Partition
      )
     )
    ))
   &ScaleRibbonWidth(Ribbon Width Scale=&RemapRange(Input Value To Remap=$Particles.RDM, Input Min=0, Input Max=1, Output Min=0.7, Output Max=2)) {module=ScaleRibbonWidth002}
   &DynamicMaterialParameters(
    Index 0 Param 1=$Particles.ScratchModule_03.Z
    Index 0 Param 2=$Particles.ScratchModule_03.Y
    Index 0 Param 3=$Particles.ScratchModule_03.Z
    Index 0 Param 4=$Particles.ScratchModule_03.W
    Index 1 Param 1=&FloatFromCurve(
     FloatCurve=DataInterface() {propertyValues {Curve {keys [{tangentMode=RCTM_User, time=0, value=0}, {tangentMode=RCTM_User, time=1, value=1}]}}}
     CurveIndex=$Particles.W
    )
    Index 1 Param 2=&FloatFromCurve(FloatCurve=DataInterface() {propertyValues {Curve {keys [<at 0 0>,<at 0.3 1>]}}}, CurveIndex=$Particles.W)
    Index 1 Param 3=1
    Index 1 Param 4=&Add_Float(
     A=&FloatFromCurve(
      FloatCurve=DataInterface() {propertyValues {Curve {keys [{tangentMode=RCTM_User, time=0, value=-1}, {tangentMode=RCTM_User, time=1, value=0.7}]}}}
      CurveIndex=$Particles.W
     )
     B=&FloatFromCurve(
      FloatCurve=DataInterface() {propertyValues {Curve {keys [
          {interpMode=RCIM_Cubic, tangentMode=RCTM_Break, time=0, value=-1}
          {interpMode=RCIM_Cubic, tangentMode=RCTM_Break, time=0.1, value=0}
          {interpMode=RCIM_Cubic, tangentMode=RCTM_Break, time=0.9, value=0}
          <at 1 -1>
         ]}}}
      CurveIndex=$Particles.RibbonLinkOrder
      Scale Curve=0.3
     )
    )
    Module.Index0Param0Write=true
    Module.Index0Param1Write=true
    Module.Index0Param2Write=true
    Module.Index0Param3Write=true
    Module.Index1Param0Write=true
    Module.Index1Param1Write=true
    Module.Index1Param2Write=true
    Module.Index1Param3Write=true
    Write Parameter Index 1=true
   ) {enabled=false}
   &DynamicMaterialParameters(
    Index 0 Param 1=$Particles.ScratchModule_03.Z
    Index 0 Param 2=$Particles.ScratchModule_03.Y
    Index 0 Param 3=$Particles.ScratchModule_03.Z
    Index 0 Param 4=$Particles.ScratchModule_03.W
    Index 1 Param 1=&FloatFromCurve(
     FloatCurve=DataInterface() {propertyValues {Curve {keys [{tangentMode=RCTM_User, time=0, value=0}, {tangentMode=RCTM_User, time=1, value=1}]}}}
     CurveIndex=$Particles.W
    )
    Index 1 Param 2=&FloatFromCurve(
     FloatCurve=DataInterface() {propertyValues {Curve {keys [
         {arriveTangent=9.0871, interpMode=RCIM_Cubic, time=0, value=0}
         {arriveTangent=6.2186, interpMode=RCIM_Cubic, leaveTangent=6.2186, tangentMode=RCTM_User, time=0.3, value=1}
        ]}}}
     CurveIndex=$Particles.W
    )
    Index 1 Param 3=&FloatFromCurve(FloatCurve=DataInterface() {propertyValues {Curve {keys [<at 0 0>,<at 0.5 1>,<at 0.6 1>,<at 1 0>]}}}, CurveIndex=$Particles.W)
    Index 1 Param 4=&Add_Float(
     A=&FloatFromCurve(
      FloatCurve=DataInterface() {propertyValues {Curve {keys [{tangentMode=RCTM_User, time=0, value=-1}, {tangentMode=RCTM_User, time=1, value=0.7}]}}}
      CurveIndex=$Particles.W
     )
     B=&FloatFromCurve(
      FloatCurve=DataInterface() {propertyValues {Curve {keys [
          {interpMode=RCIM_Cubic, tangentMode=RCTM_Break, time=0, value=-1}
          {interpMode=RCIM_Cubic, tangentMode=RCTM_Break, time=0.1, value=0}
          {interpMode=RCIM_Cubic, tangentMode=RCTM_Break, time=0.9, value=0}
          <at 1 -1>
         ]}}}
      CurveIndex=$Particles.RibbonLinkOrder
      Scale Curve=0.3
     )
    )
    Module.Index0Param0Write=true
    Module.Index0Param1Write=true
    Module.Index0Param2Write=true
    Module.Index0Param3Write=true
    Module.Index1Param0Write=true
    Module.Index1Param1Write=true
    Module.Index1Param2Write=true
    Module.Index1Param3Write=true
    Write Parameter Index 1=true
   ) {module=DynamicMaterialParameters001}
   /SetVariables(Particles.RibbonW2:NiagaraFloat=&Multiply_Float(A=$Particles.RibbonWidth, B=2.5)) {isSetParameters=true, module=SetVariables_#18}
   /SetVariables(
    Particles.High Width:NiagaraFloat=&MakeCustomFloatFromBool(Bool=&SetBoolByFloatComparison(A=$System.CloudDot, B=0.5), True Float=$Particles.RibbonWidth, False Float=0)
    Particles.low Width:NiagaraFloat=&MakeCustomFloatFromBool(Bool=&SetBoolByFloatComparison(A=$System.CloudDot, B=0.5), True Float=1, False Float=$Particles.RibbonWidth)
   ) {isSetParameters=true, module=SetVariables_#22}
   /SetVariables(
    Particles.High Width 2:NiagaraFloat=&MakeCustomFloatFromBool(Bool=&SetBoolByFloatComparison(A=$System.CloudDot, B=0.5), True Float=$Particles.RibbonW2, False Float=0)
    Particles.low Width 2:NiagaraFloat=&MakeCustomFloatFromBool(Bool=&SetBoolByFloatComparison(A=$System.CloudDot, B=0.5), True Float=1, False Float=$Particles.RibbonW2)
   ) {isSetParameters=true, module=SetVariables_#23}
   /SetVariables(
    Particles.Glow Cam Offset=100
    Particles.Sprite Color:LinearColor=&MultiplyLinearColor(
     Linear Color=<rgba 0.15 0.6 1.5 0.1>
     Scale Factor=&MakeFloat4FromVectorAndFloat(
      "Vector (RGB)"=&VectorFromFloat(Value=&FNTimeOfDayScalar(Night Scale=1, Morning Scale=1, Evening Scale=1))
      "Float (Alpha)"=&FloatFromCurve(
       FloatCurve=DataInterface() {propertyValues {Curve {keys [<at -0.4 0.15>,<at -0.2 1.25>]}}}
       CurveIndex=&DotProduct(Vector A=$System.CamForVec, Vector B=<0 0 1>)
       Scale Curve=&FloatFromCurve(
        FloatCurve=DataInterface() {propertyValues {Curve {keys [{tangentMode=RCTM_User, time=0.655, value=0}, {tangentMode=RCTM_User, time=0.855, value=1}]}}}
        CurveIndex=$Engine.System.Age
        Scale Curve=&FloatFromCurve(
         FloatCurve=DataInterface() {propertyValues {Curve {keys [
             {arriveTangent=9.0725, interpMode=RCIM_Cubic, leaveTangent=9.0725, tangentMode=RCTM_User, time=0, value=0}
             {arriveTangent=-2.5249, interpMode=RCIM_Cubic, leaveTangent=-2.5249, tangentMode=RCTM_User, time=0.05, value=1}
             {arriveTangent=-0.0405, interpMode=RCIM_Cubic, leaveTangent=-0.0405, tangentMode=RCTM_User, time=1, value=0}
            ]}}}
         CurveIndex=$Particles.W
         Scale Curve=2
        )
       )
      )
     )
    )
    "Particles.SpriteSizeGlow ":Vector2f=&Vector2DFromFloat(Value=&FloatFromCurve(
      FloatCurve=DataInterface() {propertyValues {Curve {keys [<at 0 1 RCIM_Constant>,<at 0.01 0 RCIM_Constant>,<at 0.99 1 RCIM_Constant>]}}}
      CurveIndex=$Particles.PartitionParticles.Normalized Index In Partition
      Scale Curve=&Lerp_Float(A=40, B=40, Alpha=0.5)
     ))
    Particles.dynam:Vector4f=<2 1 0 0>
   ) {isSetParameters=true}
   /SetVariables(
    Particles.DEcalSprite Color:LinearColor=&MultiplyLinearColor(
     Linear Color=&MultiplyLinearColor(
      Linear Color=<rgba 0.3 1.2 3 1>
      Scale Factor=&MakeFloat4FromVectorAndFloat(
       "Vector (RGB)"=&VectorFromFloat(Value=&FNTimeOfDayScalar(Night Scale=0.4, Morning Scale=1, Evening Scale=1))
       "Float (Alpha)"=&FloatFromCurve(
        FloatCurve=DataInterface() {propertyValues {Curve {keys [
            {arriveTangent=9.0725, interpMode=RCIM_Cubic, leaveTangent=9.0725, tangentMode=RCTM_User, time=0, value=0}
            {arriveTangent=-2.5249, interpMode=RCIM_Cubic, leaveTangent=-2.5249, tangentMode=RCTM_User, time=0.05, value=1}
            {arriveTangent=-0.0405, interpMode=RCIM_Cubic, leaveTangent=-0.0405, tangentMode=RCTM_User, time=1, value=0}
           ]}}}
        CurveIndex=$Particles.W
        Scale Curve=2
       )
      )
     )
     Scale Factor=&MakeFloat4FromVectorAndFloat(
      "Vector (RGB)"=&VectorFromFloat(Value=&FloatFromCurve(
        FloatCurve=DataInterface() {propertyValues {Curve {keys [<at -0.4 0.15>,<at -0.2 1.25>]}}}
        CurveIndex=&DotProduct(Vector A=$System.CamForVec, Vector B=<0 0 1>)
        Scale Curve=&FloatFromCurve(
         FloatCurve=DataInterface() {propertyValues {Curve {keys [{tangentMode=RCTM_User, time=0.655, value=0}, {tangentMode=RCTM_User, time=0.855, value=1}]}}}
         CurveIndex=$Engine.System.Age
        )
       ))
      "Float (Alpha)"=1
     )
    )
    "Particles.Decal SpriteSizeGlow ":Vector2f=&Vector2DFromFloat(Value=&FloatFromCurve(
      FloatCurve=DataInterface() {propertyValues {Curve {keys [<at 0 1 RCIM_Constant>,<at 0.01 0>]}}}
      CurveIndex=$Particles.PartitionParticles.Normalized Index In Partition
      Scale Curve=&Lerp_Float(
       A=35
       B=50
       Alpha=&FloatFromCurve(
        FloatCurve=DataInterface() {propertyValues {Curve {keys [<at -0.5 0.15>,<at -0.1 1.25>]}}}
        CurveIndex=&DotProduct(Vector A=$System.CamForVec, Vector B=<0 0 1>)
        Scale Curve=&FloatFromCurve(
         FloatCurve=DataInterface() {propertyValues {Curve {keys [{tangentMode=RCTM_User, time=0.655, value=0}, {tangentMode=RCTM_User, time=0.855, value=1}]}}}
         CurveIndex=$Engine.System.Age
        )
       )
      )
     ))
   ) {isSetParameters=true}
   /SetVariables(Particles.Front decal col:LinearColor=&MultiplyLinearColor(
     Linear Color=$Particles.DEcalSprite Color
     Scale Factor=&MakeFloat4FromVectorAndFloat("Vector (RGB)"=&VectorFromFloat(Value=1), "Float (Alpha)"=0.05)
    )) {isSetParameters=true}
  ]
  rendererClasses [&NiagaraRibbonRendererProperties, &NiagaraSpriteRendererProperties]
  renderers [
   {rendererClass=&NiagaraRibbonRendererProperties}
   {rendererClass=&NiagaraRibbonRendererProperties, rendererIndex=1}
   {rendererClass=&NiagaraRibbonRendererProperties, rendererIndex=2}
   {rendererClass=&NiagaraRibbonRendererProperties, rendererIndex=3}
   {rendererClass=&NiagaraSpriteRendererProperties, rendererIndex=4}
   {rendererClass=&NiagaraSpriteRendererProperties, rendererIndex=5}
   {rendererClass=&NiagaraSpriteRendererProperties, rendererIndex=6}
  ]
 }
 Small_Lightning {
  emitterSpawnScript [
   /SetVariables(Emitter.particles per line:NiagaraInt32=&MakeCustomIntFromBool(
     Bool=&IsPlatformSetActive(New Platform Set=DataInterface() {propertyValues {Platforms {qualityLevelMask=-31}}})
     True INT=4
     False INT=&MakeCustomIntFromBool(
      Bool=&IsPlatformSetActive(New Platform Set=DataInterface() {propertyValues {Platforms {qualityLevelMask=-26}}})
      True INT=6
      False INT=8
     )
    )) {isSetParameters=true, module=SetVariables_#14}
   /SetVariables(Emitter.Num Groups=4) {isSetParameters=true, module=SetVariables_#15}
  ]
  emitterUpdateScript [
   &EmitterState(Delay First Loop Only=true, Life Cycle Mode=::NewEnumerator1, Loop Delay=1.5, Loop Duration=1, UseLoopDelay=true)
   &SpawnBurst_Instantaneous(
    Spawn Count=&Multiply_Int(A=$Emitter.particles per line, B=$Emitter.Num Groups)
    Use Spawn Probability=false
    UseLoopCountLimit=true
   )
   /SetVariables(
    Emitter.Timer 2:NiagaraFloat=$System.Sys Timer 2
    Emitter.Timer 3:NiagaraFloat=$System.Sys Timer 3
    Emitter.timer 1:NiagaraFloat=$System.Sys Timer 1
    Emitter.timer 4:NiagaraFloat=$System.Sys Timer 4
   ) {isSetParameters=true}
   /SetVariables(
    Emitter.Shape 3:NiagaraInt32=$System.SysShape 3
    Emitter.shape 1:NiagaraInt32=$System.SysShape 1
    Emitter.shape 2:NiagaraInt32=$System.SysShape 2
    Emitter.shape 4:NiagaraInt32=$System.SysShape 4
   ) {isSetParameters=true}
   /SetVariables(
    Emitter.ini Pos Y 1:NiagaraFloat=$System.sys ini Pos Y 1
    Emitter.ini Pos Y 2:NiagaraFloat=$System.sys ini Pos Y 2
    Emitter.ini Pos Y 3:NiagaraFloat=$System.sys ini Pos Y 3
    Emitter.ini Pos Y 4:NiagaraFloat=$System.sys ini Pos Y 4
   ) {isSetParameters=true}
   /SetVariables(
    Emitter.ini Pos 1:NiagaraFloat=$System.sys ini Pos 1
    Emitter.ini Pos 2:NiagaraFloat=$System.sys ini Pos 2
    Emitter.ini Pos 3:NiagaraFloat=$System.sys ini Pos 3
    Emitter.ini Pos 4:NiagaraFloat=$System.sys ini Pos 4
   ) {isSetParameters=true}
  ]
  particleSpawnScript [
   &InitializeParticle(
    Color Mode=::NewEnumerator0
    Lifetime=0.5
    Ribbon Width=7
    Ribbon Width Mode=::NewEnumerator1
    Sprite Size Mode=::NewEnumerator3
    UsePositionOffset=false
    Write Material Random=true
   )
   &PartitionParticles(
    Normalized Index In Partition Includes One=true
    Number Of Partitions=$Emitter.Num Groups
    Particles Per Partition:NiagaraInt32=$Emitter.particles per line
    Partition Mode:ENiagara_PartitionMode=::NewEnumerator1
    Write Partition Index to Ribbon Link Order:ENiagara_AttributeSamplingApplyOutput=::NewEnumerator0
    Write Partition to Ribbon ID=::NewEnumerator0
   )
  ]
  particleUpdateScript [
   &ParticleState(Kill Particles When Lifetime Has Elapsed=false, Let Infinitely Lived Particles Die When Emitter Deactivates=true)
   &self:ScratchModule_08(
    Timer Amount:NiagaraFloat=&RandomRangeFloat(Minimum=0, Maximum=0.35)
    float:NiagaraFloat=&RandomRangeFloat(Minimum=0, Maximum=1, Evaluation Type=::NewEnumerator1)
   ) {module=ScratchModule_08}
   &self:ScratchModule_08(
    Timer Amount:NiagaraFloat=&RandomRangeFloat(Minimum=0.1, Maximum=0.35)
    float:NiagaraFloat=&RandomRangeFloat(Minimum=0, Maximum=1, Evaluation Type=::NewEnumerator1)
   ) {module=ScratchModule_08001}
   &SkeletalMeshLocation(
    Bone / Socket Index=$Particles.PartitionParticles.Partition
    Bone Sampling Mode=::NewEnumerator4
    MaskSpawnGroup=false
    Orientation Sampling=::NewEnumerator1
    Position Sampling=::NewEnumerator1
    Sampled Position Offset=&SelectVectorFromArray(
     Direct Array Index=$Particles.PartitionParticles.Partition
     Vector Selection Array=DataInterface() {propertyValues {FloatData [<5 0 0>, <-5 0 1>, <-5 7 0>, <5 -7 0>]}}
    )
    Skeletal Mesh=$System.Skeletal Mesh
    Write To Particle Position=false
   )
   &self:SLect_Shape_Mod(
    RandomSelctor:NiagaraInt32=$Particles.PartitionParticles.Partition
    Select Pos:NiagaraFloat=$Emitter.ini Pos 1
    Select Pos 2:NiagaraFloat=$Emitter.ini Pos 2
    Select Pos 3:NiagaraFloat=$Emitter.ini Pos 3
    Select Pos 4:NiagaraFloat=$Emitter.ini Pos 4
    Select Pos Y:NiagaraFloat=$Emitter.ini Pos Y 1
    Select Pos Y 2:NiagaraFloat=$Emitter.ini Pos Y 2
    Select Pos Y 3:NiagaraFloat=$Emitter.ini Pos Y 3
    Select Pos Y 4:NiagaraFloat=$Emitter.ini Pos Y 4
    Shape Select:NiagaraInt32=$Emitter.shape 1
    Shape Select 2:NiagaraInt32=$Emitter.shape 2
    Shape Select 3:NiagaraInt32=$Emitter.Shape 3
    Shape Select 4:NiagaraInt32=$Emitter.shape 4
    Timer Amount:NiagaraFloat=$Emitter.timer 1
    Timer Amount 2:NiagaraFloat=$Emitter.Timer 2
    Timer Amount 3:NiagaraFloat=$Emitter.Timer 3
    Timer Amount 4:NiagaraFloat=$Emitter.timer 4
   ) {module=SLect_Shape_Mod}
   &ShapeLocation(
    Handle Radius=18
    Large Radius=&MakeFloatFromVector(
     Vector=&VectorFromFloat(Value=&Lerp_Float(
       A=50
       B=75
       Alpha=&FloatFromCurve(
        FloatCurve=DataInterface() {propertyValues {Curve {keys [<at -1 1>,<at -0.8 0>]}}}
        CurveIndex=&DotProduct(Vector A=<0 0 1>, Vector B=$System.CamForVec)
       )
      ))
     Channel=::NewEnumerator0
    )
    MaskSpawnGroup=false
    Offset=<0 0 42>
    Offset Mode=::NewEnumerator0
    Shape Origin=$System.Lighting Start
    Shape Primitive=::NewEnumerator3
    Torus Distribution Mode:ENiagaraTorusDistributionMode=::NewEnumerator1
    U Position=&Add_Float(
     A=&self:torus_pos_float(
      float 1:NiagaraFloat=&Add_Float(A=0.25, B=0)
      float 2:NiagaraFloat=&Add_Float(A=0.5, B=0)
      float 3:NiagaraFloat=&Add_Float(A=0, B=0)
      float 4:NiagaraFloat=&Add_Float(A=0.75, B=0)
      Master Offset:NiagaraFloat=-0.15
      Partition:NiagaraInt32=$Particles.PartitionParticles.Partition
     )
     B=&Multiply_Float(A=$Particles.PartitionParticles.Normalized Index In Partition, B=0)
    )
    UseOwnerScale=true
    V Position=-0.1
   ) {module=ShapeLocation001}
   &self:ScratchModule_03(
    EndPos:NiagaraPosition=$Particles.SkeletalMeshLocation.SampledPosition
    StartPos:NiagaraPosition=&LerpPosition(A=$Particles.ShapeLocation001.ShapePosition, B=$System.Lighting Start, Alpha=0)
   ) {module=ScratchModule_03}
   &InterpolateOverTime(
    Initial Vector Value:Vector3f=<1 1 1>
    Rate Of Change:NiagaraFloat=&Divide_Float(A=1, B=0.3)
    Target Vector Value:Vector3f=&TransformVector(
     Vector=&Multiply_Vector(
      A=&SelectVectorFromArray(
       Direct Array Index=$Particles.PartitionParticles.Partition
       Vector Selection Array=DataInterface() {propertyValues {FloatData [<300 50 -75>, <-300 300 -75>, <50 -300 -75>, <-300 -50 -75>]}}
      )
      B=&SelectVectorFromArray(
       Direct Array Index=&MakeCustomIntFromBool(Bool=$System.Use Temp, True INT=$System.Temp Array Selection, False INT=$Particles.ShapeMod)
       Vector Selection Array=$System.Vector Array Top
      )
     )
     Destination Space=::Simulation
    )
    Variable Type=::NewEnumerator2
   ) {module=InterpolateOverTime001}
   &InterpolateOverTime(
    Initial Vector Value:Vector3f=<1 1 1>
    Rate Of Change:NiagaraFloat=&Divide_Float(A=1, B=0.3)
    Target Vector Value:Vector3f=&TransformVector(
     Vector=&Multiply_Vector(
      A=&SelectVectorFromArray(
       Direct Array Index=$Particles.PartitionParticles.Partition
       Vector Selection Array=DataInterface() {propertyValues {FloatData [<150 150 100>, <-150 150 100>, <150 -150 100>, <-150 -150 100>]}}
      )
      B=&SelectVectorFromArray(
       Direct Array Index=&MakeCustomIntFromBool(Bool=$System.Use Temp, True INT=$System.Temp Array Selection, False INT=$Particles.ShapeMod)
       Vector Selection Array=$System.Vector Array bottom
      )
     )
     Destination Space=::Simulation
    )
    Variable Type=::NewEnumerator2
   ) {module=InterpolateOverTime002}
   &FN_HermiteSpline(
    Direction End:Vector3f=$Particles.InterpolateOverTime002.InterpolateOverTime_V3_.Moving Average
    Direction Start:Vector3f=$Particles.InterpolateOverTime001.InterpolateOverTime_V3_.Moving Average
    Position Blend:NiagaraFloat=$Particles.PartitionParticles.Normalized Index In Partition
    Position End:Vector3f=$Particles.SkeletalMeshLocation.SampledPosition
    Position Start:Vector3f=&LerpPosition(A=$Particles.ShapeLocation001.ShapePosition, B=$System.Lighting Start, Alpha=0)
    Write Particles Position=true
   )
   /SetVariables(Particles.RibbonLength:NiagaraFloat=&DistanceBetweenPositions(
     A=&LerpPosition(A=$Particles.ShapeLocation001.ShapePosition, B=$System.Lighting Start, Alpha=0)
     B=$Particles.SkeletalMeshLocation.SampledPosition
    )) {isSetParameters=true, module=SetVariables_#16}
   /SetVariables(Particles.W:NiagaraFloat=&Frac_Float(Float=&Add_Float(
      A=$Particles.NormalizedAge
      B=&RandomRangeFloat(
       Minimum=0
       Maximum=2
       Randomness Mode=::NewEnumerator1
       Random Seed=&Add_Integer(A=$Particles.PartitionParticles.Partition, B=16)
       Fixed Random Seed=true
      )
     ))) {isSetParameters=true, module=SetVariables_#17}
   &InterpolateOverTime(Rate Of Change:NiagaraFloat=&Divide_Float(A=1, B=0.25), Target Float Value:NiagaraFloat=$Particles.RDM)
   &self:BendRibbon_01(
    BigNoiseMask:NiagaraFloat=&FloatFromCurve(
     FloatCurve=DataInterface() {propertyValues {Curve {keys [
         {arriveTangent=2.6981, interpMode=RCIM_Cubic, leaveTangent=2.6981, tangentMode=RCTM_User, time=0, value=0}
         {interpMode=RCIM_Cubic, tangentMode=RCTM_User, time=0.5, value=1}
         {arriveTangent=-0.3624, interpMode=RCIM_Cubic, tangentMode=RCTM_Break, time=1, value=0}
        ]}}}
     CurveIndex=$Particles.RibbonLinkOrder
    )
    Offset:NiagaraFloat=&FloatFromCurve(
     FloatCurve=DataInterface() {propertyValues {Curve {keys [
         {arriveTangent=-0.0524, interpMode=RCIM_Cubic, leaveTangent=-0.0524, tangentMode=RCTM_User, time=0, value=0}
         {interpMode=RCIM_Cubic, tangentMode=RCTM_User, time=0.5, value=1}
         {arriveTangent=-4.0426, tangentMode=RCTM_User, time=1, value=0}
        ]}}}
     CurveIndex=$Particles.RibbonLinkOrder
     Scale Curve=&FloatFromCurve(
      FloatCurve=DataInterface() {propertyValues {Curve {keys [{interpMode=RCIM_Cubic, leaveTangent=2.3336, tangentMode=RCTM_Break, time=0, value=0.2}, <at 1 1>]}}}
      CurveIndex=$Particles.W
      Scale Curve=&Multiply_Float(
       A=&RemapRange(
        Input Value To Remap=$Particles.InterpolateOverTime.InterpolateOverTime_Float.Moving Average
        Input Min=0
        Input Max=1
        Output Min=-1
        Output Max=1
       )
       B=10
      )
     )
    )
    RibbonLength:NiagaraFloat=$Particles.RibbonLength
    YBigNoise:NiagaraFloat=&Waveform(
     Global Input Position=&Multiply_Float(A=$Particles.InterpolateOverTime.InterpolateOverTime_Float.Moving Average, B=10)
     Global Phase=&Add_Float(
      A=&Multiply_Float(A=$Particles.RibbonLinkOrder, B=2)
      B=&Multiply_Float(A=$Emitter.Age, B=-2)
     )
     "Waveform [1]"=::NewEnumerator12
     "[1] Frequency"=1
     "[1] Amplitude Min/Max"=<-1 1>
     "[1] Scale"=0
    )
    ZBigNoise:NiagaraFloat=&Waveform(
     Global Input Position=&Multiply_Float(A=$Particles.InterpolateOverTime.InterpolateOverTime_Float.Moving Average, B=10)
     Global Phase=&Add_Float(
      A=&Multiply_Float(A=$Particles.RibbonLinkOrder, B=3)
      B=&Multiply_Float(A=$Emitter.Age, B=2)
     )
     "Waveform [1]"=::NewEnumerator12
     "[1] Frequency"=1
     "[1] Amplitude Min/Max"=<-1 0>
     "[1] Scale"=15
    )
   ) {module=BendRibbon_01}
   &ScaleRibbonWidth(Ribbon Width Scale=&FloatFromCurve(
     FloatCurve=DataInterface() {propertyValues {Curve {keys [
         {interpMode=RCIM_Cubic, tangentMode=RCTM_Break, time=0, value=0.6}
         {interpMode=RCIM_Cubic, tangentMode=RCTM_Break, time=0.1, value=0.4}
         {interpMode=RCIM_Cubic, tangentMode=RCTM_Break, time=0.9, value=0.4}
         <at 1 1>
        ]}}}
     CurveIndex=$Particles.PartitionParticles.Normalized Index In Partition
     Scale Curve=&FloatFromCurve(
      FloatCurve=DataInterface() {propertyValues {Curve {keys [
          {tangentMode=RCTM_User, time=0, value=1}
          {arriveTangent=-0.0405, interpMode=RCIM_Cubic, leaveTangent=-0.0405, tangentMode=RCTM_User, time=0.5, value=1}
          {arriveTangent=-4.0987, interpMode=RCIM_Cubic, leaveTangent=-4.0987, tangentMode=RCTM_User, time=1, value=0}
         ]}}}
      CurveIndex=$Particles.W
     )
    ))
   &ScaleRibbonWidth(Ribbon Width Scale=&RemapRange(Input Value To Remap=$Particles.RDM, Input Min=0, Input Max=1, Output Min=0.7, Output Max=2)) {module=ScaleRibbonWidth002}
   &DynamicMaterialParameters(
    Index 0 Param 1=$Particles.ScratchModule_03.Z
    Index 0 Param 2=$Particles.ScratchModule_03.Y
    Index 0 Param 3=$Particles.ScratchModule_03.Z
    Index 0 Param 4=$Particles.ScratchModule_03.W
    Index 1 Param 1=&FloatFromCurve(
     FloatCurve=DataInterface() {propertyValues {Curve {keys [{tangentMode=RCTM_User, time=0, value=0}, {tangentMode=RCTM_User, time=1, value=1}]}}}
     CurveIndex=$Particles.W
    )
    Index 1 Param 2=&FloatFromCurve(FloatCurve=DataInterface() {propertyValues {Curve {keys [<at 0 0>,<at 0.3 1>]}}}, CurveIndex=$Particles.W)
    Index 1 Param 3=1
    Index 1 Param 4=&Add_Float(
     A=&FloatFromCurve(
      FloatCurve=DataInterface() {propertyValues {Curve {keys [{tangentMode=RCTM_User, time=0, value=-1}, {tangentMode=RCTM_User, time=1, value=0.7}]}}}
      CurveIndex=$Particles.W
     )
     B=&FloatFromCurve(
      FloatCurve=DataInterface() {propertyValues {Curve {keys [
          {interpMode=RCIM_Cubic, tangentMode=RCTM_Break, time=0, value=-1}
          {interpMode=RCIM_Cubic, tangentMode=RCTM_Break, time=0.1, value=0}
          {interpMode=RCIM_Cubic, tangentMode=RCTM_Break, time=0.9, value=0}
          <at 1 -1>
         ]}}}
      CurveIndex=$Particles.RibbonLinkOrder
      Scale Curve=0.3
     )
    )
    Module.Index0Param0Write=true
    Module.Index0Param1Write=true
    Module.Index0Param2Write=true
    Module.Index0Param3Write=true
    Module.Index1Param0Write=true
    Module.Index1Param1Write=true
    Module.Index1Param2Write=true
    Module.Index1Param3Write=true
    Write Parameter Index 1=true
   ) {enabled=false}
   &DynamicMaterialParameters(
    Index 0 Param 1=$Particles.ScratchModule_03.Z
    Index 0 Param 2=$Particles.ScratchModule_03.Y
    Index 0 Param 3=$Particles.ScratchModule_03.Z
    Index 0 Param 4=$Particles.ScratchModule_03.W
    Index 1 Param 1=&FloatFromCurve(
     FloatCurve=DataInterface() {propertyValues {Curve {keys [{tangentMode=RCTM_User, time=0, value=0}, {tangentMode=RCTM_User, time=1, value=1}]}}}
     CurveIndex=$Particles.W
    )
    Index 1 Param 2=&FloatFromCurve(
     FloatCurve=DataInterface() {propertyValues {Curve {keys [
         {arriveTangent=9.0871, interpMode=RCIM_Cubic, time=0, value=0}
         {arriveTangent=6.2186, interpMode=RCIM_Cubic, leaveTangent=6.2186, tangentMode=RCTM_User, time=0.3, value=1}
        ]}}}
     CurveIndex=$Particles.W
    )
    Index 1 Param 3=&FloatFromCurve(FloatCurve=DataInterface() {propertyValues {Curve {keys [<at 0 0>,<at 0.5 1>,<at 0.6 1>,<at 1 0>]}}}, CurveIndex=$Particles.W, Scale Curve=2)
    Index 1 Param 4=&Add_Float(
     A=&FloatFromCurve(
      FloatCurve=DataInterface() {propertyValues {Curve {keys [{tangentMode=RCTM_User, time=0, value=-1}, {tangentMode=RCTM_User, time=0.5, value=0.7}]}}}
      CurveIndex=$Particles.W
     )
     B=&FloatFromCurve(
      FloatCurve=DataInterface() {propertyValues {Curve {keys [
          {interpMode=RCIM_Cubic, tangentMode=RCTM_Break, time=0, value=-1}
          {interpMode=RCIM_Cubic, tangentMode=RCTM_Break, time=0.1, value=0}
          {interpMode=RCIM_Cubic, tangentMode=RCTM_Break, time=0.9, value=0}
          <at 1 -1>
         ]}}}
      CurveIndex=$Particles.RibbonLinkOrder
      Scale Curve=0.3
     )
    )
    Module.Index0Param0Write=true
    Module.Index0Param1Write=true
    Module.Index0Param2Write=true
    Module.Index0Param3Write=true
    Module.Index1Param0Write=true
    Module.Index1Param1Write=true
    Module.Index1Param2Write=true
    Module.Index1Param3Write=true
    Write Parameter Index 1=true
   ) {module=DynamicMaterialParameters001}
   /SetVariables(Particles.RibbonW2:NiagaraFloat=&Multiply_Float(A=$Particles.RibbonWidth, B=2.5)) {isSetParameters=true, module=SetVariables_#18}
   /SetVariables(
    Particles.High Width:NiagaraFloat=&MakeCustomFloatFromBool(Bool=&SetBoolByFloatComparison(A=$System.CloudDot, B=0.5), True Float=$Particles.RibbonWidth, False Float=0)
    Particles.low Width:NiagaraFloat=&MakeCustomFloatFromBool(Bool=&SetBoolByFloatComparison(A=$System.CloudDot, B=0.5), True Float=1, False Float=$Particles.RibbonWidth)
   ) {isSetParameters=true, module=SetVariables_#22}
   /SetVariables(
    Particles.High Width 2:NiagaraFloat=&MakeCustomFloatFromBool(Bool=&SetBoolByFloatComparison(A=$System.CloudDot, B=0.5), True Float=$Particles.RibbonW2, False Float=0)
    Particles.low Width 2:NiagaraFloat=&MakeCustomFloatFromBool(Bool=&SetBoolByFloatComparison(A=$System.CloudDot, B=0.5), True Float=1, False Float=$Particles.RibbonW2)
   ) {isSetParameters=true, module=SetVariables_#23}
   /SetVariables(
    Particles.Glow Cam Offset=100
    Particles.Sprite Color:LinearColor=&MultiplyLinearColor(
     Linear Color=<rgba 0.15 0.6 1.5 0.1>
     Scale Factor=&MakeFloat4FromVectorAndFloat(
      "Vector (RGB)"=&VectorFromFloat(Value=&FNTimeOfDayScalar(Night Scale=1, Morning Scale=1, Evening Scale=1))
      "Float (Alpha)"=&FloatFromCurve(
       FloatCurve=DataInterface() {propertyValues {Curve {keys [<at -0.4 0.15>,<at -0.2 1.25>]}}}
       CurveIndex=&DotProduct(Vector A=$System.CamForVec, Vector B=<0 0 1>)
       Scale Curve=&FloatFromCurve(
        FloatCurve=DataInterface() {propertyValues {Curve {keys [{tangentMode=RCTM_User, time=0.655, value=0}, {tangentMode=RCTM_User, time=0.855, value=1}]}}}
        CurveIndex=$Engine.System.Age
        Scale Curve=&FloatFromCurve(
         FloatCurve=DataInterface() {propertyValues {Curve {keys [
             {arriveTangent=9.0725, interpMode=RCIM_Cubic, leaveTangent=9.0725, tangentMode=RCTM_User, time=0, value=0}
             {arriveTangent=-2.5249, interpMode=RCIM_Cubic, leaveTangent=-2.5249, tangentMode=RCTM_User, time=0.05, value=1}
             {arriveTangent=-0.0405, interpMode=RCIM_Cubic, leaveTangent=-0.0405, tangentMode=RCTM_User, time=1, value=0}
            ]}}}
         CurveIndex=$Particles.W
         Scale Curve=2
        )
       )
      )
     )
    )
    "Particles.SpriteSizeGlow ":Vector2f=&Vector2DFromFloat(Value=&FloatFromCurve(
      FloatCurve=DataInterface() {propertyValues {Curve {keys [<at 0 1 RCIM_Constant>,<at 0.01 0 RCIM_Constant>,<at 0.99 1 RCIM_Constant>]}}}
      CurveIndex=$Particles.PartitionParticles.Normalized Index In Partition
      Scale Curve=&Lerp_Float(A=40, B=40, Alpha=0.5)
     ))
    Particles.dynam:Vector4f=<2 1 0 0>
   ) {isSetParameters=true}
   /SetVariables(
    Particles.DEcalSprite Color:LinearColor=&MultiplyLinearColor(
     Linear Color=&MultiplyLinearColor(
      Linear Color=<rgba 0.3 1.2 3 1>
      Scale Factor=&MakeFloat4FromVectorAndFloat(
       "Vector (RGB)"=&VectorFromFloat(Value=&FNTimeOfDayScalar(Night Scale=0.4, Morning Scale=1, Evening Scale=1))
       "Float (Alpha)"=&FloatFromCurve(
        FloatCurve=DataInterface() {propertyValues {Curve {keys [
            {arriveTangent=9.0725, interpMode=RCIM_Cubic, leaveTangent=9.0725, tangentMode=RCTM_User, time=0, value=0}
            {arriveTangent=-2.5249, interpMode=RCIM_Cubic, leaveTangent=-2.5249, tangentMode=RCTM_User, time=0.05, value=1}
            {arriveTangent=-0.0405, interpMode=RCIM_Cubic, leaveTangent=-0.0405, tangentMode=RCTM_User, time=1, value=0}
           ]}}}
        CurveIndex=$Particles.W
        Scale Curve=2
       )
      )
     )
     Scale Factor=&MakeFloat4FromVectorAndFloat(
      "Vector (RGB)"=&VectorFromFloat(Value=&FloatFromCurve(
        FloatCurve=DataInterface() {propertyValues {Curve {keys [<at -0.4 0.15>,<at -0.2 1.25>]}}}
        CurveIndex=&DotProduct(Vector A=$System.CamForVec, Vector B=<0 0 1>)
        Scale Curve=&FloatFromCurve(
         FloatCurve=DataInterface() {propertyValues {Curve {keys [{tangentMode=RCTM_User, time=0.655, value=0}, {tangentMode=RCTM_User, time=0.855, value=1}]}}}
         CurveIndex=$Engine.System.Age
        )
       ))
      "Float (Alpha)"=1
     )
    )
    "Particles.Decal SpriteSizeGlow ":Vector2f=&Vector2DFromFloat(Value=&FloatFromCurve(
      FloatCurve=DataInterface() {propertyValues {Curve {keys [<at 0 1 RCIM_Constant>,<at 0.01 0>]}}}
      CurveIndex=$Particles.PartitionParticles.Normalized Index In Partition
      Scale Curve=&Lerp_Float(
       A=35
       B=50
       Alpha=&FloatFromCurve(
        FloatCurve=DataInterface() {propertyValues {Curve {keys [<at -0.5 0.15>,<at -0.1 1.25>]}}}
        CurveIndex=&DotProduct(Vector A=$System.CamForVec, Vector B=<0 0 1>)
        Scale Curve=&FloatFromCurve(
         FloatCurve=DataInterface() {propertyValues {Curve {keys [{tangentMode=RCTM_User, time=0.655, value=0}, {tangentMode=RCTM_User, time=0.855, value=1}]}}}
         CurveIndex=$Engine.System.Age
        )
       )
      )
     ))
   ) {isSetParameters=true}
   /SetVariables(Particles.Front decal col:LinearColor=&MultiplyLinearColor(
     Linear Color=$Particles.DEcalSprite Color
     Scale Factor=&MakeFloat4FromVectorAndFloat("Vector (RGB)"=&VectorFromFloat(Value=1), "Float (Alpha)"=0.05)
    )) {isSetParameters=true}
  ]
  rendererClasses [&NiagaraRibbonRendererProperties, &NiagaraSpriteRendererProperties]
  renderers [
   {rendererClass=&NiagaraRibbonRendererProperties}
   {rendererClass=&NiagaraRibbonRendererProperties, rendererIndex=1}
   {rendererClass=&NiagaraRibbonRendererProperties, rendererIndex=2}
   {rendererClass=&NiagaraRibbonRendererProperties, rendererIndex=3}
   {rendererClass=&NiagaraSpriteRendererProperties, rendererIndex=4}
   {rendererClass=&NiagaraSpriteRendererProperties, rendererIndex=5}
   {rendererClass=&NiagaraSpriteRendererProperties, rendererIndex=6}
  ]
 }
}
summary {
 systemName=NS_StunningMountainContrail
 emitters [
  {emitterName=Decal, rendererClasses [&NiagaraSpriteRendererProperties]}
  {emitterName=Light, rendererClasses [&NiagaraLightRendererProperties]}
  {bEnabled=false, emitterName=Debug_Dot, rendererClasses [&NiagaraSpriteRendererProperties]}
  {emitterName=Rotating_Air_Bottom, rendererClasses [&NiagaraMeshRendererProperties]}
  {bEnabled=false, emitterName=Plates, rendererClasses [&NiagaraMeshRendererProperties]}
  {emitterName=Lightning_Arcs, rendererClasses [&NiagaraSpriteRendererProperties]}
  {bEnabled=false, emitterName=Rotating_Air_Meshes_Middle001, rendererClasses [&NiagaraMeshRendererProperties]}
  {
   bEnabled=false
   emitterName=Lightning_strings
   rendererClasses [&NiagaraRibbonRendererProperties, &NiagaraSpriteRendererProperties]
  }
  {emitterName=Rotating_Air_Meshes, rendererClasses [&NiagaraMeshRendererProperties]}
  {bEnabled=false, emitterName=Rotating_Air_Meshes003, rendererClasses [&NiagaraMeshRendererProperties]}
  {
   emitterName=Main_lightning
   rendererClasses [&NiagaraRibbonRendererProperties, &NiagaraSpriteRendererProperties]
  }
  {
   emitterName=Secondary_Lightning
   rendererClasses [&NiagaraRibbonRendererProperties, &NiagaraSpriteRendererProperties]
  }
  {
   emitterName=Small_Lightning
   rendererClasses [&NiagaraRibbonRendererProperties, &NiagaraSpriteRendererProperties]
  }
  {emitterName=Rotating_Air_Meshes001, rendererClasses [&NiagaraMeshRendererProperties]}
 ]
 userVariables [
  User.bHasPickaxe {type=NiagaraBool, defaultValue=true}
  User.bIsFrontEnd {type=NiagaraBool}
  User.bIsFrontEndPreview {type=NiagaraBool}
  User.ParaGlideLeanAlpha {type=NiagaraFloat, defaultValue=0, description=""}
  User.RandomArray {
   type=NiagaraDataInterfaceArrayFloat
   defaultValue {
    struct=NiagaraExt_VariableValue_DataInterface
    value {
     dataInterface=&self:NiagaraDataInterfaceArrayFloat_0
     dataInterfaceClass=&NiagaraDataInterfaceArrayFloat
    }
   }
  }
  User.Simple Counter {
   type=NiagaraDataInterfaceSimpleCounter
   description=""
   defaultValue {
    struct=NiagaraExt_VariableValue_DataInterface
    value {
     dataInterface=&self:NiagaraDataInterfaceSimpleCounter_0
     dataInterfaceClass=/Script/Niagara.NiagaraDataInterfaceSimpleCounter
    }
   }
  }
  User.Vector 4 Array {
   type=NiagaraDataInterfaceArrayFloat4
   description=""
   defaultValue {
    struct=NiagaraExt_VariableValue_DataInterface
    value {
     dataInterface=&self:NiagaraDataInterfaceArrayFloat4_0
     dataInterfaceClass=/Script/Niagara.NiagaraDataInterfaceArrayFloat4
    }
   }
  }
 ]
}
